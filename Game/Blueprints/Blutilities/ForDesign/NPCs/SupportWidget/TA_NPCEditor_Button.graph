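# Icon-and-label button widget derived from TA_BaseEditorButtonWidget, apparently for the NPC editor tool.
# It supplies its border and text to the base class, and colors the background through the validated brush and a dynamic material parameter.
# Visual validation applies the icon brush and transform, collapsing the icon without a valid resource and the text when the label is empty.
# New list data sets selection from selectedByDefault without throwing an event.
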WidgetBlueprint /Game/Blueprints/Blutilities/ForDesign/NPCs/SupportWidget/TA_NPCEditor_Button
graphs [
 EventGraph {nodes [
   K2Node_Event_1 = CustomEvent!() -> (OutputDelegate, then, currentListItemData:Object Reference) {is=title K2Node_CustomEvent}
   > Parent: handleNewListDataSet!(execute, currentListItemData:Object Reference=@K2Node_Event_1.currentListItemData) -> (then) {is=title K2Node_CallParentFunction}
   > n1 = Cast To TA_GenericListItemData!(execute, Object=@K2Node_Event_1.currentListItemData) -> (then, CastFailed, AsTA Generic List Item Data:TA Generic List Item Data Object Reference) {is=title K2Node_DynamicCast}
   > setSelected!(
    execute
    self:TA Base Editor Button Widget Object Reference
    selected:bool=Get selectedByDefault!(self:TA Generic List Item Data Object Reference=@n1) -> (selectedByDefault:bool)
    throwEvent:bool=false
   ) -> (then)
  ]}
 getBackgroundWidget {nodes [
   getBackgroundWidget() {is=title K2Node_FunctionEntry}
   > Return Node!(execute, backgroundWidget:Widget Object Reference=Get borderMain() -> (borderMain:Border Object Reference))
  ]}
 getLabelWidget {nodes [
   getLabelWidget() {is=title K2Node_FunctionEntry}
   > Return Node!(execute, labelWidget:Widget Object Reference=Get txtMain() -> (txtMain:Text Object Reference))
  ]}
 validateBackgroundColor {nodes [
   validateBackgroundColor() {is=title K2Node_FunctionEntry}
   > SetBrush(self:Border Object Reference=Get borderMain(), InBrush=getValidatedButtonBrush!() -> (brush:Slate Brush Structure))
   > n1 = GetDynamicMaterial(self:Border Object Reference=Get borderMain())
   > SetVectorParameterValue(self=@n1, ParameterName=BaseColor_DoNotSetOnInstances, Value=getValidatedButtonColor!() -> (buttonColor_Default:Linear Color Structure))
   > Return Node!(execute, handled:bool=false)
   Return Node(handled=false)
  ]}
 validateVisuals {nodes [
   validateVisuals() {is=title K2Node_FunctionEntry}
   > Parent: validateVisuals!(execute) -> (then, " ":bool) {is=title K2Node_CallParentFunction}
   > n1 = Sequence()
   > n2 = Is Valid(InputObject=Break Slate Brush!(SlateBrush:"Slate Brush Structure (by ref)"=Get iconBrush() -> (iconBrush:Slate Brush Structure)) -> (ResourceObject:Object Reference, OutlineSettings:Slate Brush Outline Settings Structure) {is=title K2Node_BreakStruct})
   SetVisibility(execute=@n2.Is Not Valid, self=Get imageIcon() -> (imageIcon:Image Object Reference), InVisibility=Collapsed)
   SetBrush(execute=@n2.Is Valid, self:Image Object Reference=Get imageIcon(), InBrush=Get iconBrush())
   > SetVisibility(self=Get imageIcon(), InVisibility=HitTestInvisible)
   > SetRenderTransform!(
    execute
    self:Widget Object Reference=Get imageIcon()
    InTransform:Widget Transform Structure=Get iconTransform() -> (iconTransform:Widget Transform Structure)
   ) -> (then)
   n3 = Len(S=Get label() -> (label:String))
   SetVisibility(
    execute=@n1.then_1
    self=Get txtMain()
    InVisibility=Select(Option 0=Collapsed, Option 1=HitTestInvisible, Index="integer > integer"(A=@n3, B=0) {is=title K2Node_CallFunction})
   )
   Return Node!(execute, " ":bool=false)
  ]}
]
variables [
 iconBrush {
  type=SlateBrush
  default {
   bIsDynamicallyLoaded=false
   drawAs=Image
   imageType=NoImage
   mirroring=NoMirror
   resourceName=None
   resourceObject=null
   tiling=NoTile
   imageSize=<32 32>
   margin=<ltrb 0 0 0 0>
   outlineSettings {
    bUseBrushTransparency=false
    roundingType=HalfHeightRadius
    width=0
    color {colorUseRule=UseColor_Specified, specifiedColor=<rgba 0 0 0 0>}
    cornerRadii=<0 0 0 1>
   }
   tintColor {colorUseRule=UseColor_Specified, specifiedColor=<rgba 1 1 1 1>}
   uVRegion {bIsValid=false, max=<0 0>, min=<0 0>}
  }
 }
 iconTransform {type=WidgetTransform, default {angle=0, scale=<1 1>, shear=<0 0>, translation=<0 0>}}
]
widgets {
 info {
  inheritedWidgetCount=0
  parentClass=/Game/Blueprints/Blutilities/ForTechArt/Widgets/Buttons/TA_BaseEditorButtonWidget.TA_BaseEditorButtonWidget_C
  rootWidgetClass=&Border
  widgetCount=4
 }
 widgets [
  borderMain {
   widgetClassPath=&Border
   bIsVariable=true
   widget=borderMain
   properties {
    background {
     bIsDynamicallyLoaded=false
     drawAs=Box
     imageType=NoImage
     mirroring=NoMirror
     resourceName=None
     resourceObject=null
     tiling=NoTile
     imageSize=<32 32>
     margin=<ltrb 0.15 0.15 0.15 0.15>
     outlineSettings {
      bUseBrushTransparency=false
      roundingType=HalfHeightRadius
      width=0
      color {colorUseRule=UseColor_Specified, specifiedColor=<rgba 0 0 0 0>}
      cornerRadii=<0 0 0 0>
     }
     tintColor {colorUseRule=UseColor_Specified, specifiedColor=<rgba 0.85 0.85 0.85 1>}
     uVRegion {bIsValid=false, max=<0 0>, min=<0 0>}
    }
    padding=<ltrb 8 4 8 4>
   }
  }
  HorizontalBox_0 {
   widgetClassPath=&HorizontalBox
   parent=borderMain
   widget=HorizontalBox_0
   slot {class=BorderSlot, properties {padding=<ltrb 8 4 8 4>}}
  }
  imageIcon {
   widgetClassPath=&Image
   bIsVariable=true
   parent=HorizontalBox_0
   widget=imageIcon
   slot {class=HorizontalBoxSlot, properties {horizontalAlignment=HAlign_Center, verticalAlignment=VAlign_Center}}
  }
  txtMain {
   widgetClassPath=&TextBlock
   bIsVariable=true
   parent=HorizontalBox_0
   widget=txtMain
   properties {
    clipping=ClipToBounds
    justification=Center
    colorAndOpacity {colorUseRule=UseColor_Specified, specifiedColor=<rgba 0.0104 0.0104 0.0104 1>}
    font {
     bForceMonospaced=false
     bMaterialIsStencil=false
     fontMaterial=null
     fontObject=&NotoSans
     letterSpacing=0
     monospacedWidth=1
     size=10
     skewAmount=0
     typefaceFontName=NotoSans-Bold
     outlineSettings {
      bApplyOutlineToDropShadows=false
      bMiteredCorners=false
      bSeparateFillAlpha=false
      outlineMaterial=null
      outlineSize=0
      outlineColor=<rgba 0 0 0 1>
     }
    }
    shadowColorAndOpacity=<rgba 0 0 0 0.4>
   }
   slot {class=HorizontalBoxSlot, properties {verticalAlignment=VAlign_Center, size {sizeRule=Fill, value=1}}}
  }
 ]
}
