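# Editor utility widget that lets designers create, load, edit and save NPCs built from AI bot spawner data and its components.
# Selectable config class defaults supply template paths and naming for new components, character data, conversation and Lager event assets.
# New assets are made as child assets or duplicates, then component copies are linked to the active spawner data.
# Indexed buttons switch detail views, a Python file picker loads NPCs, and property-change bindings update tags.
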EditorUtilityWidgetBlueprint /Game/NeverCook/Blutilities/ForDesign/NPCs/EUW_NPCEditorTool
graphs [
 AddWidgetToRoot {nodes [
   n1 = AddWidgetToRoot!() -> (then, WidgetType:User Widget Class Reference) {is=title K2Node_FunctionEntry}
   n2 = Get CanvasRoot() -> (CanvasRoot:Canvas Panel Object Reference)
   > n3 = Is Valid(InputObject=@n2)
   PrintString(execute=@n3.Is Not Valid, InString="AddWidgetToRoot failed, there is no root canvas panel defined. Use SetCanvasRoot before calling this!")
   > Return Node!(execute, WidgetInstance:User Widget Object Reference)
   n4 = Create Widget!(
    execute:Exec=@n3.Is Valid
    WorldContextObject:Object Reference=@n2
    Class:User Widget Class Reference=@n1
    OwningPlayer:Player Controller Object Reference=GetOwningPlayer(self=@n2)
   ) -> (then:Exec, ReturnValue:User Widget Object Reference)
   > n5 = AddChildToCanvas!(execute, self:Canvas Panel Object Reference=Get CanvasRoot(), Content:Widget Object Reference=@n4) -> (then, ReturnValue:Canvas Panel Slot Object Reference)
   > SetOffsets!(execute, self:Canvas Panel Slot Object Reference=@n5, InOffset:Margin Structure=Make Margin(Left=0, Top=0, Right=0, Bottom=0)) -> (then)
   > SetAnchors!(
    execute
    self:Canvas Panel Slot Object Reference=@n5
    InAnchors:Anchors Structure=Make Anchors!(Minimum:Vector 2D Structure=(X=0,Y=0), Maximum:Vector 2D Structure=(X=1,Y=1)) -> (Anchors:Anchors Structure) {is=title K2Node_MakeStruct}
   ) -> (then)
   > SetZOrder!(execute, self:Canvas Panel Slot Object Reference=@n5, InZOrder:int=9999) -> (then)
   > Return Node(WidgetInstance=@n4)
  ]}
 Component_CreateAffiliationComponent {nodes [
   n1 = Component_CreateAffiliationComponent!() -> (then, DestDirectory:String, NPCName:String) {is=title K2Node_FunctionEntry}
   > n2 = GetTemplateConfig!(execute, self) -> (then, Config:BP NPCEditor Config Object Reference)
   n3 = GetAffiliationComponent!(self:Fort Athena AISpawner Data Object Reference=Get ActiveSpawnerData() -> (ActiveSpawnerData:Fort Athena AIBot Spawner Data Object Reference)) -> (ReturnValue:Fort Athena AISpawner Data Component Affiliation Base Object Reference)
   > n4 = ConvertFilePathToAssetPath!(execute, InPath:String=GetBaseFilename(InPath=Get Object Path String(Object=@n3), bRemovePath=false)) -> (then, OutPath:String, ReturnValue:bool)
   n5 = Get DefaultAISCFolder!(self:BP NPCEditor Config Object Reference=@n2) -> (DefaultAISCFolder:String)
   > n6 = CreateAsset!(
    execute
    self:Self Object Reference
    SourceAssetPath:String=@n4
    DestinationAssetPath:String=Append(A=Append(A=@n1, B=@n5), B=Append(A=BP_AISC_Affiliation_, B=@n1.NPCName))
   ) -> (then, ReturnValue)
   > n7 = GetAISpawnerDataComponentDefaultObject!(execute, SourceObject:Object Reference=@n6) -> (then, ReturnValue:Fort Athena AISpawner Data Component Object Reference)
   > n8 = Cast To FortAthenaAISpawnerDataComponent_AffiliationBase!(execute, Object=@n7) -> (
    then
    CastFailed
    AsFort Athena AISpawner Data Component Affiliation Base:Fort Athena AISpawner Data Component Affiliation Base Object Reference
   ) {is=title K2Node_DynamicCast}
   > SetAffiliationComponent!(
    execute
    self:Fort Athena AISpawner Data Object Reference=Get ActiveSpawnerData()
    NewAffiliationComponent:Fort Athena AISpawner Data Component Affiliation Base Object Reference=@n8
   ) -> (then)
  ]}
 Component_CreateConversationComponent {nodes [
   n1 = Component_CreateConversationComponent!() -> (then, DestDirectory:String, NPCName:String, OverrideNPCCharacterData:bool=~) {is=title K2Node_FunctionEntry}
   > n2 = Sequence()
   > n3 = GetTemplateConfig()
   n4 = GetConversationComponent!(self:Fort Athena AISpawner Data Object Reference=Get ActiveSpawnerData()) -> (ReturnValue:Fort Athena AISpawner Data Component Conversation Base Object Reference)
   > n5 = ConvertFilePathToAssetPath(InPath=GetBaseFilename(InPath=Get Object Path String(Object=@n4), bRemovePath=false))
   n6 = Get DefaultAISCFolder(self=@n3)
   > n7 = CreateAsset(SourceAssetPath=@n5, DestinationAssetPath=Append(A=Append(A=@n1, B=@n6), B=Append(A=BP_AISC_Conv_, B=@n1.NPCName)))
   > n8 = GetAISpawnerDataComponentDefaultObject(SourceObject=@n7)
   > n9 = Cast To FortAthenaAISpawnerDataComponent_ConversationBase!(execute, Object=@n8) -> (
    then
    CastFailed
    AsFort Athena AISpawner Data Component Conversation Base:Fort Athena AISpawner Data Component Conversation Base Object Reference
   ) {is=title K2Node_DynamicCast}
   > n10 = Set NewConversationComponent(NewConversationComponent:Fort Athena AISpawner Data Component Conversation Base Object Reference=@n9) -> (Output_Get:Fort Athena AISpawner Data Component Conversation Base Object Reference)
   > SetConversationComponent!(
    execute
    self:Fort Athena AISpawner Data Object Reference=Get ActiveSpawnerData()
    NewConversationComponent:Fort Athena AISpawner Data Component Conversation Base Object Reference=@n10
   ) -> (then)
   Branch(execute=@n2.then_1, Condition=@n1.OverrideNPCCharacterData)
   > n11 = GetTemplateConfig()
   n12 = Get DefaultNPCCharacterDataAssetName!(self:BP NPCEditor Config Object Reference=@n11) -> (DefaultNPCCharacterDataAssetName:String)
   > n13 = DuplicateAsset!(
    execute
    SourceAssetPath:String=Get DefaultNPCCharacterDataTemplate!(self:BP NPCEditor Config Object Reference=@n11) -> (DefaultNPCCharacterDataTemplate:String)
    DestinationAssetPath:String=Append(A=@n1, B=Append(A=@n12, B=@n1.NPCName))
   ) -> (then, ReturnValue:Object Reference)
   > n14 = Cast To FortTandemCharacterData!(execute, Object=@n13) -> (then, CastFailed, AsFort Tandem Character Data:Fort Tandem Character Data Object Reference) {is=title K2Node_DynamicCast}
   > n15 = Cast To FortAthenaAISpawnerDataComponent_AIBotConversation!(
    execute
    Object=Get NewConversationComponent() -> (NewConversationComponent:Fort Athena AISpawner Data Component Conversation Base Object Reference)
   ) -> (
    then
    CastFailed
    AsFort Athena AISpawner Data Component AIBot Conversation:Fort Athena AISpawner Data Component AIBot Conversation Object Reference
   ) {is=title K2Node_DynamicCast}
   > SetEditorProperty(Object=@n15, PropertyName=CharacterData, PropertyValue:Fort Tandem Character Data Object Reference=@n14)
  ]}
 Component_CreateCosmeticComponent {nodes [
   n1 = Component_CreateCosmeticComponent!() -> (then, DestDirectory:String, NPCName:String) {is=title K2Node_FunctionEntry}
   > n2 = GetTemplateConfig()
   n3 = GetCosmeticComponent!(self:Fort Athena AISpawner Data Object Reference=Get ActiveSpawnerData()) -> (ReturnValue:Fort Athena AISpawner Data Component Cosmetic Base Object Reference)
   > n4 = ConvertFilePathToAssetPath(InPath=GetBaseFilename(InPath=Get Object Path String(Object=@n3), bRemovePath=false))
   n5 = Get DefaultAISCFolder(self=@n2)
   > n6 = CreateAsset(SourceAssetPath=@n4, DestinationAssetPath=Append(A=Append(A=@n1, B=@n5), B=Append(A=BP_AISC_CL_, B=@n1.NPCName)))
   > n7 = GetAISpawnerDataComponentDefaultObject(SourceObject=@n6)
   > n8 = Cast To FortAthenaAISpawnerDataComponent_CosmeticLoadout!(execute, Object=@n7) -> (
    then
    CastFailed
    AsFort Athena AISpawner Data Component Cosmetic Loadout:Fort Athena AISpawner Data Component Cosmetic Loadout Object Reference
   ) {is=title K2Node_DynamicCast}
   > SetCosmeticComponent!(
    execute
    self:Fort Athena AISpawner Data Object Reference=Get ActiveSpawnerData()
    NewCosmeticComponent:Fort Athena AISpawner Data Component Cosmetic Base Object Reference=@n8
   ) -> (then)
  ]}
 Component_CreateGameplayAbilityComponent {nodes [
   n1 = Component_CreateGameplayAbilityComponent!() -> (then, DestDirectory:String, NPCName:String) {is=title K2Node_FunctionEntry}
   > n2 = GetTemplateConfig()
   n3 = GetGameplayAbilityComponent!(self:Fort Athena AISpawner Data Object Reference=Get ActiveSpawnerData()) -> (ReturnValue:Fort Athena AISpawner Data Component Gameplay Ability Base Object Reference)
   > n4 = ConvertFilePathToAssetPath(InPath=GetBaseFilename(InPath=Get Object Path String(Object=@n3), bRemovePath=false))
   n5 = Get DefaultAISCFolder(self=@n2)
   > n6 = CreateAsset(SourceAssetPath=@n4, DestinationAssetPath=Append(A=Append(A=@n1, B=@n5), B=Append(A=BP_AISC_GA_, B=@n1.NPCName)))
   > n7 = GetAISpawnerDataComponentDefaultObject(SourceObject=@n6)
   > n8 = Cast To FortAthenaAISpawnerDataComponent_GameplayAbilityBase!(execute, Object=@n7) -> (
    then
    CastFailed
    AsFort Athena AISpawner Data Component Gameplay Ability Base:Fort Athena AISpawner Data Component Gameplay Ability Base Object Reference
   ) {is=title K2Node_DynamicCast}
   > SetGameplayAbilityComponent!(
    execute
    self:Fort Athena AISpawner Data Object Reference=Get ActiveSpawnerData()
    NewGameplayAbilityComponent:Fort Athena AISpawner Data Component Gameplay Ability Base Object Reference=@n8
   ) -> (then)
  ]}
 Component_CreateGameplayComponent {nodes [
   n1 = Component_CreateGameplayComponent!() -> (then, DestDirectory:String, NPCName:String) {is=title K2Node_FunctionEntry}
   > n2 = GetTemplateConfig()
   n3 = GetGameplayComponent!(self:Fort Athena AISpawner Data Object Reference=Get ActiveSpawnerData()) -> (ReturnValue:Fort Athena AISpawner Data Component Gameplay Base Object Reference)
   > n4 = ConvertFilePathToAssetPath(InPath=GetBaseFilename(InPath=Get Object Path String(Object=@n3), bRemovePath=false))
   n5 = Get DefaultAISCFolder(self=@n2)
   > n6 = CreateAsset(SourceAssetPath=@n4, DestinationAssetPath=Append(A=Append(A=@n1, B=@n5), B=Append(A=BP_AISC_Gameplay_, B=@n1.NPCName)))
   > n7 = GetAISpawnerDataComponentDefaultObject(SourceObject=@n6)
   > n8 = Cast To FortAthenaAISpawnerDataComponent_GameplayBase!(execute, Object=@n7) -> (then, CastFailed, AsFort Athena AISpawner Data Component Gameplay Base:Fort Athena AISpawner Data Component Gameplay Base Object Reference) {is=title K2Node_DynamicCast}
   > SetGameplayComponent!(
    execute
    self:Fort Athena AISpawner Data Object Reference=Get ActiveSpawnerData()
    NewGameplayComponent:Fort Athena AISpawner Data Component Gameplay Base Object Reference=@n8
   ) -> (then)
  ]}
 Component_CreateInventoryComponent {nodes [
   n1 = Component_CreateInventoryComponent!() -> (then, DestDirectory:String, NPCName:String) {is=title K2Node_FunctionEntry}
   > n2 = GetTemplateConfig()
   n3 = GetInventoryComponent!(self:Fort Athena AISpawner Data Object Reference=Get ActiveSpawnerData()) -> (ReturnValue:Fort Athena AISpawner Data Component Inventory Base Object Reference)
   > n4 = ConvertFilePathToAssetPath(InPath=GetBaseFilename(InPath=Get Object Path String(Object=@n3), bRemovePath=false))
   n5 = Get DefaultAISCFolder(self=@n2)
   > n6 = CreateAsset(SourceAssetPath=@n4, DestinationAssetPath=Append(A=Append(A=@n1, B=@n5), B=Append(A=BP_AISC_Inv_, B=@n1.NPCName)))
   > n7 = GetAISpawnerDataComponentDefaultObject(SourceObject=@n6)
   > n8 = Cast To FortAthenaAISpawnerDataComponent_InventoryBase!(execute, Object=@n7) -> (then, CastFailed, AsFort Athena AISpawner Data Component Inventory Base:Fort Athena AISpawner Data Component Inventory Base Object Reference) {is=title K2Node_DynamicCast}
   > SetInventoryComponent!(
    execute
    self:Fort Athena AISpawner Data Object Reference=Get ActiveSpawnerData()
    NewInventoryComponent:Fort Athena AISpawner Data Component Inventory Base Object Reference=@n8
   ) -> (then)
  ]}
 Component_CreateSpawnParamsComponent {nodes [
   n1 = Component_CreateSpawnParamsComponent!() -> (then, DestDirectory:String, NPCName:String) {is=title K2Node_FunctionEntry}
   > n2 = GetTemplateConfig()
   n3 = GetSpawnParamsComponent!(self:Fort Athena AISpawner Data Object Reference=Get ActiveSpawnerData()) -> (ReturnValue:Fort Athena AISpawner Data Component Spawn Params Base Object Reference)
   > n4 = ConvertFilePathToAssetPath(InPath=GetBaseFilename(InPath=Get Object Path String(Object=@n3), bRemovePath=false))
   n5 = Get DefaultAISCFolder(self=@n2)
   > n6 = CreateAsset(SourceAssetPath=@n4, DestinationAssetPath=Append(A=Append(A=@n1, B=@n5), B=Append(A=BP_AISC_Spawn_, B=@n1.NPCName)))
   > n7 = GetAISpawnerDataComponentDefaultObject(SourceObject=@n6)
   > n8 = Cast To FortAthenaAISpawnerDataComponent_SpawnParamsBase!(execute, Object=@n7) -> (
    then
    CastFailed
    AsFort Athena AISpawner Data Component Spawn Params Base:Fort Athena AISpawner Data Component Spawn Params Base Object Reference
   ) {is=title K2Node_DynamicCast}
   > SetSpawnParamsComponent!(
    execute
    self:Fort Athena AISpawner Data Object Reference=Get ActiveSpawnerData()
    NewSpawnParamsComponent:Fort Athena AISpawner Data Component Spawn Params Base Object Reference=@n8
   ) -> (then)
  ]}
 CreateAsset {nodes [
   n1 = CreateAsset!() -> (then, SourceAssetPath:String, DestinationAssetPath:String) {is=title K2Node_FunctionEntry}
   > n2 = Branch(Condition=Get CreateChildAssets() -> (CreateChildAssets:bool))
   > n3 = CreateChildAsset!(execute, self, SourcePath:String=@n1, TargetPath:String=@n1.DestinationAssetPath) -> (then, Success:bool, Object:Object Reference)
   > n4 = Branch(Condition=@n3)
   > n5 = SaveLoadedAsset(AssetToSave=@n3.Object, bOnlyIfIsDirty=true)
   Return Node!(execute=[@n5.then, @n4.else], ReturnValue:Object Reference=@n3.Object)
   n6 = DuplicateAsset(execute=@n2.else, SourceAssetPath=@n1, DestinationAssetPath=@n1.DestinationAssetPath)
   > Return Node(ReturnValue=@n6)
  ]}
 CreateChildAsset {nodes [
   n1 = CreateChildAsset!() -> (then, SourcePath:String, TargetPath:String) {is=title K2Node_FunctionEntry}
   > n2 = LoadAsset(AssetPath=@n1)
   n3 = Get Class(Object=@n2) -> (ReturnValue:Object Class Reference)
   > n4 = Branch(Condition="Equal (Class)"(A=@n3, B=/Script/Engine.Blueprint))
   > n5 = GetBlueprintAsset(Object=@n2)
   > n6 = GeneratedClass!(execute, BlueprintObj:Blueprint Object Reference=@n5) -> (then, ReturnValue:Object Class Reference)
   > n7 = CreateChildAsset!(execute, Path:String=@n1.TargetPath, Class:Object Class Reference=@n6) -> (then, OutBlueprint:Blueprint Object Reference, bSuccess:bool, ReturnValue:Object Reference) {is=title K2Node_CallFunction}
   > Return Node!(execute, Success:bool=@n7.bSuccess, Object:Object Reference=@n7)
   n8 = CreateChildAsset(execute=@n4.else, Path=@n1.TargetPath, Class=@n3)
   > Return Node(Success=@n8.bSuccess, Object=@n8.ReturnValue)
  ]}
 DisplayMessageBox {nodes [
   n1 = DisplayMessageBox!() -> (then, Title:String, Message:String) {is=title K2Node_FunctionEntry}
   > Show Message Dialog(
    Title="To Text (String)"(InString=@n1)
    Message="To Text (String)"(InString=@n1.Message)
    MessageType=Ok
    DefaultValue=No
    MessageCategory=Warning
   )
  ]}
 EventGraph {nodes [
   n1 = GetEditableObjectByIndex!(execute, self, Index:int=5) -> (then, Component:Object Reference)
   > Reset Widget Object!(
    execute
    self
    Target:Property View Base Object Reference=Get GameplayAbilityComponentView() -> (GameplayAbilityComponentView:Details View Object Reference)
    NewObject:Object Reference=@n1
   ) -> (then)
   n2 = DisplayMessageBox!(execute, self, Title:String=OnComponentListViewChanged, Message:String=SpawnerDataTagModified/DescriptorTag from ComponentListView) -> (then) {is=title K2Node_CallFunction}
   UpdateSpawnerDataTag(execute=[@n2.then, @n22.then])
   n3 = GetEditableObjectByIndex(Index=5)
   > Reset Widget Object(Target=Get GameplayAbilityComponentView(), NewObject=@n3)
   > Reset Widget Object(Target=Get ComponentListsView() -> (ComponentListsView:Details View Object Reference), NewObject=Get ActiveSpawnerData())
   "On Clicked (Btn_LoadNPC)"!() -> (OutputDelegate, then, widget:User Widget Object Reference, event:Pointer Event Structure) {is=title K2Node_ComponentBoundEvent}
   > LoadNPC()
   "On Clicked (Btn_NewNPC)"!() -> (OutputDelegate, then, widget:User Widget Object Reference, event:Pointer Event Structure) {is=title K2Node_ComponentBoundEvent}
   > n4 = AddWidgetToRoot!(
    execute
    self
    WidgetType:User Widget Class Reference=&/WBP_CreateNPCFromTemplate.WBP_CreateNPCFromTemplate_C
   ) -> (then, WidgetInstance:User Widget Object Reference) {is=title K2Node_CallFunction}
   > n5 = Cast To WBP_CreateNPCFromTemplate!(execute, Object=@n4) -> (then, CastFailed, AsWBP Create NPCFrom Template:WBP Create NPCFrom Template Object Reference) {is=title K2Node_DynamicCast}
   > Initialize!(execute, self:WBP Create NPCFrom Template Object Reference=@n5, Owner:EUW NPCEditor Tool Object Reference=Self-Reference()) -> (then)
   "On Clicked (Btn_GameplayAbility)"!() -> (OutputDelegate, then, widget:User Widget Object Reference, event:Pointer Event Structure) {is=title K2Node_ComponentBoundEvent}
   > SelectComponent!(execute, self, ComponentIndex:int=5) -> (then)
   n6 = "On Check State Changed (ShowAllCheckbox)"!() -> (OutputDelegate, then, bIsChecked:bool) {is=title K2Node_ComponentBoundEvent}
   > n7 = Branch(Condition=@n6.bIsChecked)
   > Set FullViewSourceIndex(FullViewSourceIndex:int=Get ActiveWidgetIndex() -> (ActiveWidgetIndex:int=0)) -> (Output_Get:int=0)
   n8 = Get ActivatableWidgetSwitcher() -> (ActivatableWidgetSwitcher:Common Activatable Widget Switcher Object Reference)
   > SelectActiveWidget!(execute, self, ActiveWIdgetIndex:int=int - int(A=GetChildrenCount!(self:Panel Widget Object Reference=@n8) -> (ReturnValue:int=0))) -> (then)
   > n9 = GetEditableObjectByIndex(Index=Get FullViewSourceIndex() -> (FullViewSourceIndex:int=0))
   > SetObject(self=Get FullDetailedView() -> (FullDetailedView:Details View Object Reference), NewObject=@n9)
   SelectActiveWidget(execute=@n7.else, ActiveWIdgetIndex=Get FullViewSourceIndex() -> (FullViewSourceIndex=0))
   n10 = "On Check State Changed (ComponentListShowAllCheckbox)"!() -> (OutputDelegate, then, bIsChecked:bool) {is=title K2Node_ComponentBoundEvent}
   > n11 = Branch(Condition=@n10.bIsChecked)
   > n12 = SetEditorProperty(
    Object=Get ComponentListsView()
    PropertyName=PropertiesToShow
    PropertyValue:Array of Names=Get EmptyStringArray() -> (EmptyStringArray:Array of Names)
   )
   > Branch(Condition=@n12)
   > Reset Widget Object(Target=Get ComponentListsView(), NewObject=Get ActiveSpawnerData())
   n13 = SetEditorProperty(
    execute=@n11.else
    Object=Get ComponentListsView()
    PropertyName=PropertiesToShow
    PropertyValue=Get ComponentListsViewMembers() -> (ComponentListsViewMembers:Array of Names)
   )
   > Branch(Condition=@n13)
   > Reset Widget Object(Target=Get ComponentListsView(), NewObject=Get ActiveSpawnerData())
   "On Clicked (Btn_OpenNPCLibraryConversation)"!() -> (OutputDelegate, then, widget:User Widget Object Reference, event:Pointer Event Structure) {is=title K2Node_ComponentBoundEvent}
   > OpenEditorForAssets!(
    execute
    self:Asset Editor Subsystem Object Reference=Get AssetEditorSubsystem()
    Assets:Array of Object References=Make Array("[0]":Object Reference=Get ActiveNPCLibraryConversation() -> (ActiveNPCLibraryConversation:Object Reference)) -> (Array:Array of Object References)
    OpenedMethod:EAssetTypeActivationOpenedMethod Enum=Edit
   ) -> (then, ReturnValue:bool)
   "On Selection Changed (LagerEventDataSelector)"!() -> (OutputDelegate, then, SelectedItem:String, SelectionType:ESelectInfo Enum) {is=title K2Node_ComponentBoundEvent}
   n14 = GetSelectedIndex!(self:"Searchable Combo Box (String) Object Reference"=Get LagerEventDataSelector() -> (LagerEventDataSelector:"Searchable Combo Box (String) Object Reference")) -> (ReturnValue:int=0)
   n15 = Get ActiveLagerEventDataAssets() -> (ActiveLagerEventDataAssets:Array of Strings)
   > Branch(Condition=Is Valid Index!(TargetArray=@n15, IndexToTest:int=@n14) -> (ReturnValue:bool))
   > SetLagerEventData!(execute, self, AssetPath:String="Get (a copy)"(Array=@n15, Dimension 1=@n14)) -> (then)
   n16 = Get ActiveWidgetIndex() -> (ActiveWidgetIndex=0)
   > n17 = GetEditableObjectByIndex(Index=@n16)
   > UpdateComponentDetailsVisibility!(execute, self, object:Object Reference=@n17, PanelIndex:int=@n16) -> (then)
   "On Clicked (Btn_LagerEventData)"!() -> (OutputDelegate, then, widget:User Widget Object Reference, event:Pointer Event Structure) {is=title K2Node_ComponentBoundEvent}
   > SelectComponent(ComponentIndex=8)
   "On Clicked (Btn_NPCCharacterData)"!() -> (OutputDelegate, then, widget:User Widget Object Reference, event:Pointer Event Structure) {is=title K2Node_ComponentBoundEvent}
   > SelectComponent(ComponentIndex=7)
   n18 = "On Selection Changed (ConfigSelector)"!() -> (OutputDelegate, then, SelectedItem:String, SelectionType:ESelectInfo Enum) {is=title K2Node_ComponentBoundEvent}
   > SetSelectedTemplate!(
    execute
    self
    Index:int=FindOptionIndex!(
     self:"Combo Box (String) Object Reference"=Get ConfigSelector() -> (ConfigSelector:"Combo Box (String) Object Reference")
     Option:String=@n18.SelectedItem
    ) -> (ReturnValue:int=0)
   ) -> (then)
   "On Clicked (Btn_SpawnParams)"!() -> (OutputDelegate, then, widget:User Widget Object Reference, event:Pointer Event Structure) {is=title K2Node_ComponentBoundEvent}
   > SelectComponent(ComponentIndex=6)
   "On Clicked (Btn_BrowseToAsset)"!() -> (OutputDelegate, then, widget:User Widget Object Reference, event:Pointer Event Structure) {is=title K2Node_ComponentBoundEvent}
   n19 = Get ActiveObject() -> (ActiveObject:Object Reference)
   > SyncBrowserToObjects(AssetPaths=Make Array("[0]":String=Get Object Path String(Object=@n19)) -> (Array:Array of Strings))
   "On Clicked (Btn_SaveNPC)"!() -> (OutputDelegate, then, widget:User Widget Object Reference, event:Pointer Event Structure) {is=title K2Node_ComponentBoundEvent}
   > SaveNPC()
   "On Clicked (Btn_Inventory)"!() -> (OutputDelegate, then, widget:User Widget Object Reference, event:Pointer Event Structure) {is=title K2Node_ComponentBoundEvent}
   > SelectComponent(ComponentIndex=0)
   "On Clicked (Btn_Affiliation)"!() -> (OutputDelegate, then, widget:User Widget Object Reference, event:Pointer Event Structure) {is=title K2Node_ComponentBoundEvent}
   > SelectComponent(ComponentIndex=1)
   "On Clicked (Btn_Conversation)"!() -> (OutputDelegate, then, widget:User Widget Object Reference, event:Pointer Event Structure) {is=title K2Node_ComponentBoundEvent}
   > SelectComponent(ComponentIndex=3)
   n20 = "On Check State Changed (AllowEditCheckbox)"!() -> (OutputDelegate, then, bIsChecked:bool) {is=title K2Node_ComponentBoundEvent}
   > SetComponentDetailsVisibility!(execute, self, Condition:bool=@n20.bIsChecked) -> (then)
   "On Clicked (Btn_Cosmetic)"!() -> (OutputDelegate, then, widget:User Widget Object Reference, event:Pointer Event Structure) {is=title K2Node_ComponentBoundEvent}
   > SelectComponent(ComponentIndex=4)
   "On Clicked (Btn_Gameplay)"!() -> (OutputDelegate, then, widget:User Widget Object Reference, event:Pointer Event Structure) {is=title K2Node_ComponentBoundEvent}
   > SelectComponent(ComponentIndex=2)
   n21 = OnComponentListsViewChanged!() -> (OutputDelegate, then, PropertyName:Name) {is=title K2Node_CustomEvent}
   > n22 = Branch(Condition="Equal Exactly (String)"(A="To String (Name)"(InName=@n21.PropertyName), B=DescriptorTag))
   n23 = OnAISpawnerDataGameplayTagChanged!() -> (OutputDelegate, then, PropertyName:Name) {is=title K2Node_CustomEvent}
   > UpdateSpawnerDataTag()
   n24 = OnSpawnLocationGameplayTagChanged!() -> (OutputDelegate, then, PropertyName:Name) {is=title K2Node_CustomEvent}
   > UpdateSpawnLocationTag()
   n25 = OnConversationGameplayTagChanged!() -> (OutputDelegate, then, PropertyName:Name) {is=title K2Node_CustomEvent}
   > UpdateConversationGameplayTag()
   n26 = OnReleaseVersionChanged!() -> (OutputDelegate, then, PropertyName:Name) {is=title K2Node_CustomEvent}
   > UpdateReleaseVersion()
   n27 = Event PreConstruct()
   > Parent: PreConstruct(IsDesignTime=@n27.IsDesignTime)
   Event Construct()
   > Parent: Construct()
   > Bind Event to On Property Changed(self:Property View Base Object Reference=Get ComponentListsView(), Delegate=@n21) {is=title K2Node_AddDelegate}
   > Bind Event to On Property Changed(self=Get AISpawnerDataGameplayTag() -> (AISpawnerDataGameplayTag:Fort Quest Tool Property View Object Reference), Delegate=@n23)
   > Bind Event to On Property Changed(
    self=Get GameplayAbilityComponentView()
    Delegate=OnGameplayAbilityComponentTagChanged!() -> (OutputDelegate, then, PropertyName:Name) {is=title K2Node_CustomEvent}
   )
   > Bind Event to On Property Changed(self=Get SpawnLocationGameplayTag() -> (SpawnLocationGameplayTag:Fort Quest Tool Property View Object Reference), Delegate=@n24)
   > Bind Event to On Property Changed(self=Get ConversationGameplayTag() -> (ConversationGameplayTag:Fort Quest Tool Property View Object Reference), Delegate=@n25)
   > Bind Event to On Property Changed(self=Get ReleaseVersionView() -> (ReleaseVersionView:Fort Quest Tool Property View Object Reference), Delegate=@n26)
   > Bind Event to On Property Changed(self=Get DetailedReleaseVersionView() -> (DetailedReleaseVersionView:Details View Object Reference), Delegate=@n26)
   > n28 = For Each Loop(Array:Array of BP NPCEditor Config Class References=Get Configs() -> (Configs:Array of BP NPCEditor Config Class References)) -> (Array Element:BP NPCEditor Config Class Reference)
   SetSelectedTemplate(execute=@n28.Completed, Index=0)
   n29 = GetClassDefaultObject!(execute=@n28.LoopBody, Class:Object Class Reference=@n28) -> (then, ReturnValue:Object Reference)
   > n30 = Cast To BP_NPCEditor_Config!(execute, Object=@n29) -> (then, CastFailed, AsBP NPCEditor Config:BP NPCEditor Config Object Reference) {is=title K2Node_DynamicCast}
   > AddOption!(
    execute
    self:"Combo Box (String) Object Reference"=Get ConfigSelector()
    Option:String=Get ConfigName!(self:BP NPCEditor Config Object Reference=@n30) -> (ConfigName:String)
   ) -> (then)
   Branch(Condition=@n27.IsDesignTime)
   > n31 = FindAssetData!(execute, AssetPath:String=/JunoEnemies/AI/AISpawnerData/MF/Bandit/Bow/BP_AISpawnerData_Juno_Bandit_Bow) -> (then, ReturnValue:Asset Data Structure)
   n32 = GetAsset(InAssetData=@n31)
   > n33 = Is Valid(InputObject=@n32)
   n34 = Construct Object from Class(execute=@n33.Is Not Valid, Class=/Script/FortniteAI.FortAthenaAIBotSpawnerData) -> (ReturnValue:Fort Athena AIBot Spawner Data Object Reference)
   > SetActiveObject!(execute, self, Input:Object Reference=@n34) -> (then)
   SetActiveObject(execute=@n33.Is Valid, Input=@n32)
   Event Tick()
  ]}
 ExternalAsset_CreateLagerEventData {nodes [
   n1 = ExternalAsset_CreateLagerEventData!() -> (then, DestDirectory:String, NPCName:String) {is=title K2Node_FunctionEntry}
   > n2 = GetTemplateConfig()
   > DuplicateAsset(
    SourceAssetPath=Get DefaultLagerEventDataTemplate!(self:BP NPCEditor Config Object Reference=@n2) -> (DefaultLagerEventDataTemplate:String)
    DestinationAssetPath=Append(
     A=Get DefaultLagerEventDataTargetFilePath!(self:BP NPCEditor Config Object Reference=@n2) -> (DefaultLagerEventDataTargetFilePath:String)
     B=@n1.NPCName
    )
   )
  ]}
 ExternalAsset_CreateNPCLibraryConversation {nodes [
   n1 = ExternalAsset_CreateNPCLibraryConversation!() -> (then, DestDirectory:String, NPCName:String) {is=title K2Node_FunctionEntry}
   > n2 = GetTemplateConfig()
   n3 = Get DefaultNPCLibraryConversationAssetName!(self:BP NPCEditor Config Object Reference=@n2) -> (DefaultNPCLibraryConversationAssetName:String)
   > DuplicateAsset(
    SourceAssetPath=Get DefaultNPCLibraryConversationTemplate!(self:BP NPCEditor Config Object Reference=@n2) -> (DefaultNPCLibraryConversationTemplate:String)
    DestinationAssetPath=Append(A=@n1, B=Append(A=@n3, B=@n1.NPCName))
   )
  ]}
 FindLagerEventData {nodes [
   n1 = FindLagerEventData!() -> (then, NPCName:String) {is=title K2Node_FunctionEntry}
   > Set NPCNameSearchString(NPCNameSearchString:String=@n1) -> (Output_Get:String)
   > Clear(TargetArray=Get ActiveLagerEventDataAssets())
   n2 = Get LagerEventDataSelector()
   > AddOption(self:"Searchable Combo Box (String) Object Reference"=@n2, Option=Invalid)
   > SetSelectedOption!(execute, self:"Searchable Combo Box (String) Object Reference"=@n2, Option:String=Invalid) -> (then)
   > ClearOptions(self:"Searchable Combo Box (String) Object Reference"=@n2)
   > n3 = GetTemplateConfig()
   n4 = Get DefaultLagerEventDataTargetFilePath(self=@n3)
   n5 = Append(A=@n4, B=Get NPCNameSearchString() -> (NPCNameSearchString:String))
   > n6 = DoesAssetExist(AssetPath=@n5)
   > n7 = Branch(Condition=@n6)
   > Add(TargetArray=Get ActiveLagerEventDataAssets(), NewItem=@n5) {is=title K2Node_CallArrayFunction}
   > n8 = Set DefaultToTandemPath(DefaultToTandemPath:bool=true) -> (Output_Get:bool)
   n9 = GetDescriptorTag!(self:Fort Athena Spawner Data Base Object Reference=Get ActiveSpawnerData()) -> (ReturnValue:Gameplay Tag Container Structure)
   n10 = GetLivingWorldEventDataAssetsFromTag!(
    execute=[@n7.else, @n8.then]
    GameplayTag:Gameplay Tag Structure="Get (a copy)"(Array:Array of Gameplay Tag Structures=BreakGameplayTagContainer!(GameplayTagContainer:"Gameplay Tag Container Structure (by ref)"=@n9) -> (GameplayTags:Array of Gameplay Tag Structures)) -> (Output:Gameplay Tag Structure)
   ) -> (then, OutLivingWorldEventData:Array of Names, ReturnValue:bool)
   > n11 = For Each Loop(Array:Array of Names=@n10) -> (Array Element:Name)
   n12 = Get ActiveLagerEventDataAssets()
   SortStringArray!(execute=@n11.Completed, TargetArray:Array of Strings=@n12, bStableSort:bool=false, SortOrder:EArraySortOrder Enum=Ascending) -> (then)
   > n13 = For Each Loop(Array:Array of Strings=@n12) -> (Array Element:String)
   n14 = Branch(execute=@n13.Completed, Condition=Get DefaultToTandemPath() -> (DefaultToTandemPath:bool))
   n15 = Append(A=@n4, B=Get NPCNameSearchString())
   n16 = Get LagerEventDataSelector()
   > SetSelectedIndex!(
    execute
    self:"Searchable Combo Box (String) Object Reference"=@n16
    Index:int=FindOptionIndex(self:"Searchable Combo Box (String) Object Reference"=@n16, Option=@n15) -> (ReturnValue=0)
   ) -> (then)
   > Return Node!(execute, FoundAsset:bool=true, AssetPath:String=@n15)
   n17 = Get ActiveLagerEventDataAssets()
   n18 = For Each Loop(Exec=@n14.else, Array=@n17)
   n19 = For Each Loop(Exec=@n18.Completed, Array=@n17)
   n20 = Get ActiveLagerEventDataAssets()
   n21 = Branch(execute=@n19.Completed, Condition=Is Empty!(TargetArray=@n20) -> (ReturnValue:bool))
   > Return Node(FoundAsset=false)
   SetSelectedIndex(execute=@n21.else, self=Get LagerEventDataSelector(), Index=0)
   > Return Node(FoundAsset=true, AssetPath="Get (a copy)"(Array:Array of Strings=@n20) -> (Output:String))
   n22 = Get NPCNameSearchString()
   Branch(execute=@n19.LoopBody, Condition=Contains(SearchIn=@n19, Substring=@n22))
   > SetSelectedIndex(self=Get LagerEventDataSelector(), Index=@n19.Array Index)
   > Return Node(FoundAsset=true, AssetPath=@n19)
   Branch(execute=@n18.LoopBody, Condition=EndsWith(SourceString=@n18, InSuffix=@n22))
   > SetSelectedIndex(self=Get LagerEventDataSelector(), Index=@n18.Array Index)
   > Return Node(FoundAsset=true, AssetPath=@n18)
   AddOption(execute=@n13.LoopBody, self=Get LagerEventDataSelector(), Option=@n13)
   n23 = "To String (Name)"(InName=@n11)
   n24 = LoadAsset(execute=@n11.LoopBody, AssetPath=@n23)
   > Cast To FortAthenaLivingWorldEventData!(execute, Object=@n24) -> (then, CastFailed, AsFort Athena Living World Event Data:Fort Athena Living World Event Data Object Reference) {is=title K2Node_DynamicCast}
   > Add Unique(TargetArray=Get ActiveLagerEventDataAssets(), NewItem=@n23)
  ]}
 GetAISCByName {nodes [
   n1 = GetAISCByName!() -> (then, Name:String) {is=title K2Node_FunctionEntry}
   > n2 = Switch on String!(execute:Exec, Selection:String=@n1) -> (
    Default:Exec
    InventoryComponent:Exec
    AffiliationComponent:Exec
    GameplayComponent:Exec
    ConversationComponent:Exec
    CosmeticComponent:Exec
    GameplayAbilityComponent:Exec
   ) {pinNames [InventoryComponent, AffiliationComponent, GameplayComponent, ConversationComponent, CosmeticComponent, GameplayAbilityComponent]}
   n3 = GetEditableObjectByIndex(execute=@n2.GameplayAbilityComponent, Index=5)
   > Return Node!(execute, Component:Object Reference=@n3)
   n4 = GetEditableObjectByIndex(execute=@n2.CosmeticComponent, Index=4)
   > Return Node(Component=@n4)
   n5 = GetEditableObjectByIndex(execute=@n2.ConversationComponent, Index=3)
   > Return Node(Component=@n5)
   n6 = GetEditableObjectByIndex(execute=@n2.GameplayComponent, Index=2)
   > Return Node(Component=@n6)
   n7 = GetEditableObjectByIndex(execute=@n2.AffiliationComponent, Index=1)
   > Return Node(Component=@n7)
   n8 = GetEditableObjectByIndex(execute=@n2.InventoryComponent, Index=0)
   > Return Node(Component=@n8)
   Return Node(execute=@n2.Default)
  ]}
 GetEditableObjectByIndex {nodes [
   n1 = GetEditableObjectByIndex!() -> (then, Index:int) {is=title K2Node_FunctionEntry}
   > n2 = Switch on Int!(execute:Exec, Selection:int=@n1) -> (Default:Exec, 0:Exec, 1:Exec, 2:Exec, 3:Exec, 4:Exec, 5:Exec, 6:Exec, 7:Exec, 8:Exec)
   Return Node(execute=@n2.5, Component=GetGameplayAbilityComponent(self=Get ActiveSpawnerData()))
   Return Node(execute=@n2.4, Component=GetCosmeticComponent(self=Get ActiveSpawnerData()))
   Return Node(execute=@n2.3, Component=GetConversationComponent(self=Get ActiveSpawnerData()))
   Return Node(execute=@n2.2, Component=GetGameplayComponent(self=Get ActiveSpawnerData()))
   Return Node(execute=@n2.1, Component=GetAffiliationComponent(self=Get ActiveSpawnerData()))
   Return Node(execute=@n2.0, Component=GetInventoryComponent(self=Get ActiveSpawnerData()))
   Return Node(execute=[@n6.else, @n4.else, @n2.Default])
   n3 = Get ActiveNPCCharacterData() -> (ActiveNPCCharacterData:Fort Tandem Character Data Object Reference)
   n4 = Branch(execute=@n2.7, Condition=IsValid(Object=@n3))
   > Return Node(Component=@n3)
   n5 = Get ActiveLagerEventData() -> (ActiveLagerEventData:Fort Athena Living World Event Data Object Reference)
   n6 = Branch(execute=@n2.8, Condition=IsValid(Object=@n5))
   > Return Node(Component=@n5)
   Return Node(execute=@n2.6, Component=GetSpawnParamsComponent(self=Get ActiveSpawnerData()))
  ]}
 GetObjectAsset {nodes [
   n1 = GetObjectAsset!() -> (then, Object:Object Reference) {is=title K2Node_FunctionEntry}
   > n2 = ConvertFilePathToAssetPath(InPath=GetBaseFilename(InPath=Get Object Path String(Object=@n1), bRemovePath=false))
   > n3 = FindAssetData(AssetPath=@n2)
   > Return Node!(execute, Asset:Object Reference=GetAsset(InAssetData=@n3))
  ]}
 GetTemplateConfig {nodes [
   GetTemplateConfig() {is=title K2Node_FunctionEntry}
   > n1 = GetClassDefaultObject(Class="Get (a copy)"(Array:Array of BP NPCEditor Config Class References=Get Configs(), Dimension 1=Get SelectedTemplateIndex() -> (SelectedTemplateIndex:int=0)) -> (Output:BP NPCEditor Config Class Reference))
   > n2 = Cast To BP_NPCEditor_Config(Object=@n1)
   > Return Node!(execute, Config:BP NPCEditor Config Object Reference=@n2)
  ]}
 IsObjectOverriden {nodes [
   n1 = IsObjectOverriden!() -> (then, Object:Object Reference) {is=title K2Node_FunctionEntry}
   n2 = Get Object Path String(Object=@n1)
   > Return Node!(
    execute
    ReturnValue:bool=Contains(SearchIn=GetBaseFilename(InPath=@n2, bRemovePath=true), Substring=Get ActiveNPCName() -> (ActiveNPCName:String))
   )
  ]}
 LoadNPC {nodes [
   LoadNPC() {is=title K2Node_FunctionEntry}
   > n1 = GetTemplateConfig() {is=title K2Node_CallFunction}
   > n2 = Execute Python Script!(
    execute
    PythonScript="import unreal\r\nimport TechArt.Tools.FilePickerUtils as utils\r\nimport importlib\r\n\r\nimportlib.reload(utils)\r\n\r\ndir = unreal.Paths.project_content_dir()\r\n\r\nif multiple:\r\n    file_paths = utils.OpenDialog_File(initial_dir=initial_directory,title=title, file_types_map=file_types_map, multiple=True)\r\n    first_path = file_paths[0]\r\nelse:\r\n    first_path = utils.OpenDialog_File(initial_dir=initial_directory,title=title, file_types_map=file_types_map, multiple=False)\r\n    file_paths = [first_path]"
    file_types_map:Map of Strings to Strings=Get FileTypes() -> (FileTypes:Map of Strings to Strings)
    initial_directory:String=Append(
     A=GetProjectDirectory!() -> (ReturnValue:String)
     B=Get DefaultOpenNPCSubFolderPath!(self:BP NPCEditor Config Object Reference=@n1) -> (DefaultOpenNPCSubFolderPath:String)
    )
    title:String=Get PythonTitle() -> (PythonTitle:String)
    multiple:bool=Get Multiple() -> (Multiple:bool)
   ) -> (then, ReturnValue, file_paths:Array of Strings, first_path:String) {inputs [file_types_map, initial_directory, title, multiple], outputs [file_paths,first_path]}
   > Set FilePaths(FilePaths:Array of Strings=@n2.file_paths) -> (Output_Get:Array of Strings)
   > n3 = ConvertFilePathToAssetPath(InPath=@n2.first_path)
   > n4 = FindAssetData(AssetPath=@n3)
   n5 = GetAsset(InAssetData=@n4)
   > Is Valid(InputObject=@n5)
   > SetActiveObject(Input=@n5)
  ]}
 NewNPC {nodes [
   n1 = NewNPC!() -> (then, FileName:String, FileDirectory:String, TemplateName:String, Options:Array of Strings, CreateChildAssets:bool=~) {is=title K2Node_FunctionEntry}
   > Set CreateChildAssets(CreateChildAssets:bool=@n1.CreateChildAssets) -> (Output_Get:bool)
   > Set DestinationAssetName(DestinationAssetName:String=@n1) -> (Output_Get:String)
   > Set NPCTemplateName(NPCTemplateName:String=@n1.TemplateName) -> (Output_Get:String)
   > Set NPCOptions(NPCOptions:Array of Strings=@n1.Options) -> (Output_Get:Array of Strings)
   > n2 = Branch(Condition=EndsWith(SourceString=@n1.FileDirectory, InSuffix=/))
   > n3 = Set DestinationAssetDirectory(DestinationAssetDirectory:String=@n1.FileDirectory) -> (Output_Get:String)
   n4 = DoesAssetExist(
    execute=[@n3.then, @n10.then]
    AssetPath=Append(A=Get DestinationAssetDirectory() -> (DestinationAssetDirectory:String), B=Get DestinationAssetName() -> (DestinationAssetName:String))
   )
   > n5 = Branch(Condition=@n4)
   > DisplayMessageBox(Title=Invalid filename, Message="The destination filename already exist, please correct the destination.")
   n6 = Switch on String!(execute=@n5.else, Selection=Get NPCTemplateName() -> (NPCTemplateName:String)) -> (Default, DefaultTemplate:Exec) {pinNames [DefaultTemplate]}
   > n7 = FindAssetData(AssetPath=Get NPCTemplateName())
   > n8 = Branch(Condition=Is Valid Asset Data!(InAssetData:"Asset Data Structure (by ref)"=@n7) -> (ReturnValue:bool))
   > NewNPCFromTemplate!(
    execute
    self
    DestinationDirectory:String=Get DestinationAssetDirectory()
    DestinationName:String=Get DestinationAssetName()
    TemplateName:String=Get NPCTemplateName()
    Options:Array of Strings=Get NPCOptions() -> (NPCOptions:Array of Strings)
   ) -> (then)
   DisplayMessageBox(execute=@n8.else, Title=Unsupported template, Message=The selected template could not be found.)
   n9 = GetTemplateConfig(execute=@n6.DefaultTemplate)
   > NewNPCFromTemplate(
    DestinationDirectory=Get DestinationAssetDirectory()
    DestinationName=Get DestinationAssetName()
    TemplateName=Get DefaultAISpawnerDataTemplate!(self:BP NPCEditor Config Object Reference=@n9) -> (DefaultAISpawnerDataTemplate:String)
    Options=Get NPCOptions()
   )
   n10 = Set DestinationAssetDirectory(execute=@n2.else, DestinationAssetDirectory=Append(A=@n1.FileDirectory, B=/))
  ]}
 NewNPCFromTemplate {nodes [
   n1 = AddWidgetToRoot(WidgetType=/Game/Blueprints/Blutilities/ForDesign/NPCs/SupportWidget/WBP_CreateNPC_Status.WBP_CreateNPC_Status_C)
   > n2 = Cast To WBP_CreateNPC_Status!(execute, Object=@n1) -> (then, CastFailed, AsWBP Create NPC Status:WBP Create NPC Status Object Reference) {is=title K2Node_DynamicCast}
   > n3 = Set StatusWidget(StatusWidget:WBP Create NPC Status Object Reference=@n2) -> (Output_Get:WBP Create NPC Status Object Reference)
   > SetStatusText!(execute, self:WBP Create NPC Status Object Reference=@n3, StatusText:String=Creating AISpawnerData) -> (then)
   n4 = NewNPCFromTemplate!() -> (then, DestinationDirectory:String, DestinationName:String, TemplateName:String, Options:Array of Strings) {is=title K2Node_FunctionEntry}
   > n5 = GetTemplateConfig()
   > Set DestDirectory(DestDirectory:String=@n4) -> (Output_Get:String)
   > Set DestName(DestName:String=@n4.DestinationName) -> (Output_Get:String)
   > Set AISpawnerDataTemplate(AISpawnerDataTemplate:String=@n4.TemplateName) -> (Output_Get:String)
   n6 = Get DefaultNPCPrefix!(self:BP NPCEditor Config Object Reference=@n5) -> (DefaultNPCPrefix:String)
   > Set NPCName(NPCName:String=RightChop!(SourceString:String=Get DestName() -> (DestName:String), Count:int=Len(S=@n6)) -> (ReturnValue:String)) -> (Output_Get:String)
   > Set ConfigList(ConfigList:Array of Strings=@n4.Options) -> (Output_Get:Array of Strings)
   > n7 = Branch(Condition=StartsWith(SourceString=Get DestName(), InPrefix=@n6))
   > n8 = CreateAsset(
    SourceAssetPath=Get AISpawnerDataTemplate() -> (AISpawnerDataTemplate:String)
    DestinationAssetPath=Append(A=Get DestDirectory() -> (DestDirectory:String), B=Get DestName())
   ) {is=title K2Node_CallFunction}
   > n9 = GetSpawnerDataDefaultObject!(execute, SourceObject:Object Reference=@n8) -> (then, ReturnValue:Fort Athena Spawner Data Base Object Reference)
   > n10 = Cast To FortAthenaAIBotSpawnerData!(execute, Object=@n9) -> (then, CastFailed, AsFort Athena AIBot Spawner Data:Fort Athena AIBot Spawner Data Object Reference) {is=title K2Node_DynamicCast}
   > Set ActiveSpawnerData(ActiveSpawnerData:Fort Athena AIBot Spawner Data Object Reference=@n10) -> (Output_Get:Fort Athena AIBot Spawner Data Object Reference)
   > n11 = Sequence!(execute) -> (then_0, then_1, then_2:Exec, then_3:Exec, then_4:Exec, then_5:Exec, then_6:Exec, then_7:Exec, then_8:Exec, then_9:Exec)
   > Branch(Condition=Contains Item(TargetArray:Array of Strings=Get ConfigList() -> (ConfigList:Array of Strings), ItemToFind:"String (by ref)"=InventoryComponent))
   > Component_CreateInventoryComponent!(execute, self, DestDirectory:String=Get DestDirectory(), NPCName:String=Get NPCName() -> (NPCName:String)) -> (then) {is=title K2Node_CallFunction}
   Branch(execute=@n11.then_5, Condition=Contains Item(TargetArray=Get ConfigList(), ItemToFind=GameplayAbilityComponent))
   > Component_CreateGameplayAbilityComponent!(execute, self, DestDirectory:String=Get DestDirectory(), NPCName:String=Get NPCName()) -> (then) {is=title K2Node_CallFunction}
   Branch(execute=@n11.then_2, Condition=Contains Item(TargetArray=Get ConfigList(), ItemToFind=GameplayComponent))
   > Component_CreateGameplayComponent!(execute, self, DestDirectory:String=Get DestDirectory(), NPCName:String=Get NPCName()) -> (then) {is=title K2Node_CallFunction}
   SetActiveObject(execute=@n11.then_9, Input=@n8)
   > SaveNPC()
   Branch(execute=@n11.then_6, Condition=Contains Item(TargetArray=Get ConfigList(), ItemToFind=SpawnParamsComponent))
   > Component_CreateSpawnParamsComponent!(execute, self, DestDirectory:String=Get DestDirectory(), NPCName:String=Get NPCName()) -> (then) {is=title K2Node_CallFunction}
   Branch(execute=@n11.then_7, Condition=Contains Item(TargetArray=Get ConfigList(), ItemToFind=NPCLibrary_Conversation))
   > ExternalAsset_CreateNPCLibraryConversation!(execute, self, DestDirectory:String=Get DestDirectory(), NPCName:String=Get NPCName()) -> (then) {is=title K2Node_CallFunction}
   Branch(execute=@n11.then_3, Condition=Contains Item(TargetArray=Get ConfigList(), ItemToFind=ConversationComponent))
   > Component_CreateConversationComponent!(
    execute
    self
    DestDirectory:String=Get DestDirectory()
    NPCName:String=Get NPCName()
    OverrideNPCCharacterData:bool=Contains Item(TargetArray=Get ConfigList(), ItemToFind=NPCCharacterData)
   ) -> (then) {is=title K2Node_CallFunction}
   Branch(execute=@n11.then_1, Condition=Contains Item(TargetArray=Get ConfigList(), ItemToFind=AffiliationComponent))
   > Component_CreateAffiliationComponent!(execute, self, DestDirectory:String=Get DestDirectory(), NPCName:String=Get NPCName()) -> (then) {is=title K2Node_CallFunction}
   Branch(execute=@n11.then_4, Condition=Contains Item(TargetArray=Get ConfigList(), ItemToFind=CosmeticComponent))
   > Component_CreateCosmeticComponent!(execute, self, DestDirectory:String=Get DestDirectory(), NPCName:String=Get NPCName()) -> (then) {is=title K2Node_CallFunction}
   Branch(execute=@n11.then_8, Condition=Contains Item(TargetArray=Get ConfigList(), ItemToFind=LagerEventData))
   > ExternalAsset_CreateLagerEventData!(execute, self, DestDirectory:String=Get DestDirectory(), NPCName:String=Get NPCName()) -> (then) {is=title K2Node_CallFunction}
   DisplayMessageBox(execute=@n7.else, Title=Invalid NPC name, Message=The provided NPC name does not respect the nomenclature.)
   RemoveFromParent(self=Get StatusWidget() -> (StatusWidget:WBP Create NPC Status Object Reference))
  ]}
 Reset Widget Object {nodes [
   n1 = Reset Widget Object!() -> (then, Target:Property View Base Object Reference, NewObject:Object Reference) {is=title K2Node_FunctionEntry}
   > SetObject(self=@n1)
   > SetObject(self=@n1, NewObject=@n1.NewObject)
  ]}
 SaveNPC {nodes [
   SaveNPC() {is=title K2Node_FunctionEntry}
   > SaveLoadedAssets(AssetsToSave=Get ListActiveNPCObjects() -> (ListActiveNPCObjects:Array of Object References), bOnlyIfIsDirty=true)
  ]}
 SelectActiveWidget {nodes [
   n1 = SelectActiveWidget!() -> (then, ActiveWIdgetIndex:int) {is=title K2Node_FunctionEntry}
   > n2 = Set ActiveWidgetIndex(ActiveWidgetIndex:int=@n1) -> (Output_Get:int=0)
   > SetActiveWidgetIndex(self=Get ActivatableWidgetSwitcher(), Index=@n2)
  ]}
 SelectComponent {nodes [
   n1 = SelectComponent!() -> (then, ComponentIndex:int) {is=title K2Node_FunctionEntry}
   > SelectActiveWidget(ActiveWIdgetIndex=@n1) {is=title K2Node_CallFunction}
   > n2 = GetEditableObjectByIndex(Index=@n1) {is=title K2Node_CallFunction}
   > UpdateComponentDetailsVisibility(object=@n2, PanelIndex=@n1)
   > SetCheckedState!(
    execute
    self:Check Box Object Reference=Get ShowAllCheckbox() -> (ShowAllCheckbox:Check Box Object Reference)
    InCheckedState:ECheckBoxState Enum=Unchecked
   ) -> (then)
  ]}
 SetActiveObject {nodes [
   n1 = SetActiveObject!() -> (then, Input:Object Reference) {is=title K2Node_FunctionEntry}
   > n2 = Set ActiveObject(ActiveObject:Object Reference=@n1) -> (Output_Get:Object Reference)
   > n3 = Set AssetDisplayName(AssetDisplayName:String=GetDisplayName(Object=@n2)) -> (Output_Get:String)
   > "SetText (Text)"(self=Get AssetNameText() -> (AssetNameText:Common Text Object Reference), InText="To Text (String)"(InString=@n3))
   > Set IsValid(IsValid:bool=false) -> (Output_Get:bool)
   > Set ActiveNPCName(ActiveNPCName:String) -> (Output_Get:String)
   > n4 = GetTemplateConfig()
   > n5 = For Each Loop with Break!(
    Exec:Exec
    Array:Array of Strings=Get AISpawnerDataStartStrings!(self:BP NPCEditor Config Object Reference=@n4) -> (AISpawnerDataStartStrings:Array of Strings)
    Break:Exec=@n7.then
   ) -> (LoopBody:Exec, Array Element:String, Array Index:int, Completed:Exec)
   n6 = Get AssetDisplayName() -> (AssetDisplayName:String)
   > Branch(Condition=StartsWith(SourceString=@n6, InPrefix=@n5))
   > Set ActiveNPCName(ActiveNPCName=RightChop(SourceString=@n6, Count=Len(S=@n5)))
   > n7 = Set IsValid(IsValid=true)
   n8 = Branch(execute=@n5.Completed, Condition=Get IsValid() -> (IsValid:bool))
   > n9 = Branch(Condition="Equal (Class)"(A=Get Class(Object=@n2), B=&Blueprint))
   > n10 = GetSpawnerDataDefaultObject(SourceObject=@n2)
   > n11 = Cast To FortAthenaAIBotSpawnerData(Object=@n10)
   > UpdateActiveSpawnerData!(execute, self, ActiveSpawnerData:Fort Athena AIBot Spawner Data Object Reference=@n11) -> (then)
   Set ActiveObject(execute=[@n11.CastFailed, @n8.else, @n12.CastFailed])
   > UpdateActiveSpawnerData()
   n12 = Cast To FortAthenaAIBotSpawnerData(execute=@n9.else, Object=@n2)
   > UpdateActiveSpawnerData(ActiveSpawnerData=@n12)
  ]}
 SetComponentDetailsVisibility {nodes [
   n1 = SetComponentDetailsVisibility!() -> (then, Condition:bool=~) {is=title K2Node_FunctionEntry}
   > n2 = Branch(Condition=@n1)
   n3 = Get ActivatableWidgetSwitcher()
   > SetVisibility(self=@n3, InVisibility=SelfHitTestInvisible)
   > SetRenderOpacity(self=@n3, InOpacity=1)
   n4 = Get ActivatableWidgetSwitcher()
   SetVisibility(execute=@n2.else, self=@n4, InVisibility=HitTestInvisible)
   > SetRenderOpacity(self=@n4, InOpacity=0.7)
  ]}
 SetLagerEventData {nodes [
   n1 = SetLagerEventData!() -> (then, AssetPath:String) {is=title K2Node_FunctionEntry}
   > n2 = Branch(Condition=IsEmpty(InString=@n1))
   n3 = DoesAssetExist(execute=@n2.else, AssetPath=@n1)
   > n4 = Branch(Condition=@n3)
   n5 = Get SpawnLocationGameplayTag()
   > SetVisibility(self=@n5, InVisibility=Visible)
   n6 = Get LagerEventDataNotFound() -> (LagerEventDataNotFound:Common Text Object Reference)
   > SetVisibility(self=@n6, InVisibility=Collapsed)
   > n7 = FindAssetData(AssetPath=@n1)
   > n8 = Cast To FortAthenaLivingWorldEventData(Object=GetAsset(InAssetData=@n7))
   > SetObject(self=[Get SpawnLocationGameplayTag(), Get LagerEventDataView() -> (LagerEventDataView:Details View Object Reference)], NewObject=@n8)
   > Add Unique(TargetArray:Array of Object References=Get ListActiveNPCObjects(), NewItem:"Object Reference (by ref)"=@n8)
   > Set ActiveLagerEventData(ActiveLagerEventData:Fort Athena Living World Event Data Object Reference=@n8) -> (Output_Get:Fort Athena Living World Event Data Object Reference)
   SetVisibility(execute=@n4.else, self=@n5, InVisibility=Collapsed)
   > SetVisibility(self=@n6, InVisibility=Visible)
   > SetObject(self=[Get SpawnLocationGameplayTag(), Get LagerEventDataView()])
  ]}
 SetSelectedTemplate {nodes [
   n1 = SetSelectedTemplate!() -> (then, Index:int) {is=title K2Node_FunctionEntry}
   > n2 = Branch(Condition="Equal (Integer)"(A=@n1, B=Get SelectedTemplateIndex() -> (SelectedTemplateIndex=0)))
   n3 = Set SelectedTemplateIndex(execute=@n2.else, SelectedTemplateIndex:int=@n1) -> (Output_Get:int=0)
   n4 = Get ConfigSelector()
   > SetSelectedIndex(self:"Combo Box (String) Object Reference"=@n4, Index=@n3)
   > RefreshOptions(self:"Combo Box (String) Object Reference"=@n4)
   > Return Node!(execute)
  ]}
 UpdateActiveSpawnerData {nodes [
   n1 = UpdateActiveSpawnerData!() -> (then, ActiveSpawnerData:Fort Athena AIBot Spawner Data Object Reference) {is=title K2Node_FunctionEntry}
   > Clear(TargetArray:Array of Object References=Get ListActiveNPCObjects())
   > Set ActiveLagerEventData()
   > Set ActiveNPCCharacterData(ActiveNPCCharacterData:Fort Tandem Character Data Object Reference) -> (Output_Get:Fort Tandem Character Data Object Reference)
   > n2 = Set ActiveSpawnerData(ActiveSpawnerData=@n1)
   > n3 = Sequence!(execute) -> (then_0, then_1, then_2, then_3, then_4, then_5, then_6, then_7, then_8, then_9, then_10:Exec)
   > SetObject(self=[Get ComponentListsView(), Get DetailedReleaseVersionView(), Get ReleaseVersionView(), Get AISpawnerDataGameplayTag()], NewObject=@n2)
   > n4 = GetObjectAsset!(execute, self, Object:Object Reference=@n2) -> (then, Asset:Object Reference) {is=title K2Node_CallFunction}
   > Add(TargetArray:Array of Object References=Get ListActiveNPCObjects(), NewItem:"Object Reference (by ref)"=@n4)
   n5 = GetAffiliationComponent(self=@n2)
   SetObject(execute=@n3.then_3, self=Get AffiliationComponentView() -> (AffiliationComponentView:Details View Object Reference), NewObject=@n5)
   > n6 = IsObjectOverriden!(execute, self, Object:Object Reference=@n5) -> (then, ReturnValue:bool) {is=title K2Node_CallFunction}
   > Branch(Condition=@n6)
   > n7 = GetObjectAsset(Object=@n5)
   > Add(TargetArray=Get ListActiveNPCObjects(), NewItem=@n7)
   n8 = GetGameplayComponent(self=@n2)
   SetObject(execute=@n3.then_4, self=Get GameplayComponentView() -> (GameplayComponentView:Details View Object Reference), NewObject=@n8)
   > n9 = IsObjectOverriden(Object=@n8)
   > Branch(Condition=@n9)
   > n10 = GetObjectAsset(Object=@n8)
   > Add(TargetArray=Get ListActiveNPCObjects(), NewItem=@n10)
   n11 = GetCosmeticComponent(self=@n2)
   SetObject(execute=@n3.then_5, self=Get CosmeticComponentView() -> (CosmeticComponentView:Details View Object Reference), NewObject=@n11)
   > n12 = IsObjectOverriden(Object=@n11)
   > Branch(Condition=@n12)
   > n13 = GetObjectAsset(Object=@n11)
   > Add(TargetArray=Get ListActiveNPCObjects(), NewItem=@n13)
   n14 = GetSpawnParamsComponent(self=@n2)
   SetObject(execute=@n3.then_7, self=Get SpawnParamsComponentView() -> (SpawnParamsComponentView:Details View Object Reference), NewObject=@n14)
   > n15 = IsObjectOverriden(Object=@n14)
   > Branch(Condition=@n15)
   > n16 = GetObjectAsset(Object=@n14)
   > Add(TargetArray=Get ListActiveNPCObjects(), NewItem=@n16)
   n17 = GetObjectAsset(execute=@n3.then_10, Object=@n2)
   n18 = Break Directory Path!(DirectoryPath:"Directory Path Structure (by ref)"=Make Directory Path!(Path:String=Get Object Path String(Object=@n17)) -> (DirectoryPath:Directory Path Structure) {is=title K2Node_MakeStruct}) -> (Path:String) {is=title K2Node_BreakStruct}
   n19 = Append(
    A=Append(A=Split!(InPath:String=@n18) -> (PathPart:String, FilenamePart:String, ExtensionPart:String), B=/NPCLibrary_Conversation_)
    B=Get ActiveNPCName()
   )
   > n20 = DoesAssetExist(AssetPath=@n19)
   > n21 = Branch(Condition=@n20)
   > n22 = FindAssetData(AssetPath=@n19)
   n23 = GetAsset(InAssetData=@n22)
   > Set ActiveNPCLibraryConversation(ActiveNPCLibraryConversation:Object Reference=@n23) -> (Output_Get:Object Reference)
   n24 = Get Btn_OpenNPCLibraryConversation() -> (Btn_OpenNPCLibraryConversation:TA NPCEditor Button Object Reference)
   > SetVisibility(self=@n24, InVisibility=Visible)
   > Add(TargetArray=Get ListActiveNPCObjects(), NewItem=@n23)
   SetVisibility(execute=@n21.else, self=@n24, InVisibility=Collapsed)
   n25 = FindLagerEventData!(execute=@n3.then_9, self, NPCName:String=Get ActiveNPCName()) -> (then, FoundAsset:bool, AssetPath:String) {is=title K2Node_CallFunction}
   > SetLagerEventData(AssetPath=@n25.AssetPath) {is=title K2Node_CallFunction}
   n26 = Get DefaultPanelIndex() -> (DefaultPanelIndex:int=0)
   SelectActiveWidget(execute=@n3.then_8, ActiveWIdgetIndex=@n26)
   > n27 = GetEditableObjectByIndex(Index=@n26)
   > UpdateComponentDetailsVisibility(object=@n27, PanelIndex=@n26)
   n28 = GetGameplayAbilityComponent(self=@n2)
   SetObject(execute=@n3.then_6, self=Get GameplayAbilityComponentView(), NewObject=@n28)
   > n29 = IsObjectOverriden(Object=@n28)
   > Branch(Condition=@n29)
   > n30 = GetObjectAsset(Object=@n28)
   > Add(TargetArray=Get ListActiveNPCObjects(), NewItem=@n30)
   n31 = Sequence!(execute=@n3.then_2) -> (then_0, then_1, then_2)
   n32 = GetConversationComponent(self=@n2)
   > SetObject(
    self=[Get ConversationGameplayTag(), Get ConversationComponentView() -> (ConversationComponentView:Details View Object Reference)]
    NewObject=@n32
   )
   > n33 = IsObjectOverriden(Object=@n32)
   > UpdateConversationTagVisibility!(execute, self, ConversationValid:bool=@n33) -> (then)
   > Branch(Condition=@n33)
   > n34 = GetObjectAsset(Object=@n32)
   > Add(TargetArray=Get ListActiveNPCObjects(), NewItem=@n34)
   n35 = GetEditorProperty!(execute:Exec=@n31.then_1, Object:Object Reference=@n32, PropertyName:Name=CharacterData) -> (then:Exec, PropertyValue:Fort Tandem Character Data Object Reference, ReturnValue:bool)
   > n36 = Branch(Condition=@n35.ReturnValue)
   > n37 = Set ActiveNPCCharacterData(ActiveNPCCharacterData=@n35)
   > SetObject(self=Get NPCCharacterDataView() -> (NPCCharacterDataView:Details View Object Reference), NewObject=@n37)
   > n38 = IsObjectOverriden(Object=@n37)
   > Branch(Condition=@n38)
   > n39 = GetObjectAsset(Object=@n37)
   > Add(TargetArray=Get ListActiveNPCObjects(), NewItem=@n39)
   n40 = Branch(execute=@n36.else, Condition=IsValid(Object=@n32))
   > DisplayMessageBox(Title=UpdateActiveSpawnerData, Message=Couldn't extract the CharacterData from the Conversation component.)
   SetObject(execute=@n40.else, self=Get NPCCharacterDataView())
   n41 = GetInventoryComponent(self=@n2)
   SetObject(execute=@n3.then_1, self=Get InventoryComponentView() -> (InventoryComponentView:Details View Object Reference), NewObject=@n41)
   > n42 = IsObjectOverriden(Object=@n41)
   > Branch(Condition=@n42)
   > n43 = GetObjectAsset(Object=@n41)
   > Add(TargetArray=Get ListActiveNPCObjects(), NewItem=@n43)
  ]}
 UpdateComponentDetailsVisibility {nodes [
   n1 = UpdateComponentDetailsVisibility!() -> (then, object:Object Reference, PanelIndex:int) {is=title K2Node_FunctionEntry}
   n2 = Get Object Path String(Object=@n1)
   > "SetText (Text)"(
    self=Get ActiveComponentName() -> (ActiveComponentName:Common Text Object Reference)
    InText="To Text (String)"(InString=GetBaseFilename(InPath=@n2, bRemovePath=true))
   )
   > n3 = IsObjectOverriden(Object=@n1)
   > SetComponentDetailsVisibility(Condition=@n3) {is=title K2Node_CallFunction}
   > n4 = Branch(Condition=@n3)
   n5 = Get AllowEditCheckbox() -> (AllowEditCheckbox:Check Box Object Reference)
   > SetCheckedState(self=@n5, InCheckedState=Checked)
   n6 = Get GenericAssetWarning() -> (GenericAssetWarning:Common Text Object Reference)
   > n7 = SetVisibility(self=@n6, InVisibility=Hidden)
   n8 = Branch(
    execute=[@n11.then, @n7.then]
    Condition="Equal (Integer)"(A=@n1.PanelIndex, B=Get LagerEventDataViewIndex() -> (LagerEventDataViewIndex:int=0))
   )
   n9 = Get LagerEventDataSelector()
   > SetVisibility(self=@n9, InVisibility=Visible)
   n10 = Get ActiveComponentName()
   > SetVisibility(self=@n10, InVisibility=Collapsed)
   SetVisibility(execute=@n8.else, self=@n9, InVisibility=Collapsed)
   > SetVisibility(self=@n10, InVisibility=Visible)
   SetCheckedState(execute=@n4.else, self=@n5, InCheckedState=Unchecked)
   > n11 = SetVisibility(self=@n6, InVisibility=Visible)
  ]}
 UpdateConversationGameplayTag {nodes [
   UpdateConversationGameplayTag() {is=title K2Node_FunctionEntry}
   > Reset Widget Object(Target=Get ConversationComponentView(), NewObject=GetConversationComponent(self=Get ActiveSpawnerData())) {is=title K2Node_CallFunction}
  ]}
 UpdateConversationTagVisibility {nodes [
   n1 = UpdateConversationTagVisibility!() -> (then, ConversationValid:bool=~) {is=title K2Node_FunctionEntry}
   > n2 = Branch(Condition=@n1)
   n3 = Get ConversationGameplayTag()
   > SetVisibility(self=@n3, InVisibility=Visible)
   n4 = Get ConversationComponentNotValid() -> (ConversationComponentNotValid:Common Text Object Reference)
   > SetVisibility(self=@n4, InVisibility=Collapsed)
   SetVisibility(execute=@n2.else, self=@n3, InVisibility=Collapsed)
   > SetVisibility(self=@n4, InVisibility=Visible)
  ]}
 UpdateReleaseVersion {nodes [
   UpdateReleaseVersion() {is=title K2Node_FunctionEntry}
   n1 = Get ActiveSpawnerData()
   > n2 = Cast To FortAthenaAISpawnerDataComponent_AIBotConversation(Object=GetConversationComponent(self=@n1))
   > PropagateReleaseVersionToCharacterData!(
    execute
    self:Fort Athena AISpawner Data Component AIBot Conversation Object Reference=@n2
    SourceSpawnerData:Fort Athena Spawner Data Base Object Reference=@n1
   ) -> (then)
   > Reset Widget Object(Target=Get NPCCharacterDataView(), NewObject=Get ActiveNPCCharacterData())
   DisplayMessageBox(execute=@n2.CastFailed, Title=UpdateReleaseVersion, Message=Couldn't propagate FortReleaseVersion to NPCCharacterData)
  ]}
 UpdateSpawnLocationTag {nodes [
   UpdateSpawnLocationTag() {is=title K2Node_FunctionEntry}
   > Reset Widget Object(Target=Get LagerEventDataView(), NewObject=Get ActiveLagerEventData())
  ]}
 UpdateSpawnerDataTag {nodes [
   n1 = Set SpawnerDataTag(execute=@n15.then, SpawnerDataTag:Gameplay Tag Container Structure=GetDescriptorTag(self=Get ActiveSpawnerData())) -> (Output_Get:Gameplay Tag Container Structure)
   n2 = DisplayMessageBox(
    Title=UpdateSpawnerDataTag
    Message=Append(A="Tag : ", B=GetDebugStringFromGameplayTagContainer!(TagContainer:"Gameplay Tag Container Structure (by ref)"=@n1) -> (ReturnValue:String))
   )
   n3 = Sequence!(execute=[@n2.then, @n1.then]) -> (then_0, then_1, then_2, then_3:Exec, then_4:Exec)
   n4 = Get ActiveNPCCharacterData()
   Branch(execute=@n3.then_2, Condition=IsValid(Object=@n4))
   n5 = BreakGameplayTagContainer(GameplayTagContainer=Get SpawnerDataTag() -> (SpawnerDataTag:Gameplay Tag Container Structure))
   > n6 = Branch(Condition=Is Valid Index(TargetArray:Array of Gameplay Tag Structures=@n5))
   > SetEditorProperty(
    Object=@n4
    PropertyName=GameplayTag
    PropertyValue:Gameplay Tag Structure="Get (a copy)"(Array:Array of Gameplay Tag Structures=@n5) -> (Output:Gameplay Tag Structure)
   )
   SetEditorProperty(execute=@n6.else, Object=@n4, PropertyName=GameplayTag, PropertyValue=Get EmptyTag() -> (EmptyTag:Gameplay Tag Structure))
   n7 = GetTemplateConfig(execute=@n3.then_1)
   > Branch(Condition=Get SetLooseTags!(self:BP NPCEditor Config Object Reference=@n7) -> (SetLooseTags:bool))
   > n8 = GetEditableObjectByIndex(Index=5)
   > n9 = Cast To FortAthenaAISpawnerDataComponent_GameplayAbilityBase(Object=@n8)
   > n10 = SetEditorProperty(Object=@n9, PropertyName=LooseTagsToApplyToPawn, PropertyValue:Gameplay Tag Container Structure=Get SpawnerDataTag())
   > n11 = Branch(Condition=@n10)
   > Reset Widget Object(Target=Get GameplayAbilityComponentView(), NewObject=@n8)
   DisplayMessageBox(execute=@n11.else, Title=UpdateSpawnerDataTag, Message=Failed to LooseTagsToApplyToPawn on GameplayAbility component)
   DisplayMessageBox(execute=@n9.CastFailed, Title=UpdateSpawnerDataTag, Message="Error, couldn't properly access the GameplayAbility component.")
   n12 = Get ActiveLagerEventData()
   Branch(execute=@n3.then_0, Condition=IsValid(Object=@n12))
   > n13 = IsObjectOverriden(Object=@n12)
   > Branch(Condition=@n13)
   n14 = BreakGameplayTagContainer(GameplayTagContainer=Get SpawnerDataTag())
   > Branch(Condition=Is Valid Index(TargetArray=@n14))
   > SetupSimpleActorDescription!(
    execute
    self:Fort Athena Living World Event Data Object Reference=@n12
    SpawnerDataTag:"Gameplay Tag Structure (by ref)"="Get (a copy)"(Array=@n14)
   ) -> (then)
   > Reset Widget Object(Target=Get LagerEventDataView(), NewObject=@n12)
   Reset Widget Object(execute=@n3.then_3, Target=Get ComponentListsView(), NewObject=Get ActiveSpawnerData())
   > Reset Widget Object(Target=Get NPCCharacterDataView(), NewObject=Get ActiveNPCCharacterData())
   > Reset Widget Object(Target=Get AISpawnerDataGameplayTag(), NewObject=Get ActiveSpawnerData())
   n15 = UpdateSpawnerDataTag() {is=title K2Node_FunctionEntry}
  ]}
]
variables [
 ActiveLagerEventData {type=FortAthenaLivingWorldEventData, default=null}
 ActiveLagerEventDataAssets {type="array<string>", default []}
 ActiveNPCCharacterData {type=FortTandemCharacterData, default=null}
 ActiveNPCLibraryConversation {type=Object, default=null}
 ActiveNPCName {type=string, default=""}
 ActiveObject {type=Object, default=null}
 ActiveSpawnerData {type=FortAthenaAIBotSpawnerData, default=null}
 ActiveWidgetIndex {type=integer}
 ComponentListsViewMembers {
  type="array<string>"
  default [InventoryComponent, AffiliationComponent, GameplayComponent, ConversationComponent, CosmeticComponent, GameplayAbilityComponent]
 }
 Configs {
  type="array<Class>"
  default [
   /Game/NeverCook/Blutilities/ForDesign/NPCs/NPCConfig/BP_NPCEditor_Config.BP_NPCEditor_Config_C
   /Game/NeverCook/Blutilities/ForDesign/NPCs/NPCConfig/BP_NPCEditor_Config_Juno.BP_NPCEditor_Config_Juno_C
  ]
 }
 CreateChildAssets {type=boolean, default=true}
 DefaultPanelIndex {type=integer}
 EmptyStringArray {type="array<string>", default []}
 FileTypes {type="map<string>", default {.uasset=.uasset}}
 FullViewSourceIndex {type=integer, default=-1}
 LagerEventDataViewIndex {type=integer, default=8}
 ListActiveNPCObjects {type="array<Object>", default []}
 SelectedTemplateIndex {type=integer, default=-1}
]
widgets {
 info {rootWidgetClass=&CanvasPanel, widgetCount=70}
 widgets [
  CanvasRoot {widgetClassPath=&CanvasPanel, bIsVariable=true, widget=CanvasRoot}
  Border_57 {
   widgetClassPath=&Border
   parent=CanvasRoot
   widget=Border_57
   properties {brushColor=<rgba 0.0677 0.0677 0.0677 0>}
   slot {class=CanvasPanelSlot, properties {layoutData {alignment=<0 0>, anchors {maximum=<1 1>, minimum=<0 0>}, offsets=<ltrb 0 0 0 0>}}}
  }
  VerticalBox_239 {widgetClassPath=&VerticalBox, parent=Border_57, widget=VerticalBox_239, slot {class=BorderSlot}}
  HorizontalBox_279 {widgetClassPath=&HorizontalBox, parent=VerticalBox_239, widget=HorizontalBox_279, slot {class=VerticalBoxSlot}}
  Btn_NewNPC {
   widgetClassPath=/Game/Blueprints/Blutilities/ForDesign/NPCs/SupportWidget/TA_NPCEditor_Button
   bIsVariable=true
   parent=HorizontalBox_279
   widget=Btn_NewNPC
   properties {label=New NPC}
   slot {class=HorizontalBoxSlot, properties {padding=<ltrb 0 0 10 0>}}
  }
  Btn_LoadNPC {
   widgetClassPath=&TA_NPCEditor_Button
   bIsVariable=true
   parent=HorizontalBox_279
   widget=Btn_LoadNPC
   properties {label=Load NPC}
   slot {class=HorizontalBoxSlot, properties {padding=<ltrb 0 0 10 0>}}
  }
  Btn_SaveNPC {
   widgetClassPath=/Game/NeverCook/EditorUtilities/QuestTools/SupportWidget/TA_QuestEditor_Button
   bIsVariable=true
   parent=HorizontalBox_279
   widget=Btn_SaveNPC
   properties {label=Save Current NPC, buttonColor_Default=<rgba 0.0024 0.1195 0.4969 1>}
   slot {class=HorizontalBoxSlot, properties {padding=<ltrb 0 0 10 0>}}
  }
  HorizontalBox_90 {
   widgetClassPath=&HorizontalBox
   parent=HorizontalBox_279
   widget=HorizontalBox_90
   slot {class=HorizontalBoxSlot, properties {horizontalAlignment=HAlign_Right, size {sizeRule=Fill, value=1}}}
  }
  ConfigSelector {
   widgetClassPath=/Script/UMG.ComboBoxString
   bIsVariable=true
   parent=HorizontalBox_90
   widget=ConfigSelector
   properties {
    foregroundColor {colorUseRule=UseColor_Specified, specifiedColor=<rgba 0.5271 0.5271 0.5271 1>}
    itemStyle {
     bUseParentRowBrush=false
     activeBrush {
      bIsDynamicallyLoaded=false
      drawAs=Image
      imageType=NoImage
      mirroring=NoMirror
      resourceName=None
      resourceObject=null
      tiling=NoTile
      imageSize=<0 0>
      margin=<ltrb 0 0 0 0>
      outlineSettings {
       bUseBrushTransparency=false
       roundingType=HalfHeightRadius
       width=0
       color {colorUseRule=UseColor_Specified, specifiedColor=<rgba 0 0 0 0>}
       cornerRadii=<0 0 0 0>
      }
      tintColor {colorUseRule=UseColor_Specified, specifiedColor=<rgba 0 0.162 0.7454 1>}
      uVRegion {bIsValid=false, max=<0 0>, min=<0 0>}
     }
     activeHighlightedBrush {
      bIsDynamicallyLoaded=false
      drawAs=Image
      imageType=NoImage
      mirroring=NoMirror
      resourceName=None
      resourceObject=null
      tiling=NoTile
      imageSize=<0 0>
      margin=<ltrb 0 0 0 0>
      outlineSettings {
       bUseBrushTransparency=false
       roundingType=HalfHeightRadius
       width=0
       color {colorUseRule=UseColor_Specified, specifiedColor=<rgba 0 0 0 0>}
       cornerRadii=<0 0 0 0>
      }
      tintColor {colorUseRule=UseColor_Specified, specifiedColor=<rgba 0.0252 0.0319 0.0423 1>}
      uVRegion {bIsValid=false, max=<0 0>, min=<0 0>}
     }
     activeHoveredBrush {
      bIsDynamicallyLoaded=false
      drawAs=Image
      imageType=NoImage
      mirroring=NoMirror
      resourceName=None
      resourceObject=null
      tiling=NoTile
      imageSize=<0 0>
      margin=<ltrb 0 0 0 0>
      outlineSettings {
       bUseBrushTransparency=false
       roundingType=HalfHeightRadius
       width=0
       color {colorUseRule=UseColor_Specified, specifiedColor=<rgba 0 0 0 0>}
       cornerRadii=<0 0 0 0>
      }
      tintColor {colorUseRule=UseColor_Specified, specifiedColor=<rgba 0 0.162 0.7454 1>}
      uVRegion {bIsValid=false, max=<0 0>, min=<0 0>}
     }
     dropIndicator_Above {
      bIsDynamicallyLoaded=false
      drawAs=Box
      imageType=FullColor
      mirroring=NoMirror
      resourceName=../../../Engine/Content/Slate/Common/DropZoneIndicator_Above.png
      resourceObject=null
      tiling=NoTile
      imageSize=<0 0>
      margin=<ltrb 0.625 0.625 0 0>
      outlineSettings {
       bUseBrushTransparency=false
       roundingType=HalfHeightRadius
       width=0
       color {colorUseRule=UseColor_Specified, specifiedColor=<rgba 0 0 0 0>}
       cornerRadii=<0 0 0 0>
      }
      tintColor {colorUseRule=UseColor_Specified, specifiedColor=<rgba 0.0091 0.3515 0.9647 1>}
      uVRegion {bIsValid=false, max=<0 0>, min=<0 0>}
     }
     dropIndicator_Below {
      bIsDynamicallyLoaded=false
      drawAs=Box
      imageType=FullColor
      mirroring=NoMirror
      resourceName=../../../Engine/Content/Slate/Common/DropZoneIndicator_Below.png
      resourceObject=null
      tiling=NoTile
      imageSize=<0 0>
      margin=<ltrb 0.625 0 0 0.625>
      outlineSettings {
       bUseBrushTransparency=false
       roundingType=HalfHeightRadius
       width=0
       color {colorUseRule=UseColor_Specified, specifiedColor=<rgba 0 0 0 0>}
       cornerRadii=<0 0 0 0>
      }
      tintColor {colorUseRule=UseColor_Specified, specifiedColor=<rgba 0.0091 0.3515 0.9647 1>}
      uVRegion {bIsValid=false, max=<0 0>, min=<0 0>}
     }
     dropIndicator_Onto {
      bIsDynamicallyLoaded=false
      drawAs=Box
      imageType=FullColor
      mirroring=NoMirror
      resourceName=../../../Engine/Content/Slate/Common/DropZoneIndicator_Onto.png
      resourceObject=null
      tiling=NoTile
      imageSize=<0 0>
      margin=<ltrb 0.25 0.25 0.25 0.25>
      outlineSettings {
       bUseBrushTransparency=false
       roundingType=HalfHeightRadius
       width=0
       color {colorUseRule=UseColor_Specified, specifiedColor=<rgba 0 0 0 0>}
       cornerRadii=<0 0 0 0>
      }
      tintColor {colorUseRule=UseColor_Specified, specifiedColor=<rgba 0.0091 0.3515 0.9647 1>}
      uVRegion {bIsValid=false, max=<0 0>, min=<0 0>}
     }
     evenRowBackgroundBrush {
      bIsDynamicallyLoaded=false
      drawAs=Image
      imageType=NoImage
      mirroring=NoMirror
      resourceName=None
      resourceObject=null
      tiling=NoTile
      imageSize=<0 0>
      margin=<ltrb 0 0 0 0>
      outlineSettings {
       bUseBrushTransparency=false
       roundingType=HalfHeightRadius
       width=0
       color {colorUseRule=UseColor_Specified, specifiedColor=<rgba 0 0 0 0>}
       cornerRadii=<0 0 0 0>
      }
      tintColor {colorUseRule=UseColor_Specified, specifiedColor=<rgba 0.0103 0.0103 0.0103 1>}
      uVRegion {bIsValid=false, max=<0 0>, min=<0 0>}
     }
     evenRowBackgroundHoveredBrush {
      bIsDynamicallyLoaded=false
      drawAs=Image
      imageType=NoImage
      mirroring=NoMirror
      resourceName=None
      resourceObject=null
      tiling=NoTile
      imageSize=<0 0>
      margin=<ltrb 0 0 0 0>
      outlineSettings {
       bUseBrushTransparency=false
       roundingType=HalfHeightRadius
       width=0
       color {colorUseRule=UseColor_Specified, specifiedColor=<rgba 0 0 0 0>}
       cornerRadii=<0 0 0 0>
      }
      tintColor {colorUseRule=UseColor_Specified, specifiedColor=<rgba 0.0176 0.0176 0.0176 1>}
      uVRegion {bIsValid=false, max=<0 0>, min=<0 0>}
     }
     inactiveBrush {
      bIsDynamicallyLoaded=false
      drawAs=Image
      imageType=NoImage
      mirroring=NoMirror
      resourceName=None
      resourceObject=null
      tiling=NoTile
      imageSize=<0 0>
      margin=<ltrb 0 0 0 0>
      outlineSettings {
       bUseBrushTransparency=false
       roundingType=HalfHeightRadius
       width=0
       color {colorUseRule=UseColor_Specified, specifiedColor=<rgba 0 0 0 0>}
       cornerRadii=<0 0 0 0>
      }
      tintColor {colorUseRule=UseColor_Specified, specifiedColor=<rgba 0.0513 0.0953 0.159 1>}
      uVRegion {bIsValid=false, max=<0 0>, min=<0 0>}
     }
     inactiveHighlightedBrush {
      bIsDynamicallyLoaded=false
      drawAs=Image
      imageType=NoImage
      mirroring=NoMirror
      resourceName=None
      resourceObject=null
      tiling=NoTile
      imageSize=<0 0>
      margin=<ltrb 0 0 0 0>
      outlineSettings {
       bUseBrushTransparency=false
       roundingType=HalfHeightRadius
       width=0
       color {colorUseRule=UseColor_Specified, specifiedColor=<rgba 0 0 0 0>}
       cornerRadii=<0 0 0 0>
      }
      tintColor {colorUseRule=UseColor_Specified, specifiedColor=<rgba 0.0252 0.0319 0.0423 1>}
      uVRegion {bIsValid=false, max=<0 0>, min=<0 0>}
     }
     inactiveHoveredBrush {
      bIsDynamicallyLoaded=false
      drawAs=Image
      imageType=NoImage
      mirroring=NoMirror
      resourceName=None
      resourceObject=null
      tiling=NoTile
      imageSize=<0 0>
      margin=<ltrb 0 0 0 0>
      outlineSettings {
       bUseBrushTransparency=false
       roundingType=HalfHeightRadius
       width=0
       color {colorUseRule=UseColor_Specified, specifiedColor=<rgba 0 0 0 0>}
       cornerRadii=<0 0 0 0>
      }
      tintColor {colorUseRule=UseColor_Specified, specifiedColor=<rgba 0.0513 0.0953 0.159 1>}
      uVRegion {bIsValid=false, max=<0 0>, min=<0 0>}
     }
     oddRowBackgroundBrush {
      bIsDynamicallyLoaded=false
      drawAs=Image
      imageType=NoImage
      mirroring=NoMirror
      resourceName=None
      resourceObject=null
      tiling=NoTile
      imageSize=<0 0>
      margin=<ltrb 0 0 0 0>
      outlineSettings {
       bUseBrushTransparency=false
       roundingType=HalfHeightRadius
       width=0
       color {colorUseRule=UseColor_Specified, specifiedColor=<rgba 0 0 0 0>}
       cornerRadii=<0 0 0 0>
      }
      tintColor {colorUseRule=UseColor_Specified, specifiedColor=<rgba 0.0103 0.0103 0.0103 1>}
      uVRegion {bIsValid=false, max=<0 0>, min=<0 0>}
     }
     oddRowBackgroundHoveredBrush {
      bIsDynamicallyLoaded=false
      drawAs=Image
      imageType=NoImage
      mirroring=NoMirror
      resourceName=None
      resourceObject=null
      tiling=NoTile
      imageSize=<0 0>
      margin=<ltrb 0 0 0 0>
      outlineSettings {
       bUseBrushTransparency=false
       roundingType=HalfHeightRadius
       width=0
       color {colorUseRule=UseColor_Specified, specifiedColor=<rgba 0 0 0 0>}
       cornerRadii=<0 0 0 0>
      }
      tintColor {colorUseRule=UseColor_Specified, specifiedColor=<rgba 0.0176 0.0176 0.0176 1>}
      uVRegion {bIsValid=false, max=<0 0>, min=<0 0>}
     }
     parentRowBackgroundBrush {
      bIsDynamicallyLoaded=false
      drawAs=Image
      imageType=NoImage
      mirroring=NoMirror
      resourceName=None
      resourceObject=null
      tiling=NoTile
      imageSize=<32 32>
      margin=<ltrb 0 0 0 0>
      outlineSettings {
       bUseBrushTransparency=false
       roundingType=HalfHeightRadius
       width=0
       color {colorUseRule=UseColor_Specified, specifiedColor=<rgba 0 0 0 0>}
       cornerRadii=<0 0 0 0>
      }
      tintColor {colorUseRule=UseColor_Specified, specifiedColor=<rgba 1 1 1 1>}
      uVRegion {bIsValid=false, max=<0 0>, min=<0 0>}
     }
     parentRowBackgroundHoveredBrush {
      bIsDynamicallyLoaded=false
      drawAs=Image
      imageType=NoImage
      mirroring=NoMirror
      resourceName=None
      resourceObject=null
      tiling=NoTile
      imageSize=<32 32>
      margin=<ltrb 0 0 0 0>
      outlineSettings {
       bUseBrushTransparency=false
       roundingType=HalfHeightRadius
       width=0
       color {colorUseRule=UseColor_Specified, specifiedColor=<rgba 0 0 0 0>}
       cornerRadii=<0 0 0 0>
      }
      tintColor {colorUseRule=UseColor_Specified, specifiedColor=<rgba 1 1 1 1>}
      uVRegion {bIsValid=false, max=<0 0>, min=<0 0>}
     }
     selectedTextColor {colorUseRule=UseColor_Specified, specifiedColor=<rgba 1 1 1 1>}
     selectorFocusedBrush {
      bIsDynamicallyLoaded=false
      drawAs=Border
      imageType=FullColor
      mirroring=NoMirror
      resourceName=../../../Engine/Content/Slate/Common/Selector.png
      resourceObject=null
      tiling=Both
      imageSize=<0 0>
      margin=<ltrb 0.25 0.25 0.25 0.25>
      outlineSettings {
       bUseBrushTransparency=false
       roundingType=HalfHeightRadius
       width=0
       color {colorUseRule=UseColor_Specified, specifiedColor=<rgba 0 0 0 0>}
       cornerRadii=<0 0 0 0>
      }
      tintColor {colorUseRule=UseColor_Specified, specifiedColor=<rgba 0.701 0.225 0.003 1>}
      uVRegion {bIsValid=false, max=<0 0>, min=<0 0>}
     }
     textColor {colorUseRule=UseColor_Specified, specifiedColor=<rgba 0.5271 0.5271 0.5271 1>}
    }
    scrollBarStyle {
     thickness=8
     draggedThumbImage {
      bIsDynamicallyLoaded=false
      drawAs=RoundedBox
      imageType=NoImage
      mirroring=NoMirror
      resourceName=None
      resourceObject=null
      tiling=NoTile
      imageSize=<0 0>
      margin=<ltrb 0 0 0 0>
      outlineSettings {
       bUseBrushTransparency=false
       roundingType=FixedRadius
       width=0
       color {colorUseRule=UseColor_Specified, specifiedColor=<rgba 0 0 0 0>}
       cornerRadii=<4 4 4 4>
      }
      tintColor {colorUseRule=UseColor_Specified, specifiedColor=<rgba 0.2159 0.2159 0.2159 1>}
      uVRegion {bIsValid=false, max=<0 0>, min=<0 0>}
     }
     horizontalBackgroundImage {
      bIsDynamicallyLoaded=false
      drawAs=NoDrawType
      imageType=NoImage
      mirroring=NoMirror
      resourceName=None
      resourceObject=null
      tiling=NoTile
      imageSize=<0 0>
      margin=<ltrb 0 0 0 0>
      outlineSettings {
       bUseBrushTransparency=false
       roundingType=HalfHeightRadius
       width=0
       color {colorUseRule=UseColor_Specified, specifiedColor=<rgba 0 0 0 0>}
       cornerRadii=<0 0 0 0>
      }
      tintColor {colorUseRule=UseColor_Specified, specifiedColor=<rgba 1 1 1 1>}
      uVRegion {bIsValid=false, max=<0 0>, min=<0 0>}
     }
     horizontalBottomSlotImage {
      bIsDynamicallyLoaded=false
      drawAs=NoDrawType
      imageType=NoImage
      mirroring=NoMirror
      resourceName=None
      resourceObject=null
      tiling=NoTile
      imageSize=<0 0>
      margin=<ltrb 0 0 0 0>
      outlineSettings {
       bUseBrushTransparency=false
       roundingType=HalfHeightRadius
       width=0
       color {colorUseRule=UseColor_Specified, specifiedColor=<rgba 0 0 0 0>}
       cornerRadii=<0 0 0 0>
      }
      tintColor {colorUseRule=UseColor_Specified, specifiedColor=<rgba 1 1 1 1>}
      uVRegion {bIsValid=false, max=<0 0>, min=<0 0>}
     }
     horizontalTopSlotImage {
      bIsDynamicallyLoaded=false
      drawAs=NoDrawType
      imageType=NoImage
      mirroring=NoMirror
      resourceName=None
      resourceObject=null
      tiling=NoTile
      imageSize=<0 0>
      margin=<ltrb 0 0 0 0>
      outlineSettings {
       bUseBrushTransparency=false
       roundingType=HalfHeightRadius
       width=0
       color {colorUseRule=UseColor_Specified, specifiedColor=<rgba 0 0 0 0>}
       cornerRadii=<0 0 0 0>
      }
      tintColor {colorUseRule=UseColor_Specified, specifiedColor=<rgba 1 1 1 1>}
      uVRegion {bIsValid=false, max=<0 0>, min=<0 0>}
     }
     hoveredThumbImage {
      bIsDynamicallyLoaded=false
      drawAs=RoundedBox
      imageType=NoImage
      mirroring=NoMirror
      resourceName=None
      resourceObject=null
      tiling=NoTile
      imageSize=<0 0>
      margin=<ltrb 0 0 0 0>
      outlineSettings {
       bUseBrushTransparency=false
       roundingType=FixedRadius
       width=0
       color {colorUseRule=UseColor_Specified, specifiedColor=<rgba 0 0 0 0>}
       cornerRadii=<4 4 4 4>
      }
      tintColor {colorUseRule=UseColor_Specified, specifiedColor=<rgba 0.2159 0.2159 0.2159 1>}
      uVRegion {bIsValid=false, max=<0 0>, min=<0 0>}
     }
     normalThumbImage {
      bIsDynamicallyLoaded=false
      drawAs=RoundedBox
      imageType=NoImage
      mirroring=NoMirror
      resourceName=None
      resourceObject=null
      tiling=NoTile
      imageSize=<0 0>
      margin=<ltrb 0 0 0 0>
      outlineSettings {
       bUseBrushTransparency=false
       roundingType=FixedRadius
       width=0
       color {colorUseRule=UseColor_Specified, specifiedColor=<rgba 0 0 0 0>}
       cornerRadii=<4 4 4 4>
      }
      tintColor {colorUseRule=UseColor_Specified, specifiedColor=<rgba 0.0953 0.0953 0.0953 1>}
      uVRegion {bIsValid=false, max=<0 0>, min=<0 0>}
     }
     verticalBackgroundImage {
      bIsDynamicallyLoaded=false
      drawAs=NoDrawType
      imageType=NoImage
      mirroring=NoMirror
      resourceName=None
      resourceObject=null
      tiling=NoTile
      imageSize=<0 0>
      margin=<ltrb 0 0 0 0>
      outlineSettings {
       bUseBrushTransparency=false
       roundingType=HalfHeightRadius
       width=0
       color {colorUseRule=UseColor_Specified, specifiedColor=<rgba 0 0 0 0>}
       cornerRadii=<0 0 0 0>
      }
      tintColor {colorUseRule=UseColor_Specified, specifiedColor=<rgba 1 1 1 1>}
      uVRegion {bIsValid=false, max=<0 0>, min=<0 0>}
     }
     verticalBottomSlotImage {
      bIsDynamicallyLoaded=false
      drawAs=NoDrawType
      imageType=NoImage
      mirroring=NoMirror
      resourceName=None
      resourceObject=null
      tiling=NoTile
      imageSize=<0 0>
      margin=<ltrb 0 0 0 0>
      outlineSettings {
       bUseBrushTransparency=false
       roundingType=HalfHeightRadius
       width=0
       color {colorUseRule=UseColor_Specified, specifiedColor=<rgba 0 0 0 0>}
       cornerRadii=<0 0 0 0>
      }
      tintColor {colorUseRule=UseColor_Specified, specifiedColor=<rgba 1 1 1 1>}
      uVRegion {bIsValid=false, max=<0 0>, min=<0 0>}
     }
     verticalTopSlotImage {
      bIsDynamicallyLoaded=false
      drawAs=NoDrawType
      imageType=NoImage
      mirroring=NoMirror
      resourceName=None
      resourceObject=null
      tiling=NoTile
      imageSize=<0 0>
      margin=<ltrb 0 0 0 0>
      outlineSettings {
       bUseBrushTransparency=false
       roundingType=HalfHeightRadius
       width=0
       color {colorUseRule=UseColor_Specified, specifiedColor=<rgba 0 0 0 0>}
       cornerRadii=<0 0 0 0>
      }
      tintColor {colorUseRule=UseColor_Specified, specifiedColor=<rgba 1 1 1 1>}
      uVRegion {bIsValid=false, max=<0 0>, min=<0 0>}
     }
    }
    widgetStyle {
     comboButtonStyle {
      downArrowAlign=VAlign_Center
      buttonStyle {
       clickedSlateSound {resourceObject=null}
       disabled {
        bIsDynamicallyLoaded=false
        drawAs=RoundedBox
        imageType=NoImage
        mirroring=NoMirror
        resourceName=None
        resourceObject=null
        tiling=NoTile
        imageSize=<0 0>
        margin=<ltrb 0 0 0 0>
        outlineSettings {
         bUseBrushTransparency=false
         roundingType=FixedRadius
         width=1
         color {colorUseRule=UseColor_ColorTable, specifiedColor=<rgba 0 0 0 0>}
         cornerRadii=<4 4 4 4>
        }
        tintColor {colorUseRule=UseColor_Specified, specifiedColor=<rgba 0.0395 0.0395 0.0395 1>}
        uVRegion {bIsValid=false, max=<0 0>, min=<0 0>}
       }
       disabledForeground {colorUseRule=UseColor_ColorTable, specifiedColor=<rgba 0 0 0 0>}
       hovered {
        bIsDynamicallyLoaded=false
        drawAs=RoundedBox
        imageType=NoImage
        mirroring=NoMirror
        resourceName=None
        resourceObject=null
        tiling=NoTile
        imageSize=<0 0>
        margin=<ltrb 0 0 0 0>
        outlineSettings {
         bUseBrushTransparency=false
         roundingType=FixedRadius
         width=1
         color {colorUseRule=UseColor_ColorTable, specifiedColor=<rgba 0 0 0 0>}
         cornerRadii=<4 4 4 4>
        }
        tintColor {colorUseRule=UseColor_Specified, specifiedColor=<rgba 0.0048 0.0048 0.0048 1>}
        uVRegion {bIsValid=false, max=<0 0>, min=<0 0>}
       }
       hoveredForeground {colorUseRule=UseColor_ColorTable, specifiedColor=<rgba 0 0 0 0>}
       hoveredSlateSound {resourceObject=null}
       normal {
        bIsDynamicallyLoaded=false
        drawAs=RoundedBox
        imageType=NoImage
        mirroring=NoMirror
        resourceName=None
        resourceObject=null
        tiling=NoTile
        imageSize=<0 0>
        margin=<ltrb 0 0 0 0>
        outlineSettings {
         bUseBrushTransparency=false
         roundingType=FixedRadius
         width=1
         color {colorUseRule=UseColor_ColorTable, specifiedColor=<rgba 0 0 0 0>}
         cornerRadii=<4 4 4 4>
        }
        tintColor {colorUseRule=UseColor_Specified, specifiedColor=<rgba 0.0048 0.0048 0.0048 1>}
        uVRegion {bIsValid=false, max=<0 0>, min=<0 0>}
       }
       normalForeground {colorUseRule=UseColor_ColorTable, specifiedColor=<rgba 0 0 0 0>}
       normalPadding=<ltrb 8 1 8 1>
       pressed {
        bIsDynamicallyLoaded=false
        drawAs=RoundedBox
        imageType=NoImage
        mirroring=NoMirror
        resourceName=None
        resourceObject=null
        tiling=NoTile
        imageSize=<0 0>
        margin=<ltrb 0 0 0 0>
        outlineSettings {
         bUseBrushTransparency=false
         roundingType=FixedRadius
         width=1
         color {colorUseRule=UseColor_ColorTable, specifiedColor=<rgba 0 0 0 0>}
         cornerRadii=<4 4 4 4>
        }
        tintColor {colorUseRule=UseColor_Specified, specifiedColor=<rgba 0.0048 0.0048 0.0048 1>}
        uVRegion {bIsValid=false, max=<0 0>, min=<0 0>}
       }
       pressedForeground {colorUseRule=UseColor_ColorTable, specifiedColor=<rgba 0 0 0 0>}
       pressedPadding=<ltrb 8 2 8 0>
       pressedSlateSound {resourceObject=null}
      }
      contentPadding=<ltrb 0 0 0 0>
      downArrowImage {
       bIsDynamicallyLoaded=false
       drawAs=Image
       imageType=Vector
       mirroring=NoMirror
       resourceName=../../../Engine/Content/Slate/Starship/CoreWidgets/ComboBox/wide-chevron-down.svg
       resourceObject=null
       tiling=NoTile
       imageSize=<16 16>
       margin=<ltrb 0 0 0 0>
       outlineSettings {
        bUseBrushTransparency=false
        roundingType=HalfHeightRadius
        width=0
        color {colorUseRule=UseColor_Specified, specifiedColor=<rgba 0 0 0 0>}
        cornerRadii=<0 0 0 0>
       }
       tintColor {colorUseRule=UseColor_Specified, specifiedColor=<rgba 1 1 1 1>}
       uVRegion {bIsValid=false, max=<0 0>, min=<0 0>}
      }
      downArrowPadding=<ltrb 2 2 2 2>
      menuBorderBrush {
       bIsDynamicallyLoaded=false
       drawAs=RoundedBox
       imageType=NoImage
       mirroring=NoMirror
       resourceName=None
       resourceObject=null
       tiling=NoTile
       imageSize=<0 0>
       margin=<ltrb 0 0 0 0>
       outlineSettings {
        bUseBrushTransparency=false
        roundingType=FixedRadius
        width=1
        color {colorUseRule=UseColor_ColorTable, specifiedColor=<rgba 0 0 0 0>}
        cornerRadii=<0 0 0 0>
       }
       tintColor {colorUseRule=UseColor_Specified, specifiedColor=<rgba 0.0395 0.0395 0.0395 1>}
       uVRegion {bIsValid=false, max=<0 0>, min=<0 0>}
      }
      menuBorderPadding=<ltrb 0 0 0 0>
      shadowColorAndOpacity=<rgba 0 0 0 1>
      shadowOffset=<0 0>
     }
     contentPadding=<ltrb 0 0 0 0>
     menuRowPadding=<ltrb 8 3 8 3>
     pressedSlateSound {resourceObject=null}
     selectionChangeSlateSound {resourceObject=null}
    }
   }
   slot {class=HorizontalBoxSlot}
  }
  VerticalBox_136 {
   widgetClassPath=&VerticalBox
   parent=VerticalBox_239
   widget=VerticalBox_136
   slot {class=VerticalBoxSlot, properties {padding=<ltrb 0 10 0 0>}}
  }
  EditHeader {
   widgetClassPath=&CommonTextBlock
   parent=VerticalBox_136
   widget=EditHeader
   properties {
    text=Currently Editing:
    colorAndOpacity {colorUseRule=UseColor_Specified, specifiedColor=<rgba 0.1095 0.1095 0.1095 1>}
    font {
     bForceMonospaced=false
     bMaterialIsStencil=false
     fontMaterial=null
     fontObject=&NotoSans
     letterSpacing=0
     monospacedWidth=1
     size=9
     skewAmount=0
     typefaceFontName=NotoSans-Italic
     outlineSettings {
      bApplyOutlineToDropShadows=false
      bMiteredCorners=false
      bSeparateFillAlpha=false
      outlineMaterial=null
      outlineSize=0
      outlineColor=<rgba 0 0 0 1>
     }
    }
    shadowOffset=<0 0>
   }
   slot {class=VerticalBoxSlot}
  }
  AssetNameText {
   widgetClassPath=&CommonTextBlock
   bIsVariable=true
   parent=VerticalBox_136
   widget=AssetNameText
   properties {
    text=No Asset Active
    colorAndOpacity {colorUseRule=UseColor_Specified, specifiedColor=<rgba 0.9734 0.6308 0.0409 1>}
    font {
     bForceMonospaced=false
     bMaterialIsStencil=false
     fontMaterial=null
     fontObject=&NotoSans
     letterSpacing=0
     monospacedWidth=1
     size=16
     skewAmount=0
     typefaceFontName=NotoSans-Bold
     outlineSettings {
      bApplyOutlineToDropShadows=false
      bMiteredCorners=false
      bSeparateFillAlpha=false
      outlineMaterial=null
      outlineSize=0
      outlineColor=<rgba 0 0 0 1>
     }
    }
    shadowOffset=<0 0>
   }
   slot {class=VerticalBoxSlot}
  }
  HorizontalBox_114 {widgetClassPath=&HorizontalBox, parent=VerticalBox_136, widget=HorizontalBox_114, slot {class=VerticalBoxSlot}}
  VerticalBox_149 {widgetClassPath=&VerticalBox, parent=HorizontalBox_114, widget=VerticalBox_149, slot {class=HorizontalBoxSlot}}
  VerticalBox_87 {
   widgetClassPath=&VerticalBox
   parent=VerticalBox_149
   widget=VerticalBox_87
   slot {class=VerticalBoxSlot, properties {padding=<ltrb 5 0 0 0>}}
  }
  HorizontalBox_119 {widgetClassPath=&HorizontalBox, parent=VerticalBox_87, widget=HorizontalBox_119, slot {class=VerticalBoxSlot}}
  SpawnerDataTagText {
   widgetClassPath=&CommonTextBlock
   parent=HorizontalBox_119
   widget=SpawnerDataTagText
   properties {
    text=Spawner Data Tag
    colorAndOpacity {colorUseRule=UseColor_Specified, specifiedColor=<rgba 0.5156 0.5156 0.5156 1>}
    font {
     bForceMonospaced=false
     bMaterialIsStencil=false
     fontMaterial=null
     fontObject=&NotoSans
     letterSpacing=0
     monospacedWidth=1
     size=15
     skewAmount=0
     typefaceFontName=NotoSans-Italic
     outlineSettings {
      bApplyOutlineToDropShadows=false
      bMiteredCorners=false
      bSeparateFillAlpha=false
      outlineMaterial=null
      outlineSize=0
      outlineColor=<rgba 0 0 0 1>
     }
    }
    shadowOffset=<0 0>
   }
   slot {class=HorizontalBoxSlot}
  }
  AISpawnerDataGameplayTag {
   widgetClassPath=/Script/FortniteEditor.FortQuestToolPropertyView
   bIsVariable=true
   parent=HorizontalBox_119
   widget=AISpawnerDataGameplayTag
   properties {bShowNameWidgets=false, propertiesToShow [DescriptorTag]}
   slot {class=HorizontalBoxSlot, properties {padding=<ltrb 5 0 0 0>}}
  }
  HorizontalBox {widgetClassPath=&HorizontalBox, parent=VerticalBox_87, widget=HorizontalBox, slot {class=VerticalBoxSlot}}
  SpawnLocationTagText {
   widgetClassPath=&CommonTextBlock
   parent=HorizontalBox
   widget=SpawnLocationTagText
   properties {
    text=Spawn Location Tag
    colorAndOpacity {colorUseRule=UseColor_Specified, specifiedColor=<rgba 0.5156 0.5156 0.5156 1>}
    font {
     bForceMonospaced=false
     bMaterialIsStencil=false
     fontMaterial=null
     fontObject=&NotoSans
     letterSpacing=0
     monospacedWidth=1
     size=15
     skewAmount=0
     typefaceFontName=NotoSans-Italic
     outlineSettings {
      bApplyOutlineToDropShadows=false
      bMiteredCorners=false
      bSeparateFillAlpha=false
      outlineMaterial=null
      outlineSize=0
      outlineColor=<rgba 0 0 0 1>
     }
    }
    shadowOffset=<0 0>
   }
   slot {class=HorizontalBoxSlot}
  }
  LagerEventDataNotFound {
   widgetClassPath=&CommonTextBlock
   bIsVariable=true
   parent=HorizontalBox
   widget=LagerEventDataNotFound
   properties {
    text=LagerEventData not found
    visibility=Collapsed
    colorAndOpacity {colorUseRule=UseColor_Specified, specifiedColor=<rgba 0.5156 0.1504 0.1504 1>}
    font {
     bForceMonospaced=false
     bMaterialIsStencil=false
     fontMaterial=null
     fontObject=&NotoSans
     letterSpacing=0
     monospacedWidth=1
     size=15
     skewAmount=0
     typefaceFontName=NotoSans-Italic
     outlineSettings {
      bApplyOutlineToDropShadows=false
      bMiteredCorners=false
      bSeparateFillAlpha=false
      outlineMaterial=null
      outlineSize=0
      outlineColor=<rgba 0 0 0 1>
     }
    }
    shadowOffset=<0 0>
   }
   slot {class=HorizontalBoxSlot, properties {padding=<ltrb 5 0 0 0>}}
  }
  SpawnLocationGameplayTag {
   widgetClassPath=&FortQuestToolPropertyView
   bIsVariable=true
   parent=HorizontalBox
   widget=SpawnLocationGameplayTag
   properties {bShowNameWidgets=false, propertiesToShow [ProviderFiltersTagQuery]}
   slot {class=HorizontalBoxSlot, properties {padding=<ltrb 5 0 0 0>}}
  }
  HorizontalBox_1 {widgetClassPath=&HorizontalBox, parent=VerticalBox_87, widget=HorizontalBox_1, slot {class=VerticalBoxSlot}}
  ConversationTagText {
   widgetClassPath=&CommonTextBlock
   parent=HorizontalBox_1
   widget=ConversationTagText
   properties {
    text=Conversation Tag
    colorAndOpacity {colorUseRule=UseColor_Specified, specifiedColor=<rgba 0.5156 0.5156 0.5156 1>}
    font {
     bForceMonospaced=false
     bMaterialIsStencil=false
     fontMaterial=null
     fontObject=&NotoSans
     letterSpacing=0
     monospacedWidth=1
     size=15
     skewAmount=0
     typefaceFontName=NotoSans-Italic
     outlineSettings {
      bApplyOutlineToDropShadows=false
      bMiteredCorners=false
      bSeparateFillAlpha=false
      outlineMaterial=null
      outlineSize=0
      outlineColor=<rgba 0 0 0 1>
     }
    }
    shadowOffset=<0 0>
   }
   slot {class=HorizontalBoxSlot}
  }
  ConversationComponentNotValid {
   widgetClassPath=&CommonTextBlock
   bIsVariable=true
   parent=HorizontalBox_1
   widget=ConversationComponentNotValid
   properties {
    text=Conversation component cannot be edited
    visibility=Collapsed
    colorAndOpacity {colorUseRule=UseColor_Specified, specifiedColor=<rgba 0.5156 0.1504 0.1504 1>}
    font {
     bForceMonospaced=false
     bMaterialIsStencil=false
     fontMaterial=null
     fontObject=&NotoSans
     letterSpacing=0
     monospacedWidth=1
     size=15
     skewAmount=0
     typefaceFontName=NotoSans-Italic
     outlineSettings {
      bApplyOutlineToDropShadows=false
      bMiteredCorners=false
      bSeparateFillAlpha=false
      outlineMaterial=null
      outlineSize=0
      outlineColor=<rgba 0 0 0 1>
     }
    }
    shadowOffset=<0 0>
   }
   slot {class=HorizontalBoxSlot, properties {padding=<ltrb 5 0 0 0>}}
  }
  ConversationGameplayTag {
   widgetClassPath=&FortQuestToolPropertyView
   bIsVariable=true
   parent=HorizontalBox_1
   widget=ConversationGameplayTag
   properties {bShowNameWidgets=false, propertiesToShow [ConversationEntryTag]}
   slot {class=HorizontalBoxSlot, properties {padding=<ltrb 5 0 0 0>}}
  }
  VerticalBox_223 {
   widgetClassPath=&VerticalBox
   parent=HorizontalBox_114
   widget=VerticalBox_223
   slot {class=HorizontalBoxSlot, properties {padding=<ltrb 5 0 0 0>}}
  }
  Spacer_179 {
   widgetClassPath=&Spacer
   parent=VerticalBox_223
   widget=Spacer_179
   slot {class=VerticalBoxSlot, properties {size {sizeRule=Fill, value=1}}}
  }
  DetailedReleaseVersionView {
   widgetClassPath=&DetailsView
   parent=VerticalBox_223
   widget=DetailedReleaseVersionView
   properties {bShowScrollBar=false, columnWidth=0.5, propertiesToShow [FortReleaseVersion]}
   slot {class=VerticalBoxSlot}
  }
  Spacer_506 {
   widgetClassPath=&Spacer
   parent=VerticalBox_223
   widget=Spacer_506
   slot {class=VerticalBoxSlot, properties {size {sizeRule=Fill, value=1}}}
  }
  ReleaseVersionView {
   widgetClassPath=&FortQuestToolPropertyView
   bIsVariable=true
   parent=HorizontalBox_114
   widget=ReleaseVersionView
   properties {visibility=Collapsed, propertiesToShow [FortReleaseVersion]}
   slot {class=HorizontalBoxSlot, properties {horizontalAlignment=HAlign_Center, verticalAlignment=VAlign_Center, padding=<ltrb 5 0 0 0>}}
  }
  Btn_BrowseToAsset {
   widgetClassPath=&TA_QuestEditor_Button
   bIsVariable=true
   parent=HorizontalBox_114
   widget=Btn_BrowseToAsset
   properties {label=Browse To Asset, buttonColor_Default=<rgba 0.0024 0.1195 0.4969 1>}
   slot {class=HorizontalBoxSlot, properties {padding=<ltrb 10 0 10 0>}}
  }
  HorizontalBox_123 {
   widgetClassPath=&HorizontalBox
   parent=VerticalBox_239
   widget=HorizontalBox_123
   slot {class=VerticalBoxSlot, properties {padding=<ltrb 0 15 0 0>, size {sizeRule=Fill, value=0.45}}}
  }
  ScrollBox_0 {
   widgetClassPath=&ScrollBox
   parent=HorizontalBox_123
   widget=ScrollBox_0
   properties {
    widgetBarStyle {
     thickness=8
     draggedThumbImage {
      bIsDynamicallyLoaded=false
      drawAs=RoundedBox
      imageType=NoImage
      mirroring=NoMirror
      resourceName=None
      resourceObject=null
      tiling=NoTile
      imageSize=<0 0>
      margin=<ltrb 0 0 0 0>
      outlineSettings {
       bUseBrushTransparency=false
       roundingType=FixedRadius
       width=0
       color {colorUseRule=UseColor_Specified, specifiedColor=<rgba 0 0 0 0>}
       cornerRadii=<4 4 4 4>
      }
      tintColor {colorUseRule=UseColor_Specified, specifiedColor=<rgba 0.2159 0.2159 0.2159 1>}
      uVRegion {bIsValid=false, max=<0 0>, min=<0 0>}
     }
     horizontalBackgroundImage {
      bIsDynamicallyLoaded=false
      drawAs=NoDrawType
      imageType=NoImage
      mirroring=NoMirror
      resourceName=None
      resourceObject=null
      tiling=NoTile
      imageSize=<0 0>
      margin=<ltrb 0 0 0 0>
      outlineSettings {
       bUseBrushTransparency=false
       roundingType=HalfHeightRadius
       width=0
       color {colorUseRule=UseColor_Specified, specifiedColor=<rgba 0 0 0 0>}
       cornerRadii=<0 0 0 0>
      }
      tintColor {colorUseRule=UseColor_Specified, specifiedColor=<rgba 1 1 1 1>}
      uVRegion {bIsValid=false, max=<0 0>, min=<0 0>}
     }
     horizontalBottomSlotImage {
      bIsDynamicallyLoaded=false
      drawAs=NoDrawType
      imageType=NoImage
      mirroring=NoMirror
      resourceName=None
      resourceObject=null
      tiling=NoTile
      imageSize=<0 0>
      margin=<ltrb 0 0 0 0>
      outlineSettings {
       bUseBrushTransparency=false
       roundingType=HalfHeightRadius
       width=0
       color {colorUseRule=UseColor_Specified, specifiedColor=<rgba 0 0 0 0>}
       cornerRadii=<0 0 0 0>
      }
      tintColor {colorUseRule=UseColor_Specified, specifiedColor=<rgba 1 1 1 1>}
      uVRegion {bIsValid=false, max=<0 0>, min=<0 0>}
     }
     horizontalTopSlotImage {
      bIsDynamicallyLoaded=false
      drawAs=NoDrawType
      imageType=NoImage
      mirroring=NoMirror
      resourceName=None
      resourceObject=null
      tiling=NoTile
      imageSize=<0 0>
      margin=<ltrb 0 0 0 0>
      outlineSettings {
       bUseBrushTransparency=false
       roundingType=HalfHeightRadius
       width=0
       color {colorUseRule=UseColor_Specified, specifiedColor=<rgba 0 0 0 0>}
       cornerRadii=<0 0 0 0>
      }
      tintColor {colorUseRule=UseColor_Specified, specifiedColor=<rgba 1 1 1 1>}
      uVRegion {bIsValid=false, max=<0 0>, min=<0 0>}
     }
     hoveredThumbImage {
      bIsDynamicallyLoaded=false
      drawAs=RoundedBox
      imageType=NoImage
      mirroring=NoMirror
      resourceName=None
      resourceObject=null
      tiling=NoTile
      imageSize=<0 0>
      margin=<ltrb 0 0 0 0>
      outlineSettings {
       bUseBrushTransparency=false
       roundingType=FixedRadius
       width=0
       color {colorUseRule=UseColor_Specified, specifiedColor=<rgba 0 0 0 0>}
       cornerRadii=<4 4 4 4>
      }
      tintColor {colorUseRule=UseColor_Specified, specifiedColor=<rgba 0.2159 0.2159 0.2159 1>}
      uVRegion {bIsValid=false, max=<0 0>, min=<0 0>}
     }
     normalThumbImage {
      bIsDynamicallyLoaded=false
      drawAs=RoundedBox
      imageType=NoImage
      mirroring=NoMirror
      resourceName=None
      resourceObject=null
      tiling=NoTile
      imageSize=<0 0>
      margin=<ltrb 0 0 0 0>
      outlineSettings {
       bUseBrushTransparency=false
       roundingType=FixedRadius
       width=0
       color {colorUseRule=UseColor_Specified, specifiedColor=<rgba 0 0 0 0>}
       cornerRadii=<4 4 4 4>
      }
      tintColor {colorUseRule=UseColor_Specified, specifiedColor=<rgba 0.0953 0.0953 0.0953 1>}
      uVRegion {bIsValid=false, max=<0 0>, min=<0 0>}
     }
     verticalBackgroundImage {
      bIsDynamicallyLoaded=false
      drawAs=NoDrawType
      imageType=NoImage
      mirroring=NoMirror
      resourceName=None
      resourceObject=null
      tiling=NoTile
      imageSize=<0 0>
      margin=<ltrb 0 0 0 0>
      outlineSettings {
       bUseBrushTransparency=false
       roundingType=HalfHeightRadius
       width=0
       color {colorUseRule=UseColor_Specified, specifiedColor=<rgba 0 0 0 0>}
       cornerRadii=<0 0 0 0>
      }
      tintColor {colorUseRule=UseColor_Specified, specifiedColor=<rgba 1 1 1 1>}
      uVRegion {bIsValid=false, max=<0 0>, min=<0 0>}
     }
     verticalBottomSlotImage {
      bIsDynamicallyLoaded=false
      drawAs=NoDrawType
      imageType=NoImage
      mirroring=NoMirror
      resourceName=None
      resourceObject=null
      tiling=NoTile
      imageSize=<0 0>
      margin=<ltrb 0 0 0 0>
      outlineSettings {
       bUseBrushTransparency=false
       roundingType=HalfHeightRadius
       width=0
       color {colorUseRule=UseColor_Specified, specifiedColor=<rgba 0 0 0 0>}
       cornerRadii=<0 0 0 0>
      }
      tintColor {colorUseRule=UseColor_Specified, specifiedColor=<rgba 1 1 1 1>}
      uVRegion {bIsValid=false, max=<0 0>, min=<0 0>}
     }
     verticalTopSlotImage {
      bIsDynamicallyLoaded=false
      drawAs=NoDrawType
      imageType=NoImage
      mirroring=NoMirror
      resourceName=None
      resourceObject=null
      tiling=NoTile
      imageSize=<0 0>
      margin=<ltrb 0 0 0 0>
      outlineSettings {
       bUseBrushTransparency=false
       roundingType=HalfHeightRadius
       width=0
       color {colorUseRule=UseColor_Specified, specifiedColor=<rgba 0 0 0 0>}
       cornerRadii=<0 0 0 0>
      }
      tintColor {colorUseRule=UseColor_Specified, specifiedColor=<rgba 1 1 1 1>}
      uVRegion {bIsValid=false, max=<0 0>, min=<0 0>}
     }
    }
    widgetStyle {
     barThickness=8
     bottomShadowBrush {
      bIsDynamicallyLoaded=false
      drawAs=Box
      imageType=FullColor
      mirroring=NoMirror
      resourceName=../../../Engine/Content/Slate/Common/ScrollBoxShadowBottom.png
      resourceObject=null
      tiling=NoTile
      imageSize=<16 8>
      margin=<ltrb 0.5 0 0.5 1>
      outlineSettings {
       bUseBrushTransparency=false
       roundingType=HalfHeightRadius
       width=0
       color {colorUseRule=UseColor_Specified, specifiedColor=<rgba 0 0 0 0>}
       cornerRadii=<0 0 0 0>
      }
      tintColor {colorUseRule=UseColor_Specified, specifiedColor=<rgba 1 1 1 1>}
      uVRegion {bIsValid=false, max=<0 0>, min=<0 0>}
     }
     horizontalScrolledContentPadding=<ltrb 0 0 1 0>
     leftShadowBrush {
      bIsDynamicallyLoaded=false
      drawAs=Box
      imageType=FullColor
      mirroring=NoMirror
      resourceName=../../../Engine/Content/Slate/Common/ScrollBoxShadowLeft.png
      resourceObject=null
      tiling=NoTile
      imageSize=<8 16>
      margin=<ltrb 1 0.5 0 0.5>
      outlineSettings {
       bUseBrushTransparency=false
       roundingType=HalfHeightRadius
       width=0
       color {colorUseRule=UseColor_Specified, specifiedColor=<rgba 0 0 0 0>}
       cornerRadii=<0 0 0 0>
      }
      tintColor {colorUseRule=UseColor_Specified, specifiedColor=<rgba 1 1 1 1>}
      uVRegion {bIsValid=false, max=<0 0>, min=<0 0>}
     }
     rightShadowBrush {
      bIsDynamicallyLoaded=false
      drawAs=Box
      imageType=FullColor
      mirroring=NoMirror
      resourceName=../../../Engine/Content/Slate/Common/ScrollBoxShadowRight.png
      resourceObject=null
      tiling=NoTile
      imageSize=<8 16>
      margin=<ltrb 0 0.5 1 0.5>
      outlineSettings {
       bUseBrushTransparency=false
       roundingType=HalfHeightRadius
       width=0
       color {colorUseRule=UseColor_Specified, specifiedColor=<rgba 0 0 0 0>}
       cornerRadii=<0 0 0 0>
      }
      tintColor {colorUseRule=UseColor_Specified, specifiedColor=<rgba 1 1 1 1>}
      uVRegion {bIsValid=false, max=<0 0>, min=<0 0>}
     }
     topShadowBrush {
      bIsDynamicallyLoaded=false
      drawAs=Box
      imageType=FullColor
      mirroring=NoMirror
      resourceName=../../../Engine/Content/Slate/Common/ScrollBoxShadowTop.png
      resourceObject=null
      tiling=NoTile
      imageSize=<16 8>
      margin=<ltrb 0.5 1 0.5 0>
      outlineSettings {
       bUseBrushTransparency=false
       roundingType=HalfHeightRadius
       width=0
       color {colorUseRule=UseColor_Specified, specifiedColor=<rgba 0 0 0 0>}
       cornerRadii=<0 0 0 0>
      }
      tintColor {colorUseRule=UseColor_Specified, specifiedColor=<rgba 1 1 1 1>}
      uVRegion {bIsValid=false, max=<0 0>, min=<0 0>}
     }
     verticalScrolledContentPadding=<ltrb 0 0 0 1>
    }
   }
   slot {class=HorizontalBoxSlot, properties {padding=<ltrb 0 5 0 0>, size {sizeRule=Fill, value=0.5}}}
  }
  ComponentListsView {
   widgetClassPath=&DetailsView
   parent=ScrollBox_0
   widget=ComponentListsView
   properties {propertiesToShow [
     InventoryComponent
     AffiliationComponent
     GameplayComponent
     ConversationComponent
     CosmeticComponent
     GameplayAbilityComponent
     SpawnParamsComponent
    ]}
   slot {class=ScrollBoxSlot}
  }
  VerticalBox_195 {widgetClassPath=&VerticalBox, parent=HorizontalBox_123, widget=VerticalBox_195, slot {class=HorizontalBoxSlot}}
  HorizontalBox_129 {widgetClassPath=&HorizontalBox, parent=VerticalBox_195, widget=HorizontalBox_129, slot {class=VerticalBoxSlot}}
  ComponentListShowAllTextbox {
   widgetClassPath=&CommonTextBlock
   parent=HorizontalBox_129
   widget=ComponentListShowAllTextbox
   properties {
    text=Show All
    colorAndOpacity {colorUseRule=UseColor_Specified, specifiedColor=<rgba 0.5156 0.5156 0.5156 1>}
    font {
     bForceMonospaced=false
     bMaterialIsStencil=false
     fontMaterial=null
     fontObject=&NotoSans
     letterSpacing=0
     monospacedWidth=1
     size=15
     skewAmount=0
     typefaceFontName=NotoSans-Italic
     outlineSettings {
      bApplyOutlineToDropShadows=false
      bMiteredCorners=false
      bSeparateFillAlpha=false
      outlineMaterial=null
      outlineSize=0
      outlineColor=<rgba 0 0 0 1>
     }
    }
    shadowOffset=<0 0>
   }
   slot {class=HorizontalBoxSlot, properties {padding=<ltrb 30 0 0 0>}}
  }
  ComponentListShowAllCheckbox {
   widgetClassPath=/Script/UMG.CheckBox
   bIsVariable=true
   parent=HorizontalBox_129
   widget=ComponentListShowAllCheckbox
   properties {widgetStyle {
     checkBoxType=CheckBox
     backgroundHoveredImage {
      bIsDynamicallyLoaded=false
      drawAs=RoundedBox
      imageType=NoImage
      mirroring=NoMirror
      resourceName=None
      resourceObject=null
      tiling=NoTile
      imageSize=<18 18>
      margin=<ltrb 0 0 0 0>
      outlineSettings {
       bUseBrushTransparency=false
       roundingType=FixedRadius
       width=1
       color {colorUseRule=UseColor_ColorTable, specifiedColor=<rgba 0 0 0 0>}
       cornerRadii=<3 3 3 3>
      }
      tintColor {colorUseRule=UseColor_Specified, specifiedColor=<rgba 0.0048 0.0048 0.0048 1>}
      uVRegion {bIsValid=false, max=<0 0>, min=<0 0>}
     }
     backgroundImage {
      bIsDynamicallyLoaded=false
      drawAs=RoundedBox
      imageType=NoImage
      mirroring=NoMirror
      resourceName=None
      resourceObject=null
      tiling=NoTile
      imageSize=<18 18>
      margin=<ltrb 0 0 0 0>
      outlineSettings {
       bUseBrushTransparency=false
       roundingType=FixedRadius
       width=1
       color {colorUseRule=UseColor_ColorTable, specifiedColor=<rgba 0 0 0 0>}
       cornerRadii=<3 3 3 3>
      }
      tintColor {colorUseRule=UseColor_Specified, specifiedColor=<rgba 0.0048 0.0048 0.0048 1>}
      uVRegion {bIsValid=false, max=<0 0>, min=<0 0>}
     }
     backgroundPressedImage {
      bIsDynamicallyLoaded=false
      drawAs=RoundedBox
      imageType=NoImage
      mirroring=NoMirror
      resourceName=None
      resourceObject=null
      tiling=NoTile
      imageSize=<18 18>
      margin=<ltrb 0 0 0 0>
      outlineSettings {
       bUseBrushTransparency=false
       roundingType=FixedRadius
       width=1
       color {colorUseRule=UseColor_ColorTable, specifiedColor=<rgba 0 0 0 0>}
       cornerRadii=<3 3 3 3>
      }
      tintColor {colorUseRule=UseColor_Specified, specifiedColor=<rgba 0.0048 0.0048 0.0048 1>}
      uVRegion {bIsValid=false, max=<0 0>, min=<0 0>}
     }
     borderBackgroundColor {colorUseRule=UseColor_Specified, specifiedColor=<rgba 1 1 1 1>}
     checkedForeground {colorUseRule=UseColor_Specified, specifiedColor=<rgba 1 1 1 1>}
     checkedHoveredForeground {colorUseRule=UseColor_Specified, specifiedColor=<rgba 1 1 1 1>}
     checkedHoveredImage {
      bIsDynamicallyLoaded=false
      drawAs=Image
      imageType=Vector
      mirroring=NoMirror
      resourceName=../../../Engine/Content/Slate/Starship/CoreWidgets/CheckBox/check.svg
      resourceObject=null
      tiling=NoTile
      imageSize=<16 16>
      margin=<ltrb 0 0 0 0>
      outlineSettings {
       bUseBrushTransparency=false
       roundingType=HalfHeightRadius
       width=0
       color {colorUseRule=UseColor_Specified, specifiedColor=<rgba 0 0 0 0>}
       cornerRadii=<0 0 0 0>
      }
      tintColor {colorUseRule=UseColor_Specified, specifiedColor=<rgba 0.0044 0.2384 1 1>}
      uVRegion {bIsValid=false, max=<0 0>, min=<0 0>}
     }
     checkedImage {
      bIsDynamicallyLoaded=false
      drawAs=Image
      imageType=Vector
      mirroring=NoMirror
      resourceName=../../../Engine/Content/Slate/Starship/CoreWidgets/CheckBox/check.svg
      resourceObject=null
      tiling=NoTile
      imageSize=<16 16>
      margin=<ltrb 0 0 0 0>
      outlineSettings {
       bUseBrushTransparency=false
       roundingType=HalfHeightRadius
       width=0
       color {colorUseRule=UseColor_Specified, specifiedColor=<rgba 0 0 0 0>}
       cornerRadii=<0 0 0 0>
      }
      tintColor {colorUseRule=UseColor_Specified, specifiedColor=<rgba 0 0.162 0.7454 1>}
      uVRegion {bIsValid=false, max=<0 0>, min=<0 0>}
     }
     checkedPressedForeground {colorUseRule=UseColor_Specified, specifiedColor=<rgba 1 1 1 1>}
     checkedPressedImage {
      bIsDynamicallyLoaded=false
      drawAs=Image
      imageType=Vector
      mirroring=NoMirror
      resourceName=../../../Engine/Content/Slate/Starship/CoreWidgets/CheckBox/check.svg
      resourceObject=null
      tiling=NoTile
      imageSize=<16 16>
      margin=<ltrb 0 0 0 0>
      outlineSettings {
       bUseBrushTransparency=false
       roundingType=HalfHeightRadius
       width=0
       color {colorUseRule=UseColor_Specified, specifiedColor=<rgba 0 0 0 0>}
       cornerRadii=<0 0 0 0>
      }
      tintColor {colorUseRule=UseColor_Specified, specifiedColor=<rgba 0 0.0802 0.3515 1>}
      uVRegion {bIsValid=false, max=<0 0>, min=<0 0>}
     }
     checkedSlateSound {resourceObject=null}
     foregroundColor {colorUseRule=UseColor_Specified, specifiedColor=<rgba 1 1 1 1>}
     hoveredForeground {colorUseRule=UseColor_Specified, specifiedColor=<rgba 1 1 1 1>}
     hoveredSlateSound {resourceObject=null}
     padding=<ltrb 2 0 0 0>
     pressedForeground {colorUseRule=UseColor_Specified, specifiedColor=<rgba 1 1 1 1>}
     uncheckedHoveredImage {
      bIsDynamicallyLoaded=false
      drawAs=NoDrawType
      imageType=NoImage
      mirroring=NoMirror
      resourceName=None
      resourceObject=null
      tiling=NoTile
      imageSize=<0 0>
      margin=<ltrb 0 0 0 0>
      outlineSettings {
       bUseBrushTransparency=false
       roundingType=HalfHeightRadius
       width=0
       color {colorUseRule=UseColor_Specified, specifiedColor=<rgba 0 0 0 0>}
       cornerRadii=<0 0 0 0>
      }
      tintColor {colorUseRule=UseColor_Specified, specifiedColor=<rgba 1 1 1 1>}
      uVRegion {bIsValid=false, max=<0 0>, min=<0 0>}
     }
     uncheckedImage {
      bIsDynamicallyLoaded=false
      drawAs=NoDrawType
      imageType=NoImage
      mirroring=NoMirror
      resourceName=None
      resourceObject=null
      tiling=NoTile
      imageSize=<0 0>
      margin=<ltrb 0 0 0 0>
      outlineSettings {
       bUseBrushTransparency=false
       roundingType=HalfHeightRadius
       width=0
       color {colorUseRule=UseColor_Specified, specifiedColor=<rgba 0 0 0 0>}
       cornerRadii=<0 0 0 0>
      }
      tintColor {colorUseRule=UseColor_Specified, specifiedColor=<rgba 1 1 1 1>}
      uVRegion {bIsValid=false, max=<0 0>, min=<0 0>}
     }
     uncheckedPressedImage {
      bIsDynamicallyLoaded=false
      drawAs=NoDrawType
      imageType=NoImage
      mirroring=NoMirror
      resourceName=None
      resourceObject=null
      tiling=NoTile
      imageSize=<0 0>
      margin=<ltrb 0 0 0 0>
      outlineSettings {
       bUseBrushTransparency=false
       roundingType=HalfHeightRadius
       width=0
       color {colorUseRule=UseColor_Specified, specifiedColor=<rgba 0 0 0 0>}
       cornerRadii=<0 0 0 0>
      }
      tintColor {colorUseRule=UseColor_Specified, specifiedColor=<rgba 1 1 1 1>}
      uVRegion {bIsValid=false, max=<0 0>, min=<0 0>}
     }
     uncheckedSlateSound {resourceObject=null}
     undeterminedForeground {colorUseRule=UseColor_Specified, specifiedColor=<rgba 1 1 1 1>}
     undeterminedHoveredImage {
      bIsDynamicallyLoaded=false
      drawAs=Image
      imageType=Vector
      mirroring=NoMirror
      resourceName=../../../Engine/Content/Slate/Starship/CoreWidgets/CheckBox/indeterminate.svg
      resourceObject=null
      tiling=NoTile
      imageSize=<16 16>
      margin=<ltrb 0 0 0 0>
      outlineSettings {
       bUseBrushTransparency=false
       roundingType=HalfHeightRadius
       width=0
       color {colorUseRule=UseColor_Specified, specifiedColor=<rgba 0 0 0 0>}
       cornerRadii=<0 0 0 0>
      }
      tintColor {colorUseRule=UseColor_Specified, specifiedColor=<rgba 0.0044 0.2384 1 1>}
      uVRegion {bIsValid=false, max=<0 0>, min=<0 0>}
     }
     undeterminedImage {
      bIsDynamicallyLoaded=false
      drawAs=Image
      imageType=Vector
      mirroring=NoMirror
      resourceName=../../../Engine/Content/Slate/Starship/CoreWidgets/CheckBox/indeterminate.svg
      resourceObject=null
      tiling=NoTile
      imageSize=<16 16>
      margin=<ltrb 0 0 0 0>
      outlineSettings {
       bUseBrushTransparency=false
       roundingType=HalfHeightRadius
       width=0
       color {colorUseRule=UseColor_Specified, specifiedColor=<rgba 0 0 0 0>}
       cornerRadii=<0 0 0 0>
      }
      tintColor {colorUseRule=UseColor_Specified, specifiedColor=<rgba 0 0.162 0.7454 1>}
      uVRegion {bIsValid=false, max=<0 0>, min=<0 0>}
     }
     undeterminedPressedImage {
      bIsDynamicallyLoaded=false
      drawAs=Image
      imageType=Vector
      mirroring=NoMirror
      resourceName=../../../Engine/Content/Slate/Starship/CoreWidgets/CheckBox/indeterminate.svg
      resourceObject=null
      tiling=NoTile
      imageSize=<16 16>
      margin=<ltrb 0 0 0 0>
      outlineSettings {
       bUseBrushTransparency=false
       roundingType=HalfHeightRadius
       width=0
       color {colorUseRule=UseColor_Specified, specifiedColor=<rgba 0 0 0 0>}
       cornerRadii=<0 0 0 0>
      }
      tintColor {colorUseRule=UseColor_Specified, specifiedColor=<rgba 0 0.0802 0.3515 1>}
      uVRegion {bIsValid=false, max=<0 0>, min=<0 0>}
     }
    }}
   slot {class=HorizontalBoxSlot, properties {padding=<ltrb 10 0 0 0>}}
  }
  Btn_Inventory {
   widgetClassPath=&TA_NPCEditor_Button
   bIsVariable=true
   parent=VerticalBox_195
   widget=Btn_Inventory
   properties {label=Inventory}
   slot {class=VerticalBoxSlot, properties {padding=<ltrb 0 0 10 0>}}
  }
  Btn_SpawnParams {
   widgetClassPath=&TA_NPCEditor_Button
   bIsVariable=true
   parent=VerticalBox_195
   widget=Btn_SpawnParams
   properties {label=SpawnParams}
   slot {class=VerticalBoxSlot, properties {padding=<ltrb 0 0 10 0>}}
  }
  Btn_Affiliation {
   widgetClassPath=&TA_NPCEditor_Button
   bIsVariable=true
   parent=VerticalBox_195
   widget=Btn_Affiliation
   properties {label=Affiliation}
   slot {class=VerticalBoxSlot, properties {padding=<ltrb 0 0 10 0>}}
  }
  Btn_Gameplay {
   widgetClassPath=&TA_NPCEditor_Button
   bIsVariable=true
   parent=VerticalBox_195
   widget=Btn_Gameplay
   properties {label=Gameplay}
   slot {class=VerticalBoxSlot, properties {padding=<ltrb 0 0 10 0>}}
  }
  Btn_Conversation {
   widgetClassPath=&TA_NPCEditor_Button
   bIsVariable=true
   parent=VerticalBox_195
   widget=Btn_Conversation
   properties {label=Conversation}
   slot {class=VerticalBoxSlot, properties {padding=<ltrb 0 0 10 0>}}
  }
  Btn_Cosmetic {
   widgetClassPath=&TA_NPCEditor_Button
   bIsVariable=true
   parent=VerticalBox_195
   widget=Btn_Cosmetic
   properties {label=Cosmetic}
   slot {class=VerticalBoxSlot, properties {padding=<ltrb 0 0 10 0>}}
  }
  Btn_GameplayAbility {
   widgetClassPath=&TA_NPCEditor_Button
   bIsVariable=true
   parent=VerticalBox_195
   widget=Btn_GameplayAbility
   properties {label=GameplayAbility}
   slot {class=VerticalBoxSlot, properties {padding=<ltrb 0 0 10 0>}}
  }
  Btn_NPCCharacterData {
   widgetClassPath=&TA_NPCEditor_Button
   bIsVariable=true
   parent=VerticalBox_195
   widget=Btn_NPCCharacterData
   properties {label=NPCCharacterData}
   slot {class=VerticalBoxSlot, properties {padding=<ltrb 0 10 10 0>}}
  }
  Btn_LagerEventData {
   widgetClassPath=&TA_NPCEditor_Button
   bIsVariable=true
   parent=VerticalBox_195
   widget=Btn_LagerEventData
   properties {label=LagerEventData}
   slot {class=VerticalBoxSlot, properties {padding=<ltrb 0 0 10 0>}}
  }
  Btn_OpenNPCLibraryConversation {
   widgetClassPath=&TA_NPCEditor_Button
   bIsVariable=true
   parent=VerticalBox_195
   widget=Btn_OpenNPCLibraryConversation
   properties {label=Open NPCLibrary_Conversation, buttonColor_Default=<rgba 0.0024 0.1195 0.4969 1>}
   slot {class=VerticalBoxSlot, properties {padding=<ltrb 0 10 10 0>}}
  }
  HorizontalBox_238 {
   widgetClassPath=&HorizontalBox
   parent=VerticalBox_239
   widget=HorizontalBox_238
   slot {class=VerticalBoxSlot, properties {padding=<ltrb 0 15 0 0>}}
  }
  AllowEditTextBox {
   widgetClassPath=&CommonTextBlock
   parent=HorizontalBox_238
   widget=AllowEditTextBox
   properties {
    text=Allow Edit/Open for Edit
    visibility=Collapsed
    colorAndOpacity {colorUseRule=UseColor_Specified, specifiedColor=<rgba 0.5156 0.5156 0.5156 1>}
    font {
     bForceMonospaced=false
     bMaterialIsStencil=false
     fontMaterial=null
     fontObject=&NotoSans
     letterSpacing=0
     monospacedWidth=1
     size=15
     skewAmount=0
     typefaceFontName=NotoSans-Italic
     outlineSettings {
      bApplyOutlineToDropShadows=false
      bMiteredCorners=false
      bSeparateFillAlpha=false
      outlineMaterial=null
      outlineSize=0
      outlineColor=<rgba 0 0 0 1>
     }
    }
    shadowOffset=<0 0>
   }
   slot {class=HorizontalBoxSlot, properties {padding=<ltrb 30 0 0 0>}}
  }
  AllowEditCheckbox {
   widgetClassPath=&CheckBox
   bIsVariable=true
   parent=HorizontalBox_238
   widget=AllowEditCheckbox
   properties {
    visibility=Collapsed
    widgetStyle {
     checkBoxType=CheckBox
     backgroundHoveredImage {
      bIsDynamicallyLoaded=false
      drawAs=RoundedBox
      imageType=NoImage
      mirroring=NoMirror
      resourceName=None
      resourceObject=null
      tiling=NoTile
      imageSize=<18 18>
      margin=<ltrb 0 0 0 0>
      outlineSettings {
       bUseBrushTransparency=false
       roundingType=FixedRadius
       width=1
       color {colorUseRule=UseColor_ColorTable, specifiedColor=<rgba 0 0 0 0>}
       cornerRadii=<3 3 3 3>
      }
      tintColor {colorUseRule=UseColor_Specified, specifiedColor=<rgba 0.0048 0.0048 0.0048 1>}
      uVRegion {bIsValid=false, max=<0 0>, min=<0 0>}
     }
     backgroundImage {
      bIsDynamicallyLoaded=false
      drawAs=RoundedBox
      imageType=NoImage
      mirroring=NoMirror
      resourceName=None
      resourceObject=null
      tiling=NoTile
      imageSize=<18 18>
      margin=<ltrb 0 0 0 0>
      outlineSettings {
       bUseBrushTransparency=false
       roundingType=FixedRadius
       width=1
       color {colorUseRule=UseColor_ColorTable, specifiedColor=<rgba 0 0 0 0>}
       cornerRadii=<3 3 3 3>
      }
      tintColor {colorUseRule=UseColor_Specified, specifiedColor=<rgba 0.0048 0.0048 0.0048 1>}
      uVRegion {bIsValid=false, max=<0 0>, min=<0 0>}
     }
     backgroundPressedImage {
      bIsDynamicallyLoaded=false
      drawAs=RoundedBox
      imageType=NoImage
      mirroring=NoMirror
      resourceName=None
      resourceObject=null
      tiling=NoTile
      imageSize=<18 18>
      margin=<ltrb 0 0 0 0>
      outlineSettings {
       bUseBrushTransparency=false
       roundingType=FixedRadius
       width=1
       color {colorUseRule=UseColor_ColorTable, specifiedColor=<rgba 0 0 0 0>}
       cornerRadii=<3 3 3 3>
      }
      tintColor {colorUseRule=UseColor_Specified, specifiedColor=<rgba 0.0048 0.0048 0.0048 1>}
      uVRegion {bIsValid=false, max=<0 0>, min=<0 0>}
     }
     borderBackgroundColor {colorUseRule=UseColor_Specified, specifiedColor=<rgba 1 1 1 1>}
     checkedForeground {colorUseRule=UseColor_Specified, specifiedColor=<rgba 1 1 1 1>}
     checkedHoveredForeground {colorUseRule=UseColor_Specified, specifiedColor=<rgba 1 1 1 1>}
     checkedHoveredImage {
      bIsDynamicallyLoaded=false
      drawAs=Image
      imageType=Vector
      mirroring=NoMirror
      resourceName=../../../Engine/Content/Slate/Starship/CoreWidgets/CheckBox/check.svg
      resourceObject=null
      tiling=NoTile
      imageSize=<16 16>
      margin=<ltrb 0 0 0 0>
      outlineSettings {
       bUseBrushTransparency=false
       roundingType=HalfHeightRadius
       width=0
       color {colorUseRule=UseColor_Specified, specifiedColor=<rgba 0 0 0 0>}
       cornerRadii=<0 0 0 0>
      }
      tintColor {colorUseRule=UseColor_Specified, specifiedColor=<rgba 0.0044 0.2384 1 1>}
      uVRegion {bIsValid=false, max=<0 0>, min=<0 0>}
     }
     checkedImage {
      bIsDynamicallyLoaded=false
      drawAs=Image
      imageType=Vector
      mirroring=NoMirror
      resourceName=../../../Engine/Content/Slate/Starship/CoreWidgets/CheckBox/check.svg
      resourceObject=null
      tiling=NoTile
      imageSize=<16 16>
      margin=<ltrb 0 0 0 0>
      outlineSettings {
       bUseBrushTransparency=false
       roundingType=HalfHeightRadius
       width=0
       color {colorUseRule=UseColor_Specified, specifiedColor=<rgba 0 0 0 0>}
       cornerRadii=<0 0 0 0>
      }
      tintColor {colorUseRule=UseColor_Specified, specifiedColor=<rgba 0 0.162 0.7454 1>}
      uVRegion {bIsValid=false, max=<0 0>, min=<0 0>}
     }
     checkedPressedForeground {colorUseRule=UseColor_Specified, specifiedColor=<rgba 1 1 1 1>}
     checkedPressedImage {
      bIsDynamicallyLoaded=false
      drawAs=Image
      imageType=Vector
      mirroring=NoMirror
      resourceName=../../../Engine/Content/Slate/Starship/CoreWidgets/CheckBox/check.svg
      resourceObject=null
      tiling=NoTile
      imageSize=<16 16>
      margin=<ltrb 0 0 0 0>
      outlineSettings {
       bUseBrushTransparency=false
       roundingType=HalfHeightRadius
       width=0
       color {colorUseRule=UseColor_Specified, specifiedColor=<rgba 0 0 0 0>}
       cornerRadii=<0 0 0 0>
      }
      tintColor {colorUseRule=UseColor_Specified, specifiedColor=<rgba 0 0.0802 0.3515 1>}
      uVRegion {bIsValid=false, max=<0 0>, min=<0 0>}
     }
     checkedSlateSound {resourceObject=null}
     foregroundColor {colorUseRule=UseColor_Specified, specifiedColor=<rgba 1 1 1 1>}
     hoveredForeground {colorUseRule=UseColor_Specified, specifiedColor=<rgba 1 1 1 1>}
     hoveredSlateSound {resourceObject=null}
     padding=<ltrb 2 0 0 0>
     pressedForeground {colorUseRule=UseColor_Specified, specifiedColor=<rgba 1 1 1 1>}
     uncheckedHoveredImage {
      bIsDynamicallyLoaded=false
      drawAs=NoDrawType
      imageType=NoImage
      mirroring=NoMirror
      resourceName=None
      resourceObject=null
      tiling=NoTile
      imageSize=<0 0>
      margin=<ltrb 0 0 0 0>
      outlineSettings {
       bUseBrushTransparency=false
       roundingType=HalfHeightRadius
       width=0
       color {colorUseRule=UseColor_Specified, specifiedColor=<rgba 0 0 0 0>}
       cornerRadii=<0 0 0 0>
      }
      tintColor {colorUseRule=UseColor_Specified, specifiedColor=<rgba 1 1 1 1>}
      uVRegion {bIsValid=false, max=<0 0>, min=<0 0>}
     }
     uncheckedImage {
      bIsDynamicallyLoaded=false
      drawAs=NoDrawType
      imageType=NoImage
      mirroring=NoMirror
      resourceName=None
      resourceObject=null
      tiling=NoTile
      imageSize=<0 0>
      margin=<ltrb 0 0 0 0>
      outlineSettings {
       bUseBrushTransparency=false
       roundingType=HalfHeightRadius
       width=0
       color {colorUseRule=UseColor_Specified, specifiedColor=<rgba 0 0 0 0>}
       cornerRadii=<0 0 0 0>
      }
      tintColor {colorUseRule=UseColor_Specified, specifiedColor=<rgba 1 1 1 1>}
      uVRegion {bIsValid=false, max=<0 0>, min=<0 0>}
     }
     uncheckedPressedImage {
      bIsDynamicallyLoaded=false
      drawAs=NoDrawType
      imageType=NoImage
      mirroring=NoMirror
      resourceName=None
      resourceObject=null
      tiling=NoTile
      imageSize=<0 0>
      margin=<ltrb 0 0 0 0>
      outlineSettings {
       bUseBrushTransparency=false
       roundingType=HalfHeightRadius
       width=0
       color {colorUseRule=UseColor_Specified, specifiedColor=<rgba 0 0 0 0>}
       cornerRadii=<0 0 0 0>
      }
      tintColor {colorUseRule=UseColor_Specified, specifiedColor=<rgba 1 1 1 1>}
      uVRegion {bIsValid=false, max=<0 0>, min=<0 0>}
     }
     uncheckedSlateSound {resourceObject=null}
     undeterminedForeground {colorUseRule=UseColor_Specified, specifiedColor=<rgba 1 1 1 1>}
     undeterminedHoveredImage {
      bIsDynamicallyLoaded=false
      drawAs=Image
      imageType=Vector
      mirroring=NoMirror
      resourceName=../../../Engine/Content/Slate/Starship/CoreWidgets/CheckBox/indeterminate.svg
      resourceObject=null
      tiling=NoTile
      imageSize=<16 16>
      margin=<ltrb 0 0 0 0>
      outlineSettings {
       bUseBrushTransparency=false
       roundingType=HalfHeightRadius
       width=0
       color {colorUseRule=UseColor_Specified, specifiedColor=<rgba 0 0 0 0>}
       cornerRadii=<0 0 0 0>
      }
      tintColor {colorUseRule=UseColor_Specified, specifiedColor=<rgba 0.0044 0.2384 1 1>}
      uVRegion {bIsValid=false, max=<0 0>, min=<0 0>}
     }
     undeterminedImage {
      bIsDynamicallyLoaded=false
      drawAs=Image
      imageType=Vector
      mirroring=NoMirror
      resourceName=../../../Engine/Content/Slate/Starship/CoreWidgets/CheckBox/indeterminate.svg
      resourceObject=null
      tiling=NoTile
      imageSize=<16 16>
      margin=<ltrb 0 0 0 0>
      outlineSettings {
       bUseBrushTransparency=false
       roundingType=HalfHeightRadius
       width=0
       color {colorUseRule=UseColor_Specified, specifiedColor=<rgba 0 0 0 0>}
       cornerRadii=<0 0 0 0>
      }
      tintColor {colorUseRule=UseColor_Specified, specifiedColor=<rgba 0 0.162 0.7454 1>}
      uVRegion {bIsValid=false, max=<0 0>, min=<0 0>}
     }
     undeterminedPressedImage {
      bIsDynamicallyLoaded=false
      drawAs=Image
      imageType=Vector
      mirroring=NoMirror
      resourceName=../../../Engine/Content/Slate/Starship/CoreWidgets/CheckBox/indeterminate.svg
      resourceObject=null
      tiling=NoTile
      imageSize=<16 16>
      margin=<ltrb 0 0 0 0>
      outlineSettings {
       bUseBrushTransparency=false
       roundingType=HalfHeightRadius
       width=0
       color {colorUseRule=UseColor_Specified, specifiedColor=<rgba 0 0 0 0>}
       cornerRadii=<0 0 0 0>
      }
      tintColor {colorUseRule=UseColor_Specified, specifiedColor=<rgba 0 0.0802 0.3515 1>}
      uVRegion {bIsValid=false, max=<0 0>, min=<0 0>}
     }
    }
   }
   slot {class=HorizontalBoxSlot, properties {padding=<ltrb 10 0 0 0>}}
  }
  ShowAllTextBox {
   widgetClassPath=&CommonTextBlock
   parent=HorizontalBox_238
   widget=ShowAllTextBox
   properties {
    text=Show All
    colorAndOpacity {colorUseRule=UseColor_Specified, specifiedColor=<rgba 0.5156 0.5156 0.5156 1>}
    font {
     bForceMonospaced=false
     bMaterialIsStencil=false
     fontMaterial=null
     fontObject=&NotoSans
     letterSpacing=0
     monospacedWidth=1
     size=15
     skewAmount=0
     typefaceFontName=NotoSans-Italic
     outlineSettings {
      bApplyOutlineToDropShadows=false
      bMiteredCorners=false
      bSeparateFillAlpha=false
      outlineMaterial=null
      outlineSize=0
      outlineColor=<rgba 0 0 0 1>
     }
    }
    shadowOffset=<0 0>
   }
   slot {class=HorizontalBoxSlot, properties {padding=<ltrb 30 0 0 0>}}
  }
  ShowAllCheckbox {
   widgetClassPath=&CheckBox
   bIsVariable=true
   parent=HorizontalBox_238
   widget=ShowAllCheckbox
   properties {widgetStyle {
     checkBoxType=CheckBox
     backgroundHoveredImage {
      bIsDynamicallyLoaded=false
      drawAs=RoundedBox
      imageType=NoImage
      mirroring=NoMirror
      resourceName=None
      resourceObject=null
      tiling=NoTile
      imageSize=<18 18>
      margin=<ltrb 0 0 0 0>
      outlineSettings {
       bUseBrushTransparency=false
       roundingType=FixedRadius
       width=1
       color {colorUseRule=UseColor_ColorTable, specifiedColor=<rgba 0 0 0 0>}
       cornerRadii=<3 3 3 3>
      }
      tintColor {colorUseRule=UseColor_Specified, specifiedColor=<rgba 0.0048 0.0048 0.0048 1>}
      uVRegion {bIsValid=false, max=<0 0>, min=<0 0>}
     }
     backgroundImage {
      bIsDynamicallyLoaded=false
      drawAs=RoundedBox
      imageType=NoImage
      mirroring=NoMirror
      resourceName=None
      resourceObject=null
      tiling=NoTile
      imageSize=<18 18>
      margin=<ltrb 0 0 0 0>
      outlineSettings {
       bUseBrushTransparency=false
       roundingType=FixedRadius
       width=1
       color {colorUseRule=UseColor_ColorTable, specifiedColor=<rgba 0 0 0 0>}
       cornerRadii=<3 3 3 3>
      }
      tintColor {colorUseRule=UseColor_Specified, specifiedColor=<rgba 0.0048 0.0048 0.0048 1>}
      uVRegion {bIsValid=false, max=<0 0>, min=<0 0>}
     }
     backgroundPressedImage {
      bIsDynamicallyLoaded=false
      drawAs=RoundedBox
      imageType=NoImage
      mirroring=NoMirror
      resourceName=None
      resourceObject=null
      tiling=NoTile
      imageSize=<18 18>
      margin=<ltrb 0 0 0 0>
      outlineSettings {
       bUseBrushTransparency=false
       roundingType=FixedRadius
       width=1
       color {colorUseRule=UseColor_ColorTable, specifiedColor=<rgba 0 0 0 0>}
       cornerRadii=<3 3 3 3>
      }
      tintColor {colorUseRule=UseColor_Specified, specifiedColor=<rgba 0.0048 0.0048 0.0048 1>}
      uVRegion {bIsValid=false, max=<0 0>, min=<0 0>}
     }
     borderBackgroundColor {colorUseRule=UseColor_Specified, specifiedColor=<rgba 1 1 1 1>}
     checkedForeground {colorUseRule=UseColor_Specified, specifiedColor=<rgba 1 1 1 1>}
     checkedHoveredForeground {colorUseRule=UseColor_Specified, specifiedColor=<rgba 1 1 1 1>}
     checkedHoveredImage {
      bIsDynamicallyLoaded=false
      drawAs=Image
      imageType=Vector
      mirroring=NoMirror
      resourceName=../../../Engine/Content/Slate/Starship/CoreWidgets/CheckBox/check.svg
      resourceObject=null
      tiling=NoTile
      imageSize=<16 16>
      margin=<ltrb 0 0 0 0>
      outlineSettings {
       bUseBrushTransparency=false
       roundingType=HalfHeightRadius
       width=0
       color {colorUseRule=UseColor_Specified, specifiedColor=<rgba 0 0 0 0>}
       cornerRadii=<0 0 0 0>
      }
      tintColor {colorUseRule=UseColor_Specified, specifiedColor=<rgba 0.0044 0.2384 1 1>}
      uVRegion {bIsValid=false, max=<0 0>, min=<0 0>}
     }
     checkedImage {
      bIsDynamicallyLoaded=false
      drawAs=Image
      imageType=Vector
      mirroring=NoMirror
      resourceName=../../../Engine/Content/Slate/Starship/CoreWidgets/CheckBox/check.svg
      resourceObject=null
      tiling=NoTile
      imageSize=<16 16>
      margin=<ltrb 0 0 0 0>
      outlineSettings {
       bUseBrushTransparency=false
       roundingType=HalfHeightRadius
       width=0
       color {colorUseRule=UseColor_Specified, specifiedColor=<rgba 0 0 0 0>}
       cornerRadii=<0 0 0 0>
      }
      tintColor {colorUseRule=UseColor_Specified, specifiedColor=<rgba 0 0.162 0.7454 1>}
      uVRegion {bIsValid=false, max=<0 0>, min=<0 0>}
     }
     checkedPressedForeground {colorUseRule=UseColor_Specified, specifiedColor=<rgba 1 1 1 1>}
     checkedPressedImage {
      bIsDynamicallyLoaded=false
      drawAs=Image
      imageType=Vector
      mirroring=NoMirror
      resourceName=../../../Engine/Content/Slate/Starship/CoreWidgets/CheckBox/check.svg
      resourceObject=null
      tiling=NoTile
      imageSize=<16 16>
      margin=<ltrb 0 0 0 0>
      outlineSettings {
       bUseBrushTransparency=false
       roundingType=HalfHeightRadius
       width=0
       color {colorUseRule=UseColor_Specified, specifiedColor=<rgba 0 0 0 0>}
       cornerRadii=<0 0 0 0>
      }
      tintColor {colorUseRule=UseColor_Specified, specifiedColor=<rgba 0 0.0802 0.3515 1>}
      uVRegion {bIsValid=false, max=<0 0>, min=<0 0>}
     }
     checkedSlateSound {resourceObject=null}
     foregroundColor {colorUseRule=UseColor_Specified, specifiedColor=<rgba 1 1 1 1>}
     hoveredForeground {colorUseRule=UseColor_Specified, specifiedColor=<rgba 1 1 1 1>}
     hoveredSlateSound {resourceObject=null}
     padding=<ltrb 2 0 0 0>
     pressedForeground {colorUseRule=UseColor_Specified, specifiedColor=<rgba 1 1 1 1>}
     uncheckedHoveredImage {
      bIsDynamicallyLoaded=false
      drawAs=NoDrawType
      imageType=NoImage
      mirroring=NoMirror
      resourceName=None
      resourceObject=null
      tiling=NoTile
      imageSize=<0 0>
      margin=<ltrb 0 0 0 0>
      outlineSettings {
       bUseBrushTransparency=false
       roundingType=HalfHeightRadius
       width=0
       color {colorUseRule=UseColor_Specified, specifiedColor=<rgba 0 0 0 0>}
       cornerRadii=<0 0 0 0>
      }
      tintColor {colorUseRule=UseColor_Specified, specifiedColor=<rgba 1 1 1 1>}
      uVRegion {bIsValid=false, max=<0 0>, min=<0 0>}
     }
     uncheckedImage {
      bIsDynamicallyLoaded=false
      drawAs=NoDrawType
      imageType=NoImage
      mirroring=NoMirror
      resourceName=None
      resourceObject=null
      tiling=NoTile
      imageSize=<0 0>
      margin=<ltrb 0 0 0 0>
      outlineSettings {
       bUseBrushTransparency=false
       roundingType=HalfHeightRadius
       width=0
       color {colorUseRule=UseColor_Specified, specifiedColor=<rgba 0 0 0 0>}
       cornerRadii=<0 0 0 0>
      }
      tintColor {colorUseRule=UseColor_Specified, specifiedColor=<rgba 1 1 1 1>}
      uVRegion {bIsValid=false, max=<0 0>, min=<0 0>}
     }
     uncheckedPressedImage {
      bIsDynamicallyLoaded=false
      drawAs=NoDrawType
      imageType=NoImage
      mirroring=NoMirror
      resourceName=None
      resourceObject=null
      tiling=NoTile
      imageSize=<0 0>
      margin=<ltrb 0 0 0 0>
      outlineSettings {
       bUseBrushTransparency=false
       roundingType=HalfHeightRadius
       width=0
       color {colorUseRule=UseColor_Specified, specifiedColor=<rgba 0 0 0 0>}
       cornerRadii=<0 0 0 0>
      }
      tintColor {colorUseRule=UseColor_Specified, specifiedColor=<rgba 1 1 1 1>}
      uVRegion {bIsValid=false, max=<0 0>, min=<0 0>}
     }
     uncheckedSlateSound {resourceObject=null}
     undeterminedForeground {colorUseRule=UseColor_Specified, specifiedColor=<rgba 1 1 1 1>}
     undeterminedHoveredImage {
      bIsDynamicallyLoaded=false
      drawAs=Image
      imageType=Vector
      mirroring=NoMirror
      resourceName=../../../Engine/Content/Slate/Starship/CoreWidgets/CheckBox/indeterminate.svg
      resourceObject=null
      tiling=NoTile
      imageSize=<16 16>
      margin=<ltrb 0 0 0 0>
      outlineSettings {
       bUseBrushTransparency=false
       roundingType=HalfHeightRadius
       width=0
       color {colorUseRule=UseColor_Specified, specifiedColor=<rgba 0 0 0 0>}
       cornerRadii=<0 0 0 0>
      }
      tintColor {colorUseRule=UseColor_Specified, specifiedColor=<rgba 0.0044 0.2384 1 1>}
      uVRegion {bIsValid=false, max=<0 0>, min=<0 0>}
     }
     undeterminedImage {
      bIsDynamicallyLoaded=false
      drawAs=Image
      imageType=Vector
      mirroring=NoMirror
      resourceName=../../../Engine/Content/Slate/Starship/CoreWidgets/CheckBox/indeterminate.svg
      resourceObject=null
      tiling=NoTile
      imageSize=<16 16>
      margin=<ltrb 0 0 0 0>
      outlineSettings {
       bUseBrushTransparency=false
       roundingType=HalfHeightRadius
       width=0
       color {colorUseRule=UseColor_Specified, specifiedColor=<rgba 0 0 0 0>}
       cornerRadii=<0 0 0 0>
      }
      tintColor {colorUseRule=UseColor_Specified, specifiedColor=<rgba 0 0.162 0.7454 1>}
      uVRegion {bIsValid=false, max=<0 0>, min=<0 0>}
     }
     undeterminedPressedImage {
      bIsDynamicallyLoaded=false
      drawAs=Image
      imageType=Vector
      mirroring=NoMirror
      resourceName=../../../Engine/Content/Slate/Starship/CoreWidgets/CheckBox/indeterminate.svg
      resourceObject=null
      tiling=NoTile
      imageSize=<16 16>
      margin=<ltrb 0 0 0 0>
      outlineSettings {
       bUseBrushTransparency=false
       roundingType=HalfHeightRadius
       width=0
       color {colorUseRule=UseColor_Specified, specifiedColor=<rgba 0 0 0 0>}
       cornerRadii=<0 0 0 0>
      }
      tintColor {colorUseRule=UseColor_Specified, specifiedColor=<rgba 0 0.0802 0.3515 1>}
      uVRegion {bIsValid=false, max=<0 0>, min=<0 0>}
     }
    }}
   slot {class=HorizontalBoxSlot, properties {padding=<ltrb 10 0 0 0>}}
  }
  ActiveComponentName {
   widgetClassPath=&CommonTextBlock
   bIsVariable=true
   parent=HorizontalBox_238
   widget=ActiveComponentName
   properties {
    text=No component active
    colorAndOpacity {colorUseRule=UseColor_Specified, specifiedColor=<rgba 0.9734 0.9734 0.9734 1>}
    font {
     bForceMonospaced=false
     bMaterialIsStencil=false
     fontMaterial=null
     fontObject=&NotoSans
     letterSpacing=0
     monospacedWidth=1
     size=15
     skewAmount=0
     typefaceFontName=NotoSans-Regular
     outlineSettings {
      bApplyOutlineToDropShadows=false
      bMiteredCorners=false
      bSeparateFillAlpha=false
      outlineMaterial=null
      outlineSize=0
      outlineColor=<rgba 0 0 0 1>
     }
    }
    shadowOffset=<0 0>
   }
   slot {class=HorizontalBoxSlot, properties {padding=<ltrb 45 0 0 0>}}
  }
  LagerEventDataSelector {
   widgetClassPath=/Script/FortniteEditor.SearchableStringComboBox
   bIsVariable=true
   parent=HorizontalBox_238
   widget=LagerEventDataSelector
   properties {
    visibility=Collapsed
    foregroundColor {colorUseRule=UseColor_Specified, specifiedColor=<rgba 0.5271 0.5271 0.5271 1>}
    itemStyle {
     bUseParentRowBrush=false
     activeBrush {
      bIsDynamicallyLoaded=false
      drawAs=Image
      imageType=NoImage
      mirroring=NoMirror
      resourceName=None
      resourceObject=null
      tiling=NoTile
      imageSize=<0 0>
      margin=<ltrb 0 0 0 0>
      outlineSettings {
       bUseBrushTransparency=false
       roundingType=HalfHeightRadius
       width=0
       color {colorUseRule=UseColor_Specified, specifiedColor=<rgba 0 0 0 0>}
       cornerRadii=<0 0 0 0>
      }
      tintColor {colorUseRule=UseColor_Specified, specifiedColor=<rgba 0 0.162 0.7454 1>}
      uVRegion {bIsValid=false, max=<0 0>, min=<0 0>}
     }
     activeHighlightedBrush {
      bIsDynamicallyLoaded=false
      drawAs=Image
      imageType=NoImage
      mirroring=NoMirror
      resourceName=None
      resourceObject=null
      tiling=NoTile
      imageSize=<0 0>
      margin=<ltrb 0 0 0 0>
      outlineSettings {
       bUseBrushTransparency=false
       roundingType=HalfHeightRadius
       width=0
       color {colorUseRule=UseColor_Specified, specifiedColor=<rgba 0 0 0 0>}
       cornerRadii=<0 0 0 0>
      }
      tintColor {colorUseRule=UseColor_Specified, specifiedColor=<rgba 0.0252 0.0319 0.0423 1>}
      uVRegion {bIsValid=false, max=<0 0>, min=<0 0>}
     }
     activeHoveredBrush {
      bIsDynamicallyLoaded=false
      drawAs=Image
      imageType=NoImage
      mirroring=NoMirror
      resourceName=None
      resourceObject=null
      tiling=NoTile
      imageSize=<0 0>
      margin=<ltrb 0 0 0 0>
      outlineSettings {
       bUseBrushTransparency=false
       roundingType=HalfHeightRadius
       width=0
       color {colorUseRule=UseColor_Specified, specifiedColor=<rgba 0 0 0 0>}
       cornerRadii=<0 0 0 0>
      }
      tintColor {colorUseRule=UseColor_Specified, specifiedColor=<rgba 0 0.162 0.7454 1>}
      uVRegion {bIsValid=false, max=<0 0>, min=<0 0>}
     }
     dropIndicator_Above {
      bIsDynamicallyLoaded=false
      drawAs=Box
      imageType=FullColor
      mirroring=NoMirror
      resourceName=../../../Engine/Content/Slate/Common/DropZoneIndicator_Above.png
      resourceObject=null
      tiling=NoTile
      imageSize=<0 0>
      margin=<ltrb 0.625 0.625 0 0>
      outlineSettings {
       bUseBrushTransparency=false
       roundingType=HalfHeightRadius
       width=0
       color {colorUseRule=UseColor_Specified, specifiedColor=<rgba 0 0 0 0>}
       cornerRadii=<0 0 0 0>
      }
      tintColor {colorUseRule=UseColor_Specified, specifiedColor=<rgba 0.0091 0.3515 0.9647 1>}
      uVRegion {bIsValid=false, max=<0 0>, min=<0 0>}
     }
     dropIndicator_Below {
      bIsDynamicallyLoaded=false
      drawAs=Box
      imageType=FullColor
      mirroring=NoMirror
      resourceName=../../../Engine/Content/Slate/Common/DropZoneIndicator_Below.png
      resourceObject=null
      tiling=NoTile
      imageSize=<0 0>
      margin=<ltrb 0.625 0 0 0.625>
      outlineSettings {
       bUseBrushTransparency=false
       roundingType=HalfHeightRadius
       width=0
       color {colorUseRule=UseColor_Specified, specifiedColor=<rgba 0 0 0 0>}
       cornerRadii=<0 0 0 0>
      }
      tintColor {colorUseRule=UseColor_Specified, specifiedColor=<rgba 0.0091 0.3515 0.9647 1>}
      uVRegion {bIsValid=false, max=<0 0>, min=<0 0>}
     }
     dropIndicator_Onto {
      bIsDynamicallyLoaded=false
      drawAs=Box
      imageType=FullColor
      mirroring=NoMirror
      resourceName=../../../Engine/Content/Slate/Common/DropZoneIndicator_Onto.png
      resourceObject=null
      tiling=NoTile
      imageSize=<0 0>
      margin=<ltrb 0.25 0.25 0.25 0.25>
      outlineSettings {
       bUseBrushTransparency=false
       roundingType=HalfHeightRadius
       width=0
       color {colorUseRule=UseColor_Specified, specifiedColor=<rgba 0 0 0 0>}
       cornerRadii=<0 0 0 0>
      }
      tintColor {colorUseRule=UseColor_Specified, specifiedColor=<rgba 0.0091 0.3515 0.9647 1>}
      uVRegion {bIsValid=false, max=<0 0>, min=<0 0>}
     }
     evenRowBackgroundBrush {
      bIsDynamicallyLoaded=false
      drawAs=Image
      imageType=NoImage
      mirroring=NoMirror
      resourceName=None
      resourceObject=null
      tiling=NoTile
      imageSize=<0 0>
      margin=<ltrb 0 0 0 0>
      outlineSettings {
       bUseBrushTransparency=false
       roundingType=HalfHeightRadius
       width=0
       color {colorUseRule=UseColor_Specified, specifiedColor=<rgba 0 0 0 0>}
       cornerRadii=<0 0 0 0>
      }
      tintColor {colorUseRule=UseColor_Specified, specifiedColor=<rgba 0.0103 0.0103 0.0103 1>}
      uVRegion {bIsValid=false, max=<0 0>, min=<0 0>}
     }
     evenRowBackgroundHoveredBrush {
      bIsDynamicallyLoaded=false
      drawAs=Image
      imageType=NoImage
      mirroring=NoMirror
      resourceName=None
      resourceObject=null
      tiling=NoTile
      imageSize=<0 0>
      margin=<ltrb 0 0 0 0>
      outlineSettings {
       bUseBrushTransparency=false
       roundingType=HalfHeightRadius
       width=0
       color {colorUseRule=UseColor_Specified, specifiedColor=<rgba 0 0 0 0>}
       cornerRadii=<0 0 0 0>
      }
      tintColor {colorUseRule=UseColor_Specified, specifiedColor=<rgba 0.0176 0.0176 0.0176 1>}
      uVRegion {bIsValid=false, max=<0 0>, min=<0 0>}
     }
     inactiveBrush {
      bIsDynamicallyLoaded=false
      drawAs=Image
      imageType=NoImage
      mirroring=NoMirror
      resourceName=None
      resourceObject=null
      tiling=NoTile
      imageSize=<0 0>
      margin=<ltrb 0 0 0 0>
      outlineSettings {
       bUseBrushTransparency=false
       roundingType=HalfHeightRadius
       width=0
       color {colorUseRule=UseColor_Specified, specifiedColor=<rgba 0 0 0 0>}
       cornerRadii=<0 0 0 0>
      }
      tintColor {colorUseRule=UseColor_Specified, specifiedColor=<rgba 0.0513 0.0953 0.159 1>}
      uVRegion {bIsValid=false, max=<0 0>, min=<0 0>}
     }
     inactiveHighlightedBrush {
      bIsDynamicallyLoaded=false
      drawAs=Image
      imageType=NoImage
      mirroring=NoMirror
      resourceName=None
      resourceObject=null
      tiling=NoTile
      imageSize=<0 0>
      margin=<ltrb 0 0 0 0>
      outlineSettings {
       bUseBrushTransparency=false
       roundingType=HalfHeightRadius
       width=0
       color {colorUseRule=UseColor_Specified, specifiedColor=<rgba 0 0 0 0>}
       cornerRadii=<0 0 0 0>
      }
      tintColor {colorUseRule=UseColor_Specified, specifiedColor=<rgba 0.0252 0.0319 0.0423 1>}
      uVRegion {bIsValid=false, max=<0 0>, min=<0 0>}
     }
     inactiveHoveredBrush {
      bIsDynamicallyLoaded=false
      drawAs=Image
      imageType=NoImage
      mirroring=NoMirror
      resourceName=None
      resourceObject=null
      tiling=NoTile
      imageSize=<0 0>
      margin=<ltrb 0 0 0 0>
      outlineSettings {
       bUseBrushTransparency=false
       roundingType=HalfHeightRadius
       width=0
       color {colorUseRule=UseColor_Specified, specifiedColor=<rgba 0 0 0 0>}
       cornerRadii=<0 0 0 0>
      }
      tintColor {colorUseRule=UseColor_Specified, specifiedColor=<rgba 0.0513 0.0953 0.159 1>}
      uVRegion {bIsValid=false, max=<0 0>, min=<0 0>}
     }
     oddRowBackgroundBrush {
      bIsDynamicallyLoaded=false
      drawAs=Image
      imageType=NoImage
      mirroring=NoMirror
      resourceName=None
      resourceObject=null
      tiling=NoTile
      imageSize=<0 0>
      margin=<ltrb 0 0 0 0>
      outlineSettings {
       bUseBrushTransparency=false
       roundingType=HalfHeightRadius
       width=0
       color {colorUseRule=UseColor_Specified, specifiedColor=<rgba 0 0 0 0>}
       cornerRadii=<0 0 0 0>
      }
      tintColor {colorUseRule=UseColor_Specified, specifiedColor=<rgba 0.0103 0.0103 0.0103 1>}
      uVRegion {bIsValid=false, max=<0 0>, min=<0 0>}
     }
     oddRowBackgroundHoveredBrush {
      bIsDynamicallyLoaded=false
      drawAs=Image
      imageType=NoImage
      mirroring=NoMirror
      resourceName=None
      resourceObject=null
      tiling=NoTile
      imageSize=<0 0>
      margin=<ltrb 0 0 0 0>
      outlineSettings {
       bUseBrushTransparency=false
       roundingType=HalfHeightRadius
       width=0
       color {colorUseRule=UseColor_Specified, specifiedColor=<rgba 0 0 0 0>}
       cornerRadii=<0 0 0 0>
      }
      tintColor {colorUseRule=UseColor_Specified, specifiedColor=<rgba 0.0176 0.0176 0.0176 1>}
      uVRegion {bIsValid=false, max=<0 0>, min=<0 0>}
     }
     parentRowBackgroundBrush {
      bIsDynamicallyLoaded=false
      drawAs=Image
      imageType=NoImage
      mirroring=NoMirror
      resourceName=None
      resourceObject=null
      tiling=NoTile
      imageSize=<32 32>
      margin=<ltrb 0 0 0 0>
      outlineSettings {
       bUseBrushTransparency=false
       roundingType=HalfHeightRadius
       width=0
       color {colorUseRule=UseColor_Specified, specifiedColor=<rgba 0 0 0 0>}
       cornerRadii=<0 0 0 0>
      }
      tintColor {colorUseRule=UseColor_Specified, specifiedColor=<rgba 1 1 1 1>}
      uVRegion {bIsValid=false, max=<0 0>, min=<0 0>}
     }
     parentRowBackgroundHoveredBrush {
      bIsDynamicallyLoaded=false
      drawAs=Image
      imageType=NoImage
      mirroring=NoMirror
      resourceName=None
      resourceObject=null
      tiling=NoTile
      imageSize=<32 32>
      margin=<ltrb 0 0 0 0>
      outlineSettings {
       bUseBrushTransparency=false
       roundingType=HalfHeightRadius
       width=0
       color {colorUseRule=UseColor_Specified, specifiedColor=<rgba 0 0 0 0>}
       cornerRadii=<0 0 0 0>
      }
      tintColor {colorUseRule=UseColor_Specified, specifiedColor=<rgba 1 1 1 1>}
      uVRegion {bIsValid=false, max=<0 0>, min=<0 0>}
     }
     selectedTextColor {colorUseRule=UseColor_Specified, specifiedColor=<rgba 1 1 1 1>}
     selectorFocusedBrush {
      bIsDynamicallyLoaded=false
      drawAs=Border
      imageType=FullColor
      mirroring=NoMirror
      resourceName=../../../Engine/Content/Slate/Common/Selector.png
      resourceObject=null
      tiling=Both
      imageSize=<0 0>
      margin=<ltrb 0.25 0.25 0.25 0.25>
      outlineSettings {
       bUseBrushTransparency=false
       roundingType=HalfHeightRadius
       width=0
       color {colorUseRule=UseColor_Specified, specifiedColor=<rgba 0 0 0 0>}
       cornerRadii=<0 0 0 0>
      }
      tintColor {colorUseRule=UseColor_Specified, specifiedColor=<rgba 0.701 0.225 0.003 1>}
      uVRegion {bIsValid=false, max=<0 0>, min=<0 0>}
     }
     textColor {colorUseRule=UseColor_Specified, specifiedColor=<rgba 0.5271 0.5271 0.5271 1>}
    }
    widgetStyle {
     comboButtonStyle {
      downArrowAlign=VAlign_Center
      buttonStyle {
       clickedSlateSound {resourceObject=null}
       disabled {
        bIsDynamicallyLoaded=false
        drawAs=RoundedBox
        imageType=NoImage
        mirroring=NoMirror
        resourceName=None
        resourceObject=null
        tiling=NoTile
        imageSize=<0 0>
        margin=<ltrb 0 0 0 0>
        outlineSettings {
         bUseBrushTransparency=false
         roundingType=FixedRadius
         width=1
         color {colorUseRule=UseColor_ColorTable, specifiedColor=<rgba 0 0 0 0>}
         cornerRadii=<4 4 4 4>
        }
        tintColor {colorUseRule=UseColor_Specified, specifiedColor=<rgba 0.0395 0.0395 0.0395 1>}
        uVRegion {bIsValid=false, max=<0 0>, min=<0 0>}
       }
       disabledForeground {colorUseRule=UseColor_ColorTable, specifiedColor=<rgba 0 0 0 0>}
       hovered {
        bIsDynamicallyLoaded=false
        drawAs=RoundedBox
        imageType=NoImage
        mirroring=NoMirror
        resourceName=None
        resourceObject=null
        tiling=NoTile
        imageSize=<0 0>
        margin=<ltrb 0 0 0 0>
        outlineSettings {
         bUseBrushTransparency=false
         roundingType=FixedRadius
         width=1
         color {colorUseRule=UseColor_ColorTable, specifiedColor=<rgba 0 0 0 0>}
         cornerRadii=<4 4 4 4>
        }
        tintColor {colorUseRule=UseColor_Specified, specifiedColor=<rgba 0.0048 0.0048 0.0048 1>}
        uVRegion {bIsValid=false, max=<0 0>, min=<0 0>}
       }
       hoveredForeground {colorUseRule=UseColor_ColorTable, specifiedColor=<rgba 0 0 0 0>}
       hoveredSlateSound {resourceObject=null}
       normal {
        bIsDynamicallyLoaded=false
        drawAs=RoundedBox
        imageType=NoImage
        mirroring=NoMirror
        resourceName=None
        resourceObject=null
        tiling=NoTile
        imageSize=<0 0>
        margin=<ltrb 0 0 0 0>
        outlineSettings {
         bUseBrushTransparency=false
         roundingType=FixedRadius
         width=1
         color {colorUseRule=UseColor_ColorTable, specifiedColor=<rgba 0 0 0 0>}
         cornerRadii=<4 4 4 4>
        }
        tintColor {colorUseRule=UseColor_Specified, specifiedColor=<rgba 0.0048 0.0048 0.0048 1>}
        uVRegion {bIsValid=false, max=<0 0>, min=<0 0>}
       }
       normalForeground {colorUseRule=UseColor_ColorTable, specifiedColor=<rgba 0 0 0 0>}
       normalPadding=<ltrb 8 1 8 1>
       pressed {
        bIsDynamicallyLoaded=false
        drawAs=RoundedBox
        imageType=NoImage
        mirroring=NoMirror
        resourceName=None
        resourceObject=null
        tiling=NoTile
        imageSize=<0 0>
        margin=<ltrb 0 0 0 0>
        outlineSettings {
         bUseBrushTransparency=false
         roundingType=FixedRadius
         width=1
         color {colorUseRule=UseColor_ColorTable, specifiedColor=<rgba 0 0 0 0>}
         cornerRadii=<4 4 4 4>
        }
        tintColor {colorUseRule=UseColor_Specified, specifiedColor=<rgba 0.0048 0.0048 0.0048 1>}
        uVRegion {bIsValid=false, max=<0 0>, min=<0 0>}
       }
       pressedForeground {colorUseRule=UseColor_ColorTable, specifiedColor=<rgba 0 0 0 0>}
       pressedPadding=<ltrb 8 2 8 0>
       pressedSlateSound {resourceObject=null}
      }
      contentPadding=<ltrb 0 0 0 0>
      downArrowImage {
       bIsDynamicallyLoaded=false
       drawAs=Image
       imageType=Vector
       mirroring=NoMirror
       resourceName=../../../Engine/Content/Slate/Starship/CoreWidgets/ComboBox/wide-chevron-down.svg
       resourceObject=null
       tiling=NoTile
       imageSize=<16 16>
       margin=<ltrb 0 0 0 0>
       outlineSettings {
        bUseBrushTransparency=false
        roundingType=HalfHeightRadius
        width=0
        color {colorUseRule=UseColor_Specified, specifiedColor=<rgba 0 0 0 0>}
        cornerRadii=<0 0 0 0>
       }
       tintColor {colorUseRule=UseColor_Specified, specifiedColor=<rgba 1 1 1 1>}
       uVRegion {bIsValid=false, max=<0 0>, min=<0 0>}
      }
      downArrowPadding=<ltrb 2 2 2 2>
      menuBorderBrush {
       bIsDynamicallyLoaded=false
       drawAs=RoundedBox
       imageType=NoImage
       mirroring=NoMirror
       resourceName=None
       resourceObject=null
       tiling=NoTile
       imageSize=<0 0>
       margin=<ltrb 0 0 0 0>
       outlineSettings {
        bUseBrushTransparency=false
        roundingType=FixedRadius
        width=1
        color {colorUseRule=UseColor_ColorTable, specifiedColor=<rgba 0 0 0 0>}
        cornerRadii=<0 0 0 0>
       }
       tintColor {colorUseRule=UseColor_Specified, specifiedColor=<rgba 0.0395 0.0395 0.0395 1>}
       uVRegion {bIsValid=false, max=<0 0>, min=<0 0>}
      }
      menuBorderPadding=<ltrb 0 0 0 0>
      shadowColorAndOpacity=<rgba 0 0 0 1>
      shadowOffset=<0 0>
     }
     contentPadding=<ltrb 0 0 0 0>
     menuRowPadding=<ltrb 8 3 8 3>
     pressedSlateSound {resourceObject=null}
     selectionChangeSlateSound {resourceObject=null}
    }
   }
   slot {class=HorizontalBoxSlot, properties {padding=<ltrb 45 0 0 0>}}
  }
  GenericAssetWarning {
   widgetClassPath=&CommonTextBlock
   bIsVariable=true
   parent=HorizontalBox_238
   widget=GenericAssetWarning
   properties {
    text=(Generic Read Only asset)
    colorAndOpacity {colorUseRule=UseColor_Specified, specifiedColor=<rgba 0.5156 0.1504 0.1504 1>}
    font {
     bForceMonospaced=false
     bMaterialIsStencil=false
     fontMaterial=null
     fontObject=&NotoSans
     letterSpacing=0
     monospacedWidth=1
     size=15
     skewAmount=0
     typefaceFontName=NotoSans-Italic
     outlineSettings {
      bApplyOutlineToDropShadows=false
      bMiteredCorners=false
      bSeparateFillAlpha=false
      outlineMaterial=null
      outlineSize=0
      outlineColor=<rgba 0 0 0 1>
     }
    }
    shadowOffset=<0 0>
   }
   slot {class=HorizontalBoxSlot, properties {padding=<ltrb 10 0 0 0>}}
  }
  ScrollBox_53 {
   widgetClassPath=&ScrollBox
   parent=VerticalBox_239
   widget=ScrollBox_53
   properties {
    widgetBarStyle {
     thickness=8
     draggedThumbImage {
      bIsDynamicallyLoaded=false
      drawAs=RoundedBox
      imageType=NoImage
      mirroring=NoMirror
      resourceName=None
      resourceObject=null
      tiling=NoTile
      imageSize=<0 0>
      margin=<ltrb 0 0 0 0>
      outlineSettings {
       bUseBrushTransparency=false
       roundingType=FixedRadius
       width=0
       color {colorUseRule=UseColor_Specified, specifiedColor=<rgba 0 0 0 0>}
       cornerRadii=<4 4 4 4>
      }
      tintColor {colorUseRule=UseColor_Specified, specifiedColor=<rgba 0.2159 0.2159 0.2159 1>}
      uVRegion {bIsValid=false, max=<0 0>, min=<0 0>}
     }
     horizontalBackgroundImage {
      bIsDynamicallyLoaded=false
      drawAs=NoDrawType
      imageType=NoImage
      mirroring=NoMirror
      resourceName=None
      resourceObject=null
      tiling=NoTile
      imageSize=<0 0>
      margin=<ltrb 0 0 0 0>
      outlineSettings {
       bUseBrushTransparency=false
       roundingType=HalfHeightRadius
       width=0
       color {colorUseRule=UseColor_Specified, specifiedColor=<rgba 0 0 0 0>}
       cornerRadii=<0 0 0 0>
      }
      tintColor {colorUseRule=UseColor_Specified, specifiedColor=<rgba 1 1 1 1>}
      uVRegion {bIsValid=false, max=<0 0>, min=<0 0>}
     }
     horizontalBottomSlotImage {
      bIsDynamicallyLoaded=false
      drawAs=NoDrawType
      imageType=NoImage
      mirroring=NoMirror
      resourceName=None
      resourceObject=null
      tiling=NoTile
      imageSize=<0 0>
      margin=<ltrb 0 0 0 0>
      outlineSettings {
       bUseBrushTransparency=false
       roundingType=HalfHeightRadius
       width=0
       color {colorUseRule=UseColor_Specified, specifiedColor=<rgba 0 0 0 0>}
       cornerRadii=<0 0 0 0>
      }
      tintColor {colorUseRule=UseColor_Specified, specifiedColor=<rgba 1 1 1 1>}
      uVRegion {bIsValid=false, max=<0 0>, min=<0 0>}
     }
     horizontalTopSlotImage {
      bIsDynamicallyLoaded=false
      drawAs=NoDrawType
      imageType=NoImage
      mirroring=NoMirror
      resourceName=None
      resourceObject=null
      tiling=NoTile
      imageSize=<0 0>
      margin=<ltrb 0 0 0 0>
      outlineSettings {
       bUseBrushTransparency=false
       roundingType=HalfHeightRadius
       width=0
       color {colorUseRule=UseColor_Specified, specifiedColor=<rgba 0 0 0 0>}
       cornerRadii=<0 0 0 0>
      }
      tintColor {colorUseRule=UseColor_Specified, specifiedColor=<rgba 1 1 1 1>}
      uVRegion {bIsValid=false, max=<0 0>, min=<0 0>}
     }
     hoveredThumbImage {
      bIsDynamicallyLoaded=false
      drawAs=RoundedBox
      imageType=NoImage
      mirroring=NoMirror
      resourceName=None
      resourceObject=null
      tiling=NoTile
      imageSize=<0 0>
      margin=<ltrb 0 0 0 0>
      outlineSettings {
       bUseBrushTransparency=false
       roundingType=FixedRadius
       width=0
       color {colorUseRule=UseColor_Specified, specifiedColor=<rgba 0 0 0 0>}
       cornerRadii=<4 4 4 4>
      }
      tintColor {colorUseRule=UseColor_Specified, specifiedColor=<rgba 0.2159 0.2159 0.2159 1>}
      uVRegion {bIsValid=false, max=<0 0>, min=<0 0>}
     }
     normalThumbImage {
      bIsDynamicallyLoaded=false
      drawAs=RoundedBox
      imageType=NoImage
      mirroring=NoMirror
      resourceName=None
      resourceObject=null
      tiling=NoTile
      imageSize=<0 0>
      margin=<ltrb 0 0 0 0>
      outlineSettings {
       bUseBrushTransparency=false
       roundingType=FixedRadius
       width=0
       color {colorUseRule=UseColor_Specified, specifiedColor=<rgba 0 0 0 0>}
       cornerRadii=<4 4 4 4>
      }
      tintColor {colorUseRule=UseColor_Specified, specifiedColor=<rgba 0.0953 0.0953 0.0953 1>}
      uVRegion {bIsValid=false, max=<0 0>, min=<0 0>}
     }
     verticalBackgroundImage {
      bIsDynamicallyLoaded=false
      drawAs=NoDrawType
      imageType=NoImage
      mirroring=NoMirror
      resourceName=None
      resourceObject=null
      tiling=NoTile
      imageSize=<0 0>
      margin=<ltrb 0 0 0 0>
      outlineSettings {
       bUseBrushTransparency=false
       roundingType=HalfHeightRadius
       width=0
       color {colorUseRule=UseColor_Specified, specifiedColor=<rgba 0 0 0 0>}
       cornerRadii=<0 0 0 0>
      }
      tintColor {colorUseRule=UseColor_Specified, specifiedColor=<rgba 1 1 1 1>}
      uVRegion {bIsValid=false, max=<0 0>, min=<0 0>}
     }
     verticalBottomSlotImage {
      bIsDynamicallyLoaded=false
      drawAs=NoDrawType
      imageType=NoImage
      mirroring=NoMirror
      resourceName=None
      resourceObject=null
      tiling=NoTile
      imageSize=<0 0>
      margin=<ltrb 0 0 0 0>
      outlineSettings {
       bUseBrushTransparency=false
       roundingType=HalfHeightRadius
       width=0
       color {colorUseRule=UseColor_Specified, specifiedColor=<rgba 0 0 0 0>}
       cornerRadii=<0 0 0 0>
      }
      tintColor {colorUseRule=UseColor_Specified, specifiedColor=<rgba 1 1 1 1>}
      uVRegion {bIsValid=false, max=<0 0>, min=<0 0>}
     }
     verticalTopSlotImage {
      bIsDynamicallyLoaded=false
      drawAs=NoDrawType
      imageType=NoImage
      mirroring=NoMirror
      resourceName=None
      resourceObject=null
      tiling=NoTile
      imageSize=<0 0>
      margin=<ltrb 0 0 0 0>
      outlineSettings {
       bUseBrushTransparency=false
       roundingType=HalfHeightRadius
       width=0
       color {colorUseRule=UseColor_Specified, specifiedColor=<rgba 0 0 0 0>}
       cornerRadii=<0 0 0 0>
      }
      tintColor {colorUseRule=UseColor_Specified, specifiedColor=<rgba 1 1 1 1>}
      uVRegion {bIsValid=false, max=<0 0>, min=<0 0>}
     }
    }
    widgetStyle {
     barThickness=8
     bottomShadowBrush {
      bIsDynamicallyLoaded=false
      drawAs=Box
      imageType=FullColor
      mirroring=NoMirror
      resourceName=../../../Engine/Content/Slate/Common/ScrollBoxShadowBottom.png
      resourceObject=null
      tiling=NoTile
      imageSize=<16 8>
      margin=<ltrb 0.5 0 0.5 1>
      outlineSettings {
       bUseBrushTransparency=false
       roundingType=HalfHeightRadius
       width=0
       color {colorUseRule=UseColor_Specified, specifiedColor=<rgba 0 0 0 0>}
       cornerRadii=<0 0 0 0>
      }
      tintColor {colorUseRule=UseColor_Specified, specifiedColor=<rgba 1 1 1 1>}
      uVRegion {bIsValid=false, max=<0 0>, min=<0 0>}
     }
     horizontalScrolledContentPadding=<ltrb 0 0 1 0>
     leftShadowBrush {
      bIsDynamicallyLoaded=false
      drawAs=Box
      imageType=FullColor
      mirroring=NoMirror
      resourceName=../../../Engine/Content/Slate/Common/ScrollBoxShadowLeft.png
      resourceObject=null
      tiling=NoTile
      imageSize=<8 16>
      margin=<ltrb 1 0.5 0 0.5>
      outlineSettings {
       bUseBrushTransparency=false
       roundingType=HalfHeightRadius
       width=0
       color {colorUseRule=UseColor_Specified, specifiedColor=<rgba 0 0 0 0>}
       cornerRadii=<0 0 0 0>
      }
      tintColor {colorUseRule=UseColor_Specified, specifiedColor=<rgba 1 1 1 1>}
      uVRegion {bIsValid=false, max=<0 0>, min=<0 0>}
     }
     rightShadowBrush {
      bIsDynamicallyLoaded=false
      drawAs=Box
      imageType=FullColor
      mirroring=NoMirror
      resourceName=../../../Engine/Content/Slate/Common/ScrollBoxShadowRight.png
      resourceObject=null
      tiling=NoTile
      imageSize=<8 16>
      margin=<ltrb 0 0.5 1 0.5>
      outlineSettings {
       bUseBrushTransparency=false
       roundingType=HalfHeightRadius
       width=0
       color {colorUseRule=UseColor_Specified, specifiedColor=<rgba 0 0 0 0>}
       cornerRadii=<0 0 0 0>
      }
      tintColor {colorUseRule=UseColor_Specified, specifiedColor=<rgba 1 1 1 1>}
      uVRegion {bIsValid=false, max=<0 0>, min=<0 0>}
     }
     topShadowBrush {
      bIsDynamicallyLoaded=false
      drawAs=Box
      imageType=FullColor
      mirroring=NoMirror
      resourceName=../../../Engine/Content/Slate/Common/ScrollBoxShadowTop.png
      resourceObject=null
      tiling=NoTile
      imageSize=<16 8>
      margin=<ltrb 0.5 1 0.5 0>
      outlineSettings {
       bUseBrushTransparency=false
       roundingType=HalfHeightRadius
       width=0
       color {colorUseRule=UseColor_Specified, specifiedColor=<rgba 0 0 0 0>}
       cornerRadii=<0 0 0 0>
      }
      tintColor {colorUseRule=UseColor_Specified, specifiedColor=<rgba 1 1 1 1>}
      uVRegion {bIsValid=false, max=<0 0>, min=<0 0>}
     }
     verticalScrolledContentPadding=<ltrb 0 0 0 1>
    }
   }
   slot {class=VerticalBoxSlot, properties {size {sizeRule=Fill, value=1}}}
  }
  HorizontalBox_69 {
   widgetClassPath=&HorizontalBox
   parent=ScrollBox_53
   widget=HorizontalBox_69
   slot {class=ScrollBoxSlot, properties {padding=<ltrb 0 5 0 0>}}
  }
  ActivatableWidgetSwitcher {
   widgetClassPath=&CommonActivatableWidgetSwitcher
   bIsVariable=true
   parent=HorizontalBox_69
   widget=ActivatableWidgetSwitcher
   properties {renderOpacity=0.7, visibility=HitTestInvisible}
   slot {class=HorizontalBoxSlot, properties {size {sizeRule=Fill, value=1}}}
  }
  InventoryComponentView {
   widgetClassPath=&DetailsView
   parent=ActivatableWidgetSwitcher
   widget=InventoryComponentView
   properties {propertiesToShow [Items,LootInfo,LoadoutTable]}
   slot {class=WidgetSwitcherSlot}
  }
  AffiliationComponentView {
   widgetClassPath=&DetailsView
   parent=ActivatableWidgetSwitcher
   widget=AffiliationComponentView
   properties {propertiesToShow [FactionTags]}
   slot {class=WidgetSwitcherSlot}
  }
  GameplayComponentView {
   widgetClassPath=&DetailsView
   parent=ActivatableWidgetSwitcher
   widget=GameplayComponentView
   properties {propertiesToShow [NameSettings, CanRespawnOnDeath]}
   slot {class=WidgetSwitcherSlot}
  }
  ConversationComponentView {
   widgetClassPath=&DetailsView
   parent=ActivatableWidgetSwitcher
   widget=ConversationComponentView
   properties {propertiesToShow [
     ConversationComponentClass
     ConversationEntryTag
     EnterBTTaskConversationIfPlayerAroundDist
     CharacterData
     SpawnOutOfTheLoop
     SpecialActorComponentClass
     GreetingEmotes
     GreetingEmoteMaxDuration
    ]}
   slot {class=WidgetSwitcherSlot}
  }
  CosmeticComponentView {
   widgetClassPath=&DetailsView
   parent=ActivatableWidgetSwitcher
   widget=CosmeticComponentView
   properties {propertiesToShow [WeightedLoadouts]}
   slot {class=WidgetSwitcherSlot}
  }
  GameplayAbilityComponentView {
   widgetClassPath=&DetailsView
   parent=ActivatableWidgetSwitcher
   widget=GameplayAbilityComponentView
   properties {propertiesToShow [LooseTagstoApplytoPawn]}
   slot {class=WidgetSwitcherSlot}
  }
  SpawnParamsComponentView {
   widgetClassPath=&DetailsView
   parent=ActivatableWidgetSwitcher
   widget=SpawnParamsComponentView
   properties {propertiesToShow [PawnStatHandle]}
   slot {class=WidgetSwitcherSlot}
  }
  NPCCharacterDataView {
   widgetClassPath=&DetailsView
   parent=ActivatableWidgetSwitcher
   widget=NPCCharacterDataView
   slot {class=WidgetSwitcherSlot}
  }
  LagerEventDataView {
   widgetClassPath=&DetailsView
   parent=ActivatableWidgetSwitcher
   widget=LagerEventDataView
   slot {class=WidgetSwitcherSlot}
  }
  FullDetailedView {
   widgetClassPath=&DetailsView
   parent=ActivatableWidgetSwitcher
   widget=FullDetailedView
   slot {class=WidgetSwitcherSlot}
  }
 ]
}
